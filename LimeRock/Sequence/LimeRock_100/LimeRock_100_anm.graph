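# Animation Level Sequence for the LimeRock_100 cinematic shot, staging creature tentacles around a portal.
# A main tentacle parented to BP_LimeRock_Portal plays skeletal animation with two layered transform sections, while cut and background tentacles switch visible late in the shot.
# It also animates a hidden temp demon sphere and places a mannequin manager.
# Sub-tracks nest sequences for Jones, Hope, Kang Kodos, tentacles, temporary players and other subsequences.
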LevelSequence /LimeRock/Sequence/LimeRock_100/LimeRock_100_anm
bindings [
 Tentacle_LimeRock_Creature_BG_2 {
  id=#1
  possessed_class=FortSkeletalMeshActor
  tracks [
   {type=MovieSceneSpawnTrack, sections [{type=MovieSceneSpawnSection, channels [None {type=MovieSceneScriptingBoolChannel, default=true}]}]}
   {
    type=MovieScene3DTransformTrack
    sections [{
      type=MovieScene3DTransformSection
      channels [
       Location.X {type=MovieSceneScriptingDoubleChannel, default=-18949.4036, const [-13092.668,2]}
       Location.Y {type=MovieSceneScriptingDoubleChannel, default=-9241.7259, const [-3644.4551,2]}
       Location.Z {type=MovieSceneScriptingDoubleChannel, default=-57.9366, const [5990.8149,2]}
       Rotation.X {type=MovieSceneScriptingDoubleChannel, default=0, const [38.261,2]}
       Rotation.Y {type=MovieSceneScriptingDoubleChannel, default=0, const [-1.4644,2]}
       Rotation.Z {type=MovieSceneScriptingDoubleChannel, default=0, const [141.9839,2]}
       Scale.X {type=MovieSceneScriptingDoubleChannel, default=1, const [1,2]}
       Scale.Y {type=MovieSceneScriptingDoubleChannel, default=1, const [1,2]}
       Scale.Z {type=MovieSceneScriptingDoubleChannel, default=1, const [1,2]}
      ]
     }]
   }
   {
    type=MovieSceneVisibilityTrack
    sections [{type=MovieSceneVisibilitySection, channels [None {type=MovieSceneScriptingBoolChannel, default=true, keys [[8343,false],[8345,true]]}]}]
   }
   {type=MovieSceneSkeletalAnimationTrack, sections [{type=MovieSceneSkeletalAnimationSection, end_frame=9415, start_frame=7021}]}
  ]
 }
 Demon_TempSphere {
  id=#2
  possessed_class=FortStaticMeshActor
  tracks [
   {
    type=MovieSceneVisibilityTrack
    sections [{type=MovieSceneVisibilitySection, channels [None {type=MovieSceneScriptingBoolChannel, default=false}]}]
   }
   {type=MovieSceneSpawnTrack, sections [{type=MovieSceneSpawnSection, channels [None {type=MovieSceneScriptingBoolChannel, default=true}]}]}
   {
    type=MovieScene3DTransformTrack
    sections [{
      type=MovieScene3DTransformSection
      channels [
       Location.X {
        type=MovieSceneScriptingDoubleChannel
        default=-126385.2569
        thinned_from=19
        keys [[3600,-143315.7188],[7568,-143315.7144],[7617,-138031.6675],[7652,-134968.4335],[7843,-126298.5219],[8040,-121323.2236],[8287,-108289.4899],[8391,-107124.3427],[8480,-111781.0936],[8520,-113517.552],[8595,-114515.7928],[8657,-114515.7915],[8839,-108799.9182],[9055,-108799.9182],[9164,-106244.0573],[9284,-107095.8897]]
       }
       Location.Y {
        type=MovieSceneScriptingDoubleChannel
        default=-95370.1841
        thinned_from=22
        keys [[3600,-108090.7969],[7652,-108090.8001],[7860,-107019.7863],[8024,-105324.9603],[8113,-104199.1751],[8201,-102562.2946],[8340,-98621.6994],[8383,-97237.1498],[8430,-96634.0324],[8518,-97071.6622],[8602,-96715.2884],[8657,-96763.8828],[8863,-94092.5499],[8963,-93693.9648],[9043,-93588.7262],[9167,-92137.6565],[9196,-92176.7647]]
       }
       Location.Z {
        type=MovieSceneScriptingDoubleChannel
        default=18966
        thinned_from=20
        keys [[3600,2987.7542],[7568,2987.7542],[7669,3728.7775],[7758,6332.9637],[7851,11382.1212],[8024,15058.2116],[8107,13430.9823],[8186,9301.1332],[8273,8475.9378],[8401,10572.6676],[8508,18286.7021],[8603,16852.9777],[8742,15759.8049],[8909,17083.008],[8960,16933.3215],[9096,15910.4361],[9199,16847.5552],[9353,16170.8219]]
       }
       Rotation.X {
        type=MovieSceneScriptingDoubleChannel
        default=0
        thinned_from=19
        keys [[3600,0],[8155,0],[8218,-4.5979],[8271,-0.6223],[8355,7.2469],[8452,1.267],[8524,0],[8636,0],[8775,11.0143],[8858,16.7789],[9052,14.8718],[9186,23.5859],[9295,15.7686]]
       }
       Rotation.Y {
        type=MovieSceneScriptingDoubleChannel
        default=0
        thinned_from=17
        keys [[3600,0],[7963,0],[8136,25.012],[8212,-25.1038],[8299,-32.053],[8377,-3.7369],[8506,14.8252],[8598,-6.5337],[8737,-9.9857],[8947,-11.5553],[9021,-9.6417],[9179,-47.4731]]
       }
       Rotation.Z {
        type=MovieSceneScriptingDoubleChannel
        default=0
        thinned_from=12
        keys [[3600,15.8039],[8260,15.8039],[8500,4.5636],[8617,2.1088],[8906,-0.2803],[9043,-3.3854],[9227,-18.8038]]
       }
       Scale.X {type=MovieSceneScriptingDoubleChannel, default=60, const [60,2]}
       Scale.Y {type=MovieSceneScriptingDoubleChannel, default=60, const [60,2]}
       Scale.Z {type=MovieSceneScriptingDoubleChannel, default=74.6732, const [74.6732,2]}
      ]
     }]
   }
  ]
 }
 LimeRock_SceneRoot {id=#3, possessed_class=Actor, tracks []}
 TentacleCut_LimeRock_Creature {
  id=#4
  possessed_class=FortSkeletalMeshActor
  tracks [
   {
    type=MovieScene3DTransformTrack
    sections [{
      type=MovieScene3DTransformSection
      channels [
       Location.X {type=MovieSceneScriptingDoubleChannel, default=-15087.45}
       Location.Y {type=MovieSceneScriptingDoubleChannel, default=9.4932}
       Location.Z {type=MovieSceneScriptingDoubleChannel, default=4257.6446}
       Rotation.X {type=MovieSceneScriptingDoubleChannel, default=36.1864}
       Rotation.Y {type=MovieSceneScriptingDoubleChannel, default=-11.8528}
       Rotation.Z {type=MovieSceneScriptingDoubleChannel, default=-49.4069}
       Scale.X {type=MovieSceneScriptingDoubleChannel, default=1}
       Scale.Y {type=MovieSceneScriptingDoubleChannel, default=1}
       Scale.Z {type=MovieSceneScriptingDoubleChannel, default=1}
      ]
     }]
   }
   {
    type=MovieSceneVisibilityTrack
    sections [{type=MovieSceneVisibilitySection, channels [None {type=MovieSceneScriptingBoolChannel, default=false, keys [[8336,false],[8342,true]]}]}]
   }
   {type=MovieSceneSkeletalAnimationTrack, sections [{type=MovieSceneSkeletalAnimationSection, end_frame=9415, start_frame=7021}]}
  ]
 }
 Tentacle_LimeRock_Creature {
  id=#5
  parent_id=#6
  possessed_class=SkeletalMeshComponent
  tracks [
   {
    type=MovieSceneSkeletalAnimationTrack
    sections [{
      type=MovieSceneSkeletalAnimationSection
      end_frame=10251
      start_frame=1315
      channels [Weight {type=MovieSceneScriptingFloatChannel, default=1, const [1,6]}]
     }]
   }
   {
    type=MovieScene3DTransformTrack
    sections [
     {
      type=MovieScene3DTransformSection
      channels [
       Location.X {
        type=MovieSceneScriptingDoubleChannel
        default=-1803
        thinned_from=14
        keys [[3600,-1524.7745],[4841,-1524.7746],[5902,-1532.6347],[5937,-1604.2585],[6042,-1987.2024],[6693,-1987.0837],[6758,-1947.2703],[6951,-1564.9128],[7070,-1531.0894],[7152,-1531.3217]]
       }
       Location.Y {
        type=MovieSceneScriptingDoubleChannel
        default=-966
        thinned_from=15
        keys [[3600,-1273.8304],[4841,-1273.8304],[4961,-1278.0388],[5050,-1283.5927],[5767,-1298.08],[5902,-1292.6509],[5937,-1042.0029],[6042,-594.9387],[6285,-593.417],[6693,-595.3813],[6758,-710.9603],[6951,-1232.0418],[7070,-1298.1377],[7152,-1300.2707]]
       }
       Location.Z {
        type=MovieSceneScriptingDoubleChannel
        default=-3578
        thinned_from=14
        keys [[3600,-3341.2949],[5767,-3343.9559],[5902,-3334.3087],[5937,-2888.7969],[6042,-2094.1105],[6285,-2091.3752],[6693,-2094.8667],[6758,-2126.5423],[6951,-3217.8023],[7070,-3343.9898],[7152,-3345.2875]]
       }
       Rotation.X {
        type=MovieSceneScriptingDoubleChannel
        default=9
        keys [[3600,-20.4328],[4841,-20.4328],[5050,-20.0152],[5767,-20.0806],[5813,-19.6089],[5902,-19.5617],[5937,-7.5901],[5981,3.9746],[6042,5.3359],[6285,4.4527],[6693,4.4004],[6758,-2.86],[6861,9],[6951,-18.0959],[7070,-20.1706],[7152,-20.1706]]
       }
       Rotation.Y {
        type=MovieSceneScriptingDoubleChannel
        default=31
        keys [[3600,10.1822],[4841,10.1822],[4961,9.8785],[5050,10.1826],[5767,9.9427],[5902,9.9238],[5937,8.273],[6042,1.561],[6285,1.1572],[6693,1.334],[6758,1.3956],[6861,31],[6951,9.0877],[7070,10],[7152,10]]
       }
       Rotation.Z {
        type=MovieSceneScriptingDoubleChannel
        default=-3
        thinned_from=15
        keys [[3600,-66.1616],[5767,-66.0688],[5813,-65.4629],[5902,-65.1335],[5937,-39.337],[5981,5.83],[6042,-5.5594],[6285,-7.6911],[6693,-7.8045],[6758,-3.1002],[6951,-57.8656],[7070,-66.0586],[7152,-66.0586]]
       }
       Scale.X {type=MovieSceneScriptingDoubleChannel, default=0.5, const [0.5,11]}
       Scale.Y {type=MovieSceneScriptingDoubleChannel, default=0.5, const [0.5,11]}
       Scale.Z {type=MovieSceneScriptingDoubleChannel, default=0.5, const [0.5,11]}
      ]
     }
     {
      type=MovieScene3DTransformSection
      channels [
       Location.X {
        type=MovieSceneScriptingDoubleChannel
        default=883
        keys [[3600,0],[5050,0],[5767,0.0143],[5868,3.5427],[5902,4.4623],[5937,265.0146],[6042,828.2173],[6285,883],[6693,883],[6758,881.5462],[6951,54.13],[7070,0]]
       }
       Location.Y {
        type=MovieSceneScriptingDoubleChannel
        default=-1336
        keys [[3600,0],[5050,0],[5767,-0.0216],[5868,-5.3602],[5902,-6.7515],[5937,-400.9733],[6042,-1253.1125],[6285,-1336],[6693,-1336],[6758,-1333.8004],[6951,-81.9],[7070,0]]
       }
       Location.Z {
        type=MovieSceneScriptingDoubleChannel
        default=2202
        keys [[3600,0],[5050,0],[5767,0.0356],[5868,8.8346],[5902,11.1278],[5937,660.8857],[6042,2065.3845],[6285,2202],[6693,2202],[6758,2198.3746],[6951,134.9879],[7070,0]]
       }
       Rotation.X {
        type=MovieSceneScriptingDoubleChannel
        default=80
        thinned_from=11
        keys [[3600,0],[5050,0],[5767,0.283],[5902,0.4043],[5937,24.0104],[6042,78.6351],[6285,80],[6758,79.674],[6951,4.8923],[7070,0]]
       }
       Rotation.Y {
        type=MovieSceneScriptingDoubleChannel
        default=-4
        keys [[3600,0],[5050,0],[5767,-0.0142],[5902,-0.0202],[5937,-30.4312],[6042,-32.8009],[6285,-32.8305],[6693,-33.141],[6758,-33.1507],[6951,-2.0356],[7070,0]]
       }
       Rotation.Z {
        type=MovieSceneScriptingDoubleChannel
        default=51
        thinned_from=11
        keys [[3600,0],[5050,0],[5767,0.0318],[5902,0.0455],[5937,2.7012],[6042,8.8464],[6285,9],[6758,8.9537],[6951,0.5498],[7070,0]]
       }
       Scale.X {type=MovieSceneScriptingDoubleChannel, default=0, const [0,9]}
       Scale.Y {type=MovieSceneScriptingDoubleChannel, default=0, const [0,9]}
       Scale.Z {type=MovieSceneScriptingDoubleChannel, default=0, const [0,9]}
      ]
     }
    ]
   }
  ]
 }
 BP_LimeRock_Portal {
  id=#6
  possessed_class=BP_LimeRock_Portal_C
  tracks [{
    type=MovieScene3DTransformTrack
    sections [{
      type=MovieScene3DTransformSection
      channels [
       Location.X {type=MovieSceneScriptingDoubleChannel, default=-101065.1944, const [-101065.1953,2]}
       Location.Y {type=MovieSceneScriptingDoubleChannel, default=-99603.4429, const [-99603.4453,2]}
       Location.Z {type=MovieSceneScriptingDoubleChannel, default=6710.2577, const [6710.2578,2]}
       Rotation.X {type=MovieSceneScriptingDoubleChannel, default=1.7056, const [1.7056,2]}
       Rotation.Y {type=MovieSceneScriptingDoubleChannel, default=-17.3156, const [-17.3156,2]}
       Rotation.Z {type=MovieSceneScriptingDoubleChannel, default=15.1815, const [15.1815,2]}
       Scale.X {type=MovieSceneScriptingDoubleChannel, default=1, const [1,2]}
       Scale.Y {type=MovieSceneScriptingDoubleChannel, default=1, const [1,2]}
       Scale.Z {type=MovieSceneScriptingDoubleChannel, default=1, const [1,2]}
      ]
     }]
   }]
 }
 Tentacle_LimeRock_Creature_BG_1 {
  id=#7
  possessed_class=FortSkeletalMeshActor
  tracks [
   {type=MovieSceneSpawnTrack, sections [{type=MovieSceneSpawnSection, channels [None {type=MovieSceneScriptingBoolChannel, default=true}]}]}
   {
    type=MovieScene3DTransformTrack
    sections [{
      type=MovieScene3DTransformSection
      channels [
       Location.X {type=MovieSceneScriptingDoubleChannel, default=-23269.4036, keys [[8686,-13426.2261],[11083,-13885.7471]]}
       Location.Y {type=MovieSceneScriptingDoubleChannel, default=4782.2741, keys [[8686,3447.7924],[11083,3553.6179]]}
       Location.Z {type=MovieSceneScriptingDoubleChannel, default=-57.9366, keys [[8686,6315.5965],[11083,6391.2788]]}
       Rotation.X {type=MovieSceneScriptingDoubleChannel, default=0, keys [[8679,33.5161],[11083,34.1139]]}
       Rotation.Y {type=MovieSceneScriptingDoubleChannel, default=0, keys [[8679,-6.9984],[11083,-7.1232]]}
       Rotation.Z {type=MovieSceneScriptingDoubleChannel, default=0, keys [[8679,63.0983],[11083,64.2236]]}
       Scale.X {type=MovieSceneScriptingDoubleChannel, default=1, const [1,2]}
       Scale.Y {type=MovieSceneScriptingDoubleChannel, default=1, const [1,2]}
       Scale.Z {type=MovieSceneScriptingDoubleChannel, default=1, const [1,2]}
      ]
     }]
   }
   {
    type=MovieSceneVisibilityTrack
    sections [{type=MovieSceneVisibilitySection, channels [None {type=MovieSceneScriptingBoolChannel, default=false, keys [[8319,false],[8321,true]]}]}]
   }
   {type=MovieSceneSkeletalAnimationTrack, sections [{type=MovieSceneSkeletalAnimationSection, end_frame=9415, start_frame=7021}]}
  ]
 }
 B_LimeRock_MannequinManager {
  id=#8
  possessed_class=B_LimeRock_MannequinManager_C
  tracks [
   {
    type=MovieSceneSpawnTrack
    sections [{type=MovieSceneSpawnSection, channels [None {type=MovieSceneScriptingBoolChannel, default=true, const [true,1]}]}]
   }
   {
    type=MovieScene3DTransformTrack
    sections [{
      type=MovieScene3DTransformSection
      channels [
       Location.X {type=MovieSceneScriptingDoubleChannel, default=-90836, const [-90836,1]}
       Location.Y {type=MovieSceneScriptingDoubleChannel, default=-95820, const [-95820,1]}
       Location.Z {type=MovieSceneScriptingDoubleChannel, default=1006.0774, const [1006.0775,1]}
       Rotation.X {type=MovieSceneScriptingDoubleChannel, default=0, const [0,1]}
       Rotation.Y {type=MovieSceneScriptingDoubleChannel, default=0, const [0,1]}
       Rotation.Z {type=MovieSceneScriptingDoubleChannel, default=90, const [90,1]}
       Scale.X {type=MovieSceneScriptingDoubleChannel, default=1, const [1,1]}
       Scale.Y {type=MovieSceneScriptingDoubleChannel, default=1, const [1,1]}
       Scale.Z {type=MovieSceneScriptingDoubleChannel, default=1, const [1,1]}
      ]
     }]
   }
  ]
 }
]
display_rate {denominator=1, numerator=60}
tick_resolution {denominator=1, numerator=24000}
tracks [
 {type=MovieSceneSubTrack, display_name=Jones, sections [{type=MovieSceneSubSection, end_frame=9414, start_frame=3604}]}
 {type=MovieSceneSubTrack, display_name=Hope, sections [{type=MovieSceneSubSection, end_frame=9414, start_frame=3600}]}
 {type=MovieSceneSubTrack, display_name=Kang Kodos, sections [{type=MovieSceneSubSection, end_frame=9414, start_frame=3600}]}
 {type=MovieSceneSubTrack, display_name=Tentacles, sections [{type=MovieSceneSubSection, end_frame=9414, start_frame=7688}]}
 {type=MovieSceneSubTrack, display_name=TEMP_Players, sections [{type=MovieSceneSubSection, end_frame=9414, start_frame=3600}]}
 {type=MovieSceneSubTrack, display_name=Subsequences, sections [{type=MovieSceneSubSection, end_frame=9414, start_frame=3600}]}
]
