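# Level sequence for LimeRock_100 that animates a Fishsticks player mannequin at a fixed spot, popping it up for three separate appearances.
# A visibility track shows and hides the mannequin, timed with keyed Z location and Z scale that raise and stretch it on entry, then mostly lower and squash it on exit.
# Overlapping skeletal animation sections play during each appearance, while X/Y position and rotation stay constant.
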LevelSequence /LimeRock/Sequence/LimeRock_100/LimeRock_100_anm_Fishsticks
bindings [FortPlayerMannequin - Fishstix {
  id=#1
  possessed_class=FortPlayerMannequin
  tracks [
   {
    type=MovieScene3DTransformTrack
    sections [
     {
      type=MovieScene3DTransformSection
      channels [
       Location.X {type=MovieSceneScriptingDoubleChannel, default=-90844}
       Location.Y {type=MovieSceneScriptingDoubleChannel, default=-94444}
       Location.Z {
        type=MovieSceneScriptingDoubleChannel
        default=1084
        keys [[4323.9975,1084],[4398,1108],[4838,1108],[4885.9975,1084],[6445.9975,1084],[6520,1108],[7078,1108],[7125.9975,1084],[7840,1084],[7914.0025,1108],[8593.0025,1108],[8618,1084]]
       }
       Rotation.X {type=MovieSceneScriptingDoubleChannel, default=-7.7385}
       Rotation.Y {type=MovieSceneScriptingDoubleChannel, default=2.94}
       Rotation.Z {type=MovieSceneScriptingDoubleChannel, default=-173.0976}
      ]
     }
     {
      type=MovieScene3DTransformSection
      channels [
       Scale.X {type=MovieSceneScriptingDoubleChannel, default=1}
       Scale.Y {type=MovieSceneScriptingDoubleChannel, default=1}
       Scale.Z {
        type=MovieSceneScriptingDoubleChannel
        default=1
        keys [[4323.9975,0.7],[4398,1],[4838,1],[4886,0.7],[6442.9975,0.7],[6517,1],[7078,1],[7126,0.7],[7840,0.7],[7914.0025,1],[8593.0025,1],[8641.0025,0.7]]
       }
      ]
     }
    ]
   }
   {
    type=MovieSceneSkeletalAnimationTrack
    sections [
     {type=MovieSceneSkeletalAnimationSection, end_frame=4487, start_frame=4320}
     {type=MovieSceneSkeletalAnimationSection, end_frame=4908, start_frame=4448}
     {type=MovieSceneSkeletalAnimationSection, end_frame=6609, start_frame=6442}
     {type=MovieSceneSkeletalAnimationSection, end_frame=7095, start_frame=6570}
     {type=MovieSceneSkeletalAnimationSection, end_frame=7126, start_frame=7066}
     {type=MovieSceneSkeletalAnimationSection, end_frame=7943, start_frame=7779}
     {type=MovieSceneSkeletalAnimationSection, end_frame=8399, start_frame=7904}
     {type=MovieSceneSkeletalAnimationSection, end_frame=8618, start_frame=8298}
    ]
   }
   {
    type=MovieSceneVisibilityTrack
    sections [{
      type=MovieSceneVisibilitySection
      channels [None {
        type=MovieSceneScriptingBoolChannel
        default=false
        keys [[3600,false],[4323.9975,true],[4885.9975,false],[6445.9975,true],[7125.9975,false],[7840,true],[8618,false]]
       }]
     }]
   }
  ]
 }]
display_rate {denominator=1, numerator=60}
tick_resolution {denominator=1, numerator=24000}
tracks []
