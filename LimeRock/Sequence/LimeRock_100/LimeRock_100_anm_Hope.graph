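# Level Sequence animating the character Hope and her Mark III glider in the LimeRock_100 cinematic, covering a gliding descent and landing.
# Hope's densely keyed transform carries her down to a settled position, with a second zero-scale transform section layering a shifting offset over it.
# Skeletal animation drives both actors, and visibility keys reveal them together, then hide the glider after a short transform move at touchdown.
# The Backpack component is bound under Hope without tracks.
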LevelSequence /LimeRock/Sequence/LimeRock_100/LimeRock_100_anm_Hope
bindings [
 Mark_III_Glider_Hope {
  id=#1
  possessed_class=FortSkeletalMeshActor
  tracks [
   {type=MovieSceneSkeletalAnimationTrack, sections [{type=MovieSceneSkeletalAnimationSection, end_frame=2204, start_frame=2040}]}
   {
    type=MovieSceneVisibilityTrack
    sections [{
      type=MovieSceneVisibilitySection
      channels [None {type=MovieSceneScriptingBoolChannel, default=false, keys [[1800,false],[2044,true],[2137,true],[2145,false]]}]
     }]
   }
   {
    type=MovieScene3DTransformTrack
    sections [{
      type=MovieScene3DTransformSection
      channels [
       Location.X {type=MovieSceneScriptingDoubleChannel, default=0, keys [[2141,0],[2144,-4.3653]]}
       Location.Y {type=MovieSceneScriptingDoubleChannel, default=0, keys [[2141,0],[2142,14.4602],[2143,21.3237],[2144,6.5065]]}
       Location.Z {type=MovieSceneScriptingDoubleChannel, default=0, keys [[2141,0],[2144,-2.2754]]}
       Rotation.X {type=MovieSceneScriptingDoubleChannel, default=0, keys [[2141,0],[2144,-17.9679]]}
       Rotation.Y {type=MovieSceneScriptingDoubleChannel, default=0, keys [[2141,0],[2144,42.7163]]}
       Rotation.Z {type=MovieSceneScriptingDoubleChannel, default=0, keys [[2141,0],[2144,0]]}
       Scale.X {type=MovieSceneScriptingDoubleChannel, default=1, const [1,2]}
       Scale.Y {type=MovieSceneScriptingDoubleChannel, default=1, const [1,2]}
       Scale.Z {type=MovieSceneScriptingDoubleChannel, default=1, const [1,2]}
      ]
     }]
   }
  ]
 }
 Hope {
  id=#2
  possessed_class=FortPlayerMannequin
  tracks [
   {
    type=MovieScene3DTransformTrack
    sections [
     {
      type=MovieScene3DTransformSection
      channels [
       Location.X {
        type=MovieSceneScriptingDoubleChannel
        default=4063.0489
        thinned_from=118
        keys [[2040.5,1136.8695],[2041.5,1162.1668],[2042.5,1206.1775],[2043.5,1231.2672],[2044,1349.2272],[2045,1374.9101],[2045.5,1402.5206],[2048.5,1428.9745],[2049,1458.0975],[2051,1515.8838],[2051.5,1543.6871],[2052.5,1570.0669],[2053.5,1647.9902],[2054,1675.3931],[2056,1732.7049],[2056.5,1760.1672],[2058.5,1813.3864],[2059,1839.2175],[2060,1864.8706],[2060.5,1890.2272],[2061.5,1914.1751],[2062.5,1942.6605],[2063,1970.5972],[2065,2022.3929],[2065.5,2049.7958],[2066.5,2075.4193],[2067,2101.5322],[2069,2151.1184],[2069.5,2180.2414],[2070.5,2206.2208],[2072,2232.1854],[2072.5,2275.1136],[2074.5,2330.6609],[2075,2357.2927],[2078,2435.8092],[2080,2508.0384],[2081,2538.5108],[2081.5,2568.2418],[2083.5,2622.4545],[2084,2647.2179],[2085,2673.5087],[2085.5,2698.3018],[2087.5,2748.9705],[2088,2774.9647],[2089,2800.2769],[2089.5,2826.8494],[2091.5,2879.9351],[2092,2904.4613],[2093,2931.7456],[2093.5,2957.117],[2095.5,3008.2454],[2096,3034.2841],[2097,3059.359],[2097.5,3084.1521],[2099.5,3137.6827],[2100,3163.4248],[2101,3188.989],[2101.5,3214.0935],[2103.5,3266.334],[2104,3292.8176],[2105,3318.0852],[2105.5,3343.1156],[2107.5,3395.3412],[2108,3421.8248],[2110,3475.9782],[2110.5,3503.0994],[2112.5,3530.5616],[2113.5,3594.6945],[2114.5,3622.3347],[2115,3649.1889],[2117,3704.1134],[2117.5,3730.3745],[2119.5,3756.6664],[2120.5,3805.8824],[2125.5,3910.9492],[2135.5,4018.4898],[2144.5,4063.0489]]
       }
       Location.Y {
        type=MovieSceneScriptingDoubleChannel
        default=-152.6655
        thinned_from=118
        keys [[2040.5,4409.5486],[2041.5,4370.1076],[2042.5,4301.4903],[2043.5,4262.3729],[2044,4078.461],[2045,4038.4189],[2045.5,3995.3712],[2048.5,3954.1269],[2049,3908.7211],[2051,3818.6263],[2051.5,3775.2781],[2052.5,3734.1494],[2053.5,3612.6589],[2054,3569.9349],[2056,3480.5799],[2056.5,3437.7635],[2058.5,3354.7893],[2059,3314.516],[2060,3274.52],[2060.5,3234.9865],[2061.5,3197.6493],[2062.5,3153.2376],[2063,3109.6814],[2065,3028.9266],[2065.5,2986.2027],[2066.5,2946.253],[2067,2905.5404],[2069,2828.2304],[2069.5,2782.8246],[2070.5,2742.3201],[2072,2701.8387],[2072.5,2634.9091],[2074.5,2548.3052],[2075,2506.7835],[2078,2384.3682],[2080,2271.7555],[2081,2224.2459],[2081.5,2177.8922],[2083.5,2093.3691],[2084,2054.7603],[2085,2013.7703],[2085.5,1975.1152],[2087.5,1896.1175],[2088,1855.5899],[2089,1816.1257],[2089.5,1774.6964],[2091.5,1691.9303],[2092,1653.6914],[2093,1611.1524],[2093.5,1571.5958],[2095.5,1491.8814],[2096,1451.2844],[2097,1412.1901],[2097.5,1373.5351],[2099.5,1290.0754],[2100,1249.9408],[2101,1210.0836],[2101.5,1170.9431],[2102.5,1132.1725],[2103.5,1089.4948],[2104,1048.2042],[2105,1008.8094],[2105.5,969.7844],[2106.5,930.1353],[2107.5,888.3592],[2108,847.0687],[2109,806.3792],[2110,762.638],[2110.5,720.3533],[2112.5,677.5369],[2113.5,577.5471],[2114.5,534.4532],[2115,492.5847],[2116,451.8489],[2117,406.9518],[2117.5,366.008],[2119.5,325.0163],[2120.5,248.2835],[2121.5,207.5919],[2122.5,174.3722],[2124.5,112.0297],[2125.5,84.4735],[2126.5,61.0767],[2127.5,37.1782],[2128.5,16.2745],[2129.5,-1.4739],[2130.5,-18.6698],[2131.5,-34.367],[2132.5,-47.9075],[2133.5,-60.4791],[2134.5,-72.3853],[2135.5,-83.1934],[2136.5,-92.435],[2137.5,-101.2054],[2138.5,-109.1693],[2139.5,-116.5226],[2140,-122.8741],[2141,-128.2024],[2141.5,-133.0654],[2142.5,-138.1866],[2143.5,-145.2723],[2144,-152.6655],[2144.5,-152.6655]]
       }
       Location.Z {
        type=MovieSceneScriptingDoubleChannel
        default=-350.7504
        thinned_from=118
        keys [[2040.5,1427.3945],[2041.5,1414.3787],[2043,1378.8257],[2044,1318.1339],[2045,1304.9198],[2045.5,1290.7139],[2048,1277.103],[2051,1232.3871],[2051.5,1218.082],[2052.5,1204.5093],[2053.5,1164.4169],[2054,1150.3177],[2056,1120.8301],[2056.5,1106.7005],[2058.5,1079.3186],[2059,1066.0282],[2060,1052.8293],[2060.5,1039.7831],[2061.5,1027.4616],[2062.5,1012.8055],[2063,998.4318],[2065,971.7823],[2065.5,957.6832],[2066.5,944.4996],[2067,931.0642],[2069,905.5515],[2069.5,890.5674],[2070.5,877.2007],[2072,863.8416],[2072.5,841.7545],[2074.5,813.1748],[2075,799.4724],[2078,759.0748],[2080,721.912],[2081,706.2336],[2081.5,690.9366],[2083.5,663.0436],[2084,650.3025],[2085,636.7756],[2085.5,624.0192],[2087.5,597.9496],[2088,584.5752],[2089,571.5519],[2089.5,557.88],[2091.5,530.5668],[2092,517.9477],[2093,503.9097],[2093.5,490.8558],[2094.5,478.2062],[2095.5,464.5496],[2096,451.1524],[2097,438.2511],[2097.5,425.4947],[2099.5,397.9526],[2100,384.708],[2101,371.5549],[2101.5,358.6384],[2102.5,345.8439],[2103.5,331.76],[2104,318.1339],[2105,305.1334],[2105.5,292.255],[2106.5,279.1706],[2107,265.3843],[2109,238.3304],[2109.5,223.8956],[2110.5,209.9415],[2112.5,195.8118],[2113.5,162.8147],[2114.5,148.5935],[2115,134.7767],[2116,121.3337],[2117,106.5174],[2117.5,93.0057],[2119.5,78.9676],[2120.5,49.4953],[2121.5,31.7112],[2122.5,15.8115],[2123.5,-0.462],[2124.5,-18.2919],[2125.5,-35.6182],[2126.5,-51.7163],[2127.5,-69.7369],[2128.5,-87.0709],[2129.5,-103.1765],[2131.5,-137.3028],[2133.5,-169.7125],[2135.5,-203.3886],[2136.5,-219.0518],[2139.5,-267.4526],[2140,-282.6428],[2141,-296.4215],[2141.5,-309.9179],[2142.5,-323.0709],[2143.5,-337.8109],[2144,-350.7504],[2144.5,-350.7504]]
       }
       Rotation.X {type=MovieSceneScriptingDoubleChannel, default=0, const [0,3]}
       Rotation.Y {type=MovieSceneScriptingDoubleChannel, default=0, const [0,3]}
       Rotation.Z {type=MovieSceneScriptingDoubleChannel, default=201, keys [[2040.29,270],[2040.5,180]]}
       Scale.X {type=MovieSceneScriptingDoubleChannel, default=1, const [1,1]}
       Scale.Y {type=MovieSceneScriptingDoubleChannel, default=1, const [1,1]}
       Scale.Z {type=MovieSceneScriptingDoubleChannel, default=1, const [1,1]}
      ]
     }
     {
      type=MovieScene3DTransformSection
      end_frame=4726
      start_frame=1800
      channels [
       Location.X {type=MovieSceneScriptingDoubleChannel, default=586, keys [[2044,586],[2144.5,-128.1175]]}
       Location.Y {type=MovieSceneScriptingDoubleChannel, default=-493, keys [[2044,-493],[2144.5,-47.9919]]}
       Location.Z {type=MovieSceneScriptingDoubleChannel, default=5, keys [[2044,5],[2144.5,0]]}
       Rotation.X {type=MovieSceneScriptingDoubleChannel, default=0, const [0,2]}
       Rotation.Y {type=MovieSceneScriptingDoubleChannel, default=0, const [0,2]}
       Rotation.Z {type=MovieSceneScriptingDoubleChannel, default=-8, keys [[2044,-0.7291],[2144.5,-8]]}
       Scale.X {type=MovieSceneScriptingDoubleChannel, default=0, const [0,1]}
       Scale.Y {type=MovieSceneScriptingDoubleChannel, default=0, const [0,1]}
       Scale.Z {type=MovieSceneScriptingDoubleChannel, default=0, const [0,1]}
      ]
     }
    ]
   }
   {
    type=MovieSceneVisibilityTrack
    sections [{type=MovieSceneVisibilitySection, channels [None {type=MovieSceneScriptingBoolChannel, default=false, keys [[1800,false],[2044,true]]}]}]
   }
   {type=MovieSceneSkeletalAnimationTrack, sections [{type=MovieSceneSkeletalAnimationSection, end_frame=4707, start_frame=1800}]}
  ]
 }
 Backpack {id=#3, parent_id=#2, possessed_class=SkeletalMeshComponentBudgeted, tracks []}
]
display_rate {denominator=1, numerator=30}
tick_resolution {denominator=1, numerator=24000}
tracks []
