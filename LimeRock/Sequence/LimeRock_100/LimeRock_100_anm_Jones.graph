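# Level Sequence for the LimeRock_100 cinematic that animates Jones descending with a Mark III glider.
# Jones combines a skeletal animation section with a densely keyed descent path and a second, sparsely keyed transform section that adjusts position.
# The glider has its own animation and attach track, appears with Jones mid-descent, then tumbles, shrinks and hides as the descent ends.
# A Yogurt device gadget holds a fixed transform and becomes visible later.
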LevelSequence /LimeRock/Sequence/LimeRock_100/LimeRock_100_anm_Jones
bindings [
 Yogurt_Device_Gadget {
  id=#1
  possessed_class=FortSkeletalMeshActor
  tracks [
   {
    type=MovieScene3DTransformTrack
    sections [{
      type=MovieScene3DTransformSection
      channels [
       Location.X {type=MovieSceneScriptingDoubleChannel, default=-1}
       Location.Y {type=MovieSceneScriptingDoubleChannel, default=-4}
       Location.Z {type=MovieSceneScriptingDoubleChannel, default=-6}
       Rotation.X {type=MovieSceneScriptingDoubleChannel, default=-14}
       Rotation.Y {type=MovieSceneScriptingDoubleChannel, default=6}
       Rotation.Z {type=MovieSceneScriptingDoubleChannel, default=69}
       Scale.X {type=MovieSceneScriptingDoubleChannel, default=1}
       Scale.Y {type=MovieSceneScriptingDoubleChannel, default=1}
       Scale.Z {type=MovieSceneScriptingDoubleChannel, default=1}
      ]
     }]
   }
   {type=MovieSceneSkeletalAnimationTrack, sections []}
   {
    type=MovieSceneVisibilityTrack
    sections [{type=MovieSceneVisibilitySection, channels [None {type=MovieSceneScriptingBoolChannel, default=false, keys [[2572,false],[2632,true]]}]}]
   }
  ]
 }
 Jones {
  id=#2
  possessed_class=FortPlayerMannequin
  tracks [
   {
    type=MovieScene3DTransformTrack
    sections [
     {
      type=MovieScene3DTransformSection
      channels [
       Location.X {
        type=MovieSceneScriptingDoubleChannel
        default=4057
        thinned_from=432
        keys [[1834.5,-2435.8552],[1868,-2427.8583],[1875.5,-2395.8266],[1880,-2364.0455],[1884,-2326.9099],[1884.5,-2319.2782],[1886.5,-2303.6548],[1887,-2294.5986],[1889,-2276.5588],[1893,-2216.0531],[1893.5,-2201.3711],[1895.5,-2169.5113],[1896,-2149.3718],[1898,-2111.4435],[1898.5,-2091.843],[1900.5,-2048.3655],[1901,-2025.3594],[1902,-2002.335],[1902.5,-1980.6744],[1904.5,-1931.5136],[1905,-1907.731],[1906,-1880.2898],[1906.5,-1854.3961],[1907.5,-1826.3333],[1908,-1803.7243],[1909,-1782.0121],[1909.5,-1761.1587],[1910.5,-1741.8494],[1911,-1724.0391],[1912,-1707.4259],[1912.5,-1691.8352],[1913.5,-1677.3623],[1914,-1663.865],[1915,-1650.3885],[1915.5,-1638.3058],[1916.5,-1627.1048],[1917,-1616.0233],[1918,-1606.092],[1918.5,-1596.4469],[1919.5,-1587.9293],[1920,-1580.1474],[1921,-1572.4836],[1921.5,-1565.65],[1922.5,-1559.1518],[1923,-1553.1329],[1924,-1547.5206],[1924.5,-1542.1663],[1927.5,-1523.8301],[1934.5,-1496.807],[1937,-1489.2011],[1938.5,-1480.5988],[1939,-1474.5607],[1941,-1461.2224],[1941.5,-1452.6461],[1942.5,-1442.9387],[1943,-1433.5276],[1944,-1423.481],[1944.5,-1411.5992],[1945.5,-1400.2682],[1946,-1387.9323],[1948,-1363.2674],[1948.5,-1352.2428],[1949.5,-1342.1992],[1950,-1332.6326],[1951,-1323.7596],[1951.5,-1315.9075],[1952.5,-1308.1741],[1953,-1301.0154],[1954,-1294.3441],[1954.5,-1288.2154],[1955.5,-1282.2734],[1956,-1276.974],[1957,-1271.7021],[1957.5,-1267.1051],[1960.5,-1251.0809],[1963,-1239.7145],[1963.5,-1234.4626],[1964.5,-1228.4463],[1965,-1221.5013],[1966,-1213.4371],[1966.5,-1205.1715],[1967.5,-1196.311],[1968,-1186.5427],[1969,-1176.0563],[1969.5,-1164.9575],[1970.5,-1153.2367],[1971,-1139.8871],[1972,-1126.6471],[1972.5,-1113.33],[1974.5,-1081.7937],[1975,-1065.1706],[1976,-1048.2268],[1976.5,-1030.3762],[1977.5,-1012.0117],[1978,-992.7938],[1979,-972.6849],[1979.5,-951.5628],[1980.5,-929.8539],[1981,-907.234],[1982,-884.7158],[1982.5,-861.8435],[1983.5,-838.5121],[1984,-813.6162],[1985,-788.6547],[1985.5,-762.3339],[1986.5,-734.6354],[1987,-707.9719],[1988,-683.0772],[1988.5,-657.2565],[1989.5,-630.2348],[1990,-601.7267],[1991,-575.6469],[1991.5,-549.4048],[1992.5,-522.8873],[1993,-496.9724],[1994,-470.5896],[1994.5,-445.1422],[1996.5,-394.9157],[1997,-372.1828],[1998,-351.6555],[1998.5,-333.0321],[1999.5,-315.1093],[2000,-298.439],[2000.5,-282.8572],[2001.5,-268.3854],[2002,-254.4444],[2003,-241.8323],[2003.5,-230.1563],[2004.5,-218.7787],[2005.5,-207.9206],[2006,-197.7549],[2007,-188.4887],[2007.5,-180.0243],[2008,-172.0262],[2009,-164.7563],[2009.5,-157.8722],[2010.5,-151.2418],[2011.5,-144.998],[2012,-139.2355],[2012.5,-133.8711],[2014.5,-124.0571],[2015,-119.7121],[2015.5,-115.7537],[2017.5,-108.2611],[2018,-104.4369],[2019,-99.7479],[2019.5,-94.7022],[2020.5,-88.9157],[2021,-82.3607],[2022,-75.3727],[2022.5,-67.5602],[2023.5,-58.4935],[2024,-48.53],[2025,-38.4387],[2025.5,-28.0826],[2026.5,-16.3051],[2027.5,-3.5919],[2028,9.0354],[2029,21.062],[2029.5,31.6384],[2030.5,40.8947],[2031,49.7176],[2031.5,57.9153],[2032.5,65.607],[2033,73.0213],[2034,80.7412],[2034.5,89.0933],[2035.5,98.143],[2036.5,107.6182],[2037,117.8036],[2038,128.7958],[2038.5,140.0733],[2039.5,152.0515],[2040,164.9993],[2041,178.2952],[2041.5,193.0254],[2042.5,208.3914],[2043,223.9436],[2044,239.6154],[2044.5,255.6314],[2045.5,272.2148],[2046,290.1546],[2047,307.8347],[2048,327.188],[2048.5,347.4703],[2049.5,367.5439],[2050,388.5416],[2051,410.2632],[2051.5,431.2759],[2052.5,453.5367],[2053.5,478.9647],[2054,502.3156],[2055,524.8244],[2055.5,547.6715],[2056.5,570.4982],[2057,592.447],[2058,616.3694],[2058.5,639.4709],[2059.5,662.9146],[2060,686.0516],[2061,709.2426],[2061.5,732.1846],[2062.5,754.2599],[2063,776.1168],[2064,798.4444],[2064.5,821.0359],[2066.5,870.2127],[2067,896.782],[2068,919.3459],[2068.5,941.0788],[2069.5,963.5311],[2070,985.6063],[2071,1009.0149],[2071.5,1031.9466],[2072.5,1054.2179],[2073,1076.5118],[2074,1100.2961],[2074.5,1123.3734],[2076.5,1169.4604],[2077,1194.2488],[2078,1216.4216],[2078.5,1238.6484],[2079.5,1260.3105],[2080,1282.9962],[2081,1305.3095],[2081.5,1327.7206],[2082.5,1349.3502],[2083,1373.367],[2084,1395.9499],[2084.5,1417.6506],[2086.5,1461.8429],[2087,1484.8533],[2088,1507.4677],[2088.5,1528.9742],[2089.5,1551.4282],[2090,1573.2682],[2092,1620.4772],[2092.5,1645.4643],[2093.5,1667.6445],[2094,1690.4327],[2095,1712.794],[2095.5,1735.7734],[2096.5,1758.6178],[2097,1781.8777],[2099,1825.9307],[2099.5,1849.602],[2100.5,1871.3029],[2101,1893.6467],[2102,1915.5347],[2102.5,1937.9397],[2103.5,1960.8374],[2104,1983.6488],[2105,2005.9899],[2105.5,2025.4467],[2106.5,2043.7527],[2107,2062.4359],[2108,2080.0752],[2108.5,2095.2753],[2109.5,2110.6917],[2110,2126.7923],[2111,2143.831],[2111.5,2160.8072],[2112.5,2179.2033],[2113,2198.582],[2114,2218.8021],[2114.5,2240.4846],[2115.5,2261.7928],[2116,2283.3572],[2117.5,2306.3701],[2118,2344.1647],[2119,2369.7912],[2119.5,2394.181],[2120.5,2418.7534],[2121,2444.0688],[2123,2494.1523],[2125,2564.8981],[2126,2594.8277],[2126.5,2622.3839],[2127.5,2648.1326],[2128,2672.9033],[2129,2695.8581],[2129.5,2717.6747],[2130.5,2738.0894],[2131,2757.0246],[2132,2774.4213],[2132.5,2790.4532],[2134.5,2819.3024],[2135,2832.8452],[2136,2846.2403],[2136.5,2857.9585],[2137.5,2869.3197],[2138,2879.5179],[2143,2933.3831],[2143.5,2943.8127],[2144.5,2955.593],[2145,2967.7181],[2146,2980.5083],[2146.5,2993.9619],[2147.5,3008.0991],[2148,3023.3442],[2149,3039.3148],[2150,3072.3516],[2152,3109.0068],[2153,3149.3161],[2155,3193.1158],[2155.5,3215.8861],[2156.5,3240.8234],[2157,3262.5258],[2158,3282.1261],[2158.5,3300.9403],[2159.5,3319.4156],[2160,3335.2774],[2162,3365.9989],[2162.5,3380.7416],[2163.5,3396.3658],[2164,3411.945],[2166,3448.5773],[2166.5,3467.7339],[2168.5,3509.7515],[2169,3531.1956],[2170,3554.67],[2170.5,3578.2918],[2171.5,3602.3677],[2172,3626.2186],[2173,3652.1704],[2173.5,3680.2681],[2174.5,3708.5092],[2175,3738.0524],[2176,3767.8877],[2176.5,3796.4334],[2178.5,3849.2676],[2181,3874.478],[2182,3945.5714],[2182.5,3965.7169],[2183.5,3983.8171],[2184,3999.4389],[2187,4038.7335],[2187.5,4053.0043],[2202,4165.0381]]
       }
       Location.Y {
        type=MovieSceneScriptingDoubleChannel
        default=-849
        thinned_from=432
        keys [[1834.5,8926.0707],[1886.5,8752.3949],[1893,8636.78],[1895.5,8575.0684],[1896,8548.2408],[1898,8497.5484],[1898.5,8471.281],[1900.5,8412.8806],[1901,8381.9153],[1902,8350.8891],[1902.5,8321.671],[1904.5,8255.2671],[1905,8223.1036],[1906,8185.9649],[1906.5,8150.8962],[1908,8082.2483],[1909,8052.8337],[1909.5,8024.5824],[1912.5,7930.6661],[1918.5,7801.4384],[1942.5,7594.5804],[1951.5,7426.6132],[1967.5,7268.7185],[1974.5,7117.211],[1975,7095.0286],[1976,7072.3292],[1976.5,7048.3491],[1977.5,7023.5783],[1978,6997.5702],[1979,6970.2626],[1979.5,6941.4876],[1980.5,6911.8263],[1981,6880.8368],[1982,6849.9112],[1982.5,6818.4298],[1983.5,6786.2516],[1984,6751.8503],[1985,6717.2968],[1985.5,6680.8014],[1986.5,6642.3349],[1987,6605.25],[1988,6570.5807],[1988.5,6534.5816],[1989.5,6496.8752],[1990,6457.0581],[1991,6420.6031],[1991.5,6383.8928],[1992.5,6346.7715],[1993,6310.4687],[1994,6273.4876],[1994.5,6237.796],[1996.5,6167.3302],[1997,6135.4369],[1998,6106.6378],[1998.5,6080.51],[1999.5,6055.365],[2000.5,6010.1165],[2001.5,5989.8132],[2002,5970.2544],[2008,5854.625],[2022,5717.7671],[2027.5,5612.2687],[2028,5593.6264],[2031.5,5521.4631],[2036.5,5447.7832],[2039.5,5379.3977],[2040,5358.9566],[2041,5337.7857],[2041.5,5314.1418],[2042.5,5289.2881],[2043,5263.9566],[2044,5238.2673],[2044.5,5211.8645],[2045.5,5184.3858],[2046,5154.5162],[2048,5092.5154],[2048.5,5058.3925],[2049.5,5024.5247],[2050,4989.033],[2051,4952.2353],[2051.5,4916.5725],[2053.5,4835.4245],[2054,4795.596],[2055,4757.1523],[2055.5,4718.0792],[2056.5,4678.9941],[2057,4641.3683],[2058,4600.3151],[2058.5,4560.6272],[2059.5,4520.3123],[2060,4480.4862],[2061,4440.5328],[2061.5,4400.9746],[2062.5,4362.8823],[2063,4325.1395],[2064,4286.5585],[2064.5,4247.4956],[2066.5,4162.3834],[2067,4116.3565],[2068,4077.2459],[2068.5,4039.5584],[2069.5,4000.6054],[2070,3962.291],[2071,3921.6378],[2071.5,3881.7561],[2072.5,3842.9831],[2073,3804.0963],[2074,3762.5497],[2074.5,3722.1491],[2076.5,3641.215],[2077,3597.5222],[2078,3558.3149],[2078.5,3518.917],[2079.5,3480.4304],[2080,3440.0301],[2081,3400.2081],[2081.5,3360.1262],[2082.5,3321.3695],[2083,3278.2521],[2084,3237.637],[2084.5,3198.5427],[2086.5,3118.7512],[2087,3077.117],[2088,3036.1426],[2088.5,2997.1291],[2089.5,2956.3486],[2090,2916.6408],[2092,2830.6718],[2092.5,2785.1088],[2093.5,2744.7727],[2094,2703.3604],[2095,2662.8721],[2095.5,2621.3765],[2096.5,2580.2458],[2097,2538.4977],[2099,2459.8531],[2099.5,2417.8629],[2100.5,2379.5232],[2101,2340.1905],[2102,2301.7992],[2102.5,2262.6278],[2103.5,2222.7257],[2104,2183.0908],[2105,2144.333],[2105.5,2110.579],[2106.5,2078.8214],[2107,2046.4094],[2108,2015.8084],[2108.5,1989.4388],[2109.5,1962.7744],[2110,1935.153],[2111,1906.1116],[2111.5,1877.5],[2112.5,1846.7375],[2113,1814.6697],[2114,1781.4949],[2114.5,1746.2428],[2115.5,1711.8792],[2116,1677.3696],[2117.5,1640.8052],[2118,1581.2703],[2119,1541.226],[2119.5,1503.3253],[2120.5,1465.346],[2121,1426.4085],[2123,1349.9143],[2124,1299.7218],[2125,1242.9269],[2126,1197.9837],[2126.5,1156.7398],[2127.5,1118.3178],[2128,1081.4056],[2129,1047.1992],[2129.5,1014.6892],[2130.5,984.268],[2131,956.0516],[2132,930.1277],[2132.5,906.2377],[2134.5,863.2479],[2135,843.067],[2136,823.1062],[2136.5,805.6442],[2137.5,788.7142],[2138,773.5173],[2139,759.835],[2139.5,747.0056],[2140.5,734.3638],[2141,721.7359],[2142,708.676],[2143,694.181],[2143.5,679.3318],[2144.5,662.7093],[2145,645.7341],[2146,627.9476],[2146.5,609.3484],[2147.5,589.9057],[2148,569.0377],[2149,547.2696],[2150,502.4864],[2151,478.5345],[2152,453.1234],[2153,399.3936],[2154,370.9604],[2155,341.5872],[2155.5,311.8028],[2156.5,279.2524],[2157,250.9245],[2158,225.3405],[2158.5,200.7826],[2159.5,176.667],[2160,155.9628],[2161,135.7456],[2162,115.8624],[2162.5,96.7188],[2163.5,76.6796],[2164,56.8777],[2165,35.048],[2166,10.9441],[2166.5,-12.7567],[2168.5,-64.118],[2169,-90.0225],[2170,-118.1288],[2170.5,-146.1997],[2171.5,-174.5607],[2172,-202.457],[2173,-232.5721],[2173.5,-264.9391],[2174.5,-297.197],[2175,-330.7033],[2176,-364.2991],[2176.5,-396.3337],[2177.5,-427.6477],[2178.5,-455.6254],[2181,-483.9171],[2182,-563.6998],[2182.5,-586.3075],[2183.5,-606.62],[2184,-624.1511],[2186,-655.1547],[2187,-668.2484],[2187.5,-684.2634],[2189.5,-712.5701],[2192.5,-745.9162],[2194.5,-764.8177],[2196.5,-780.2521],[2200.5,-803.7353],[2202,-809.9902]]
       }
       Location.Z {
        type=MovieSceneScriptingDoubleChannel
        default=-403.6068
        thinned_from=420
        keys [[1834.3112,4586.8075],[1873.3112,4586.8075],[1877.3112,4659.5502],[1877.8112,4678.1258],[1880.8112,4730.1515],[1881.3112,4745.9726],[1885.3112,4787.8448],[1888.8112,4785.6405],[1891.3112,4766.1319],[1891.8112,4750.4363],[1896.3112,4680.8701],[1901.3112,4597.8394],[1901.8112,4584.3964],[1920.3112,4280.922],[1921.3112,4255.8976],[1923.8112,4218.9484],[1924.3112,4206.4438],[1925.3112,4194.0689],[1928.8112,4131.1035],[1930.8112,4106.3462],[1931.3112,4093.1397],[1932.3112,4080.4291],[1934.3112,4041.0843],[1935.3112,4028.4958],[1935.8112,4015.7624],[1936.8112,4003.0747],[1937.3112,3990.5854],[1939.3112,3963.562],[1939.8112,3950.7599],[1940.8112,3938.0951],[1941.3112,3924.1867],[1943.3112,3897.3618],[1945.3112,3858.9783],[1946.3112,3846.4737],[1946.8112,3833.6945],[1947.8112,3820.4727],[1948.3112,3807.5943],[1949.3112,3794.9982],[1949.8112,3782.1808],[1950.8112,3769.4168],[1951.3112,3757.3624],[1952.3112,3744.6442],[1952.8112,3732.071],[1953.8112,3719.5053],[1954.3112,3707.1839],[1955.3112,3694.3741],[1955.8112,3682.1823],[1956.8112,3669.1742],[1957.3112,3657.0664],[1958.3112,3644.7983],[1958.8112,3632.6371],[1959.8112,3619.7357],[1960.3112,3607.2769],[1961.3112,3594.9784],[1961.8112,3582.8171],[1962.8112,3569.9768],[1963.3112,3557.1747],[1965.3112,3531.3187],[1965.8112,3517.746],[1966.8112,3505.1422],[1967.3112,3492.7445],[1968.3112,3480.1178],[1968.8112,3467.5293],[1969.8112,3455.0782],[1970.3112,3442.7338],[1971.3112,3429.4968],[1971.8112,3417.099],[1972.8112,3405.243],[1973.3112,3391.8457],[1974.3112,3379.1733],[1974.8112,3366.3636],[1975.8112,3353.8437],[1976.3112,3341.1789],[1977.3112,3328.6362],[1977.8112,3315.9867],[1978.8112,3303.2151],[1979.3112,3290.2603],[1980.3112,3277.3896],[1980.8112,3264.412],[1981.8112,3251.8921],[1982.3112,3239.5478],[1983.3112,3227.3102],[1983.8112,3214.6149],[1985.8112,3189.5218],[1986.8112,3163.4827],[1988.8112,3139.3204],[1989.3112,3126.4344],[1990.3112,3112.8311],[1990.8112,3100.38],[1991.8112,3087.8449],[1992.3112,3075.1725],[1993.3112,3062.7823],[1993.8112,3050.1633],[1994.8112,3037.9868],[1995.3112,3025.4364],[1996.3112,3012.2605],[1996.8112,2999.3668],[1997.8112,2986.8698],[1998.3112,2974.7773],[1999.3112,2962.3185],[1999.8112,2949.9512],[2000.8112,2937.5687],[2001.3112,2925.3006],[2002.3112,2912.6358],[2002.8112,2900.4059],[2003.8112,2888.2904],[2004.3112,2875.6943],[2005.3112,2862.793],[2005.8112,2849.8764],[2006.8112,2837.2345],[2007.3112,2824.9054],[2008.3112,2812.4237],[2008.8112,2800.3235],[2009.8112,2788.0631],[2010.3112,2775.4593],[2011.3112,2762.7259],[2011.8112,2750.1755],[2012.8112,2737.648],[2013.3112,2725.2503],[2014.3112,2712.3032],[2014.8112,2699.8902],[2015.8112,2687.7823],[2016.3112,2675.4151],[2017.3112,2662.468],[2017.8112,2649.849],[2018.8112,2636.7646],[2019.3112,2624.5423],[2020.3112,2612.1216],[2020.8112,2599.5102],[2021.8112,2587.3185],[2022.3112,2574.8444],[2023.3112,2561.5464],[2023.8112,2548.0728],[2024.8112,2535.3928],[2025.3112,2523.201],[2027.3112,2496.933],[2027.8112,2483.7113],[2028.8112,2470.2149],[2029.3112,2457.4433],[2030.3112,2445.5185],[2030.8112,2433.3649],[2031.8112,2421.3333],[2032.3112,2409.2636],[2033.3112,2396.8582],[2033.8112,2384.079],[2034.8112,2371.4066],[2035.3112,2358.7265],[2036.3112,2346.283],[2036.8112,2333.7479],[2037.8112,2320.9229],[2038.3112,2308.4565],[2039.3112,2295.8527],[2039.8112,2282.8675],[2041.8112,2256.6452],[2042.3112,2243.1794],[2043.3112,2230.1026],[2043.8112,2217.4225],[2044.8112,2204.9256],[2045.3112,2192.4286],[2046.3112,2179.3671],[2046.8112,2166.9236],[2047.8112,2153.74],[2048.3112,2140.3733],[2049.3112,2127.5559],[2049.8112,2114.5554],[2051.8112,2089.2258],[2052.3112,2076.5686],[2053.3112,2062.4085],[2053.8112,2049.5835],[2054.8112,2037.2086],[2055.3112,2024.6353],[2056.3112,2012.0621],[2056.8112,1999.9619],[2057.8112,1986.763],[2058.3112,1974.0067],[2059.3112,1961.052],[2059.8112,1948.2575],[2060.8112,1935.4248],[2061.3112,1922.7219],[2062.3112,1910.492],[2062.8112,1898.3765],[2064.8112,1873.4589],[2065.3112,1859.52],[2066.3112,1846.1533],[2066.8112,1831.3904],[2067.8112,1818.8477],[2068.3112,1806.7627],[2069.3112,1794.2734],[2069.8112,1781.9901],[2070.8112,1768.9591],[2071.3112,1756.1798],[2073.3112,1731.308],[2073.8112,1718.01],[2074.8112,1705.0858],[2075.3112,1692.5507],[2076.3112,1679.2145],[2076.8112,1665.2603],[2077.8112,1652.7481],[2078.3112,1640.1825],[2079.3112,1627.9145],[2079.8112,1615.0437],[2080.8112,1602.3636],[2081.3112,1589.6073],[2083.3112,1563.5681],[2083.8112,1550.6592],[2084.8112,1538.2386],[2085.3112,1525.8713],[2086.3112,1512.9013],[2086.8112,1499.6872],[2087.8112,1486.6867],[2088.3112,1474.3119],[2089.3112,1461.38],[2089.8112,1448.7915],[2091.8112,1421.547],[2092.3112,1407.1122],[2094.3112,1381.1951],[2094.8112,1368.3472],[2095.8112,1355.1712],[2096.3112,1342.1021],[2097.3112,1328.8269],[2097.8112,1316.5512],[2098.8112,1303.7873],[2099.3112,1290.3977],[2100.3112,1278.1601],[2100.8112,1265.5945],[2101.8112,1253.3188],[2102.3112,1240.7837],[2103.3112,1228.0045],[2103.8112,1215.3015],[2104.8112,1202.446],[2105.3112,1190.4984],[2106.3112,1178.4821],[2106.8112,1165.39],[2107.8112,1152.092],[2108.3112,1139.8468],[2109.3112,1127.2583],[2109.8112,1114.6927],[2110.8112,1101.9516],[2111.3112,1089.8056],[2113.3112,1064.3921],[2113.8112,1051.5824],[2114.8112,1038.3759],[2115.3112,1025.8713],[2117.3112,1001.0376],[2118.3112,981.0486],[2118.8112,967.8421],[2119.8112,955.3223],[2120.3112,942.7567],[2121.3112,929.8554],[2121.8112,917.0304],[2124.3112,887.7564],[2124.8112,868.8279],[2125.8112,853.8285],[2126.3112,840.0498],[2128.3112,814.4303],[2128.8112,801.7578],[2129.8112,788.8871],[2130.3112,775.9476],[2131.3112,763.115],[2131.8112,750.4578],[2132.8112,738.0066],[2133.3112,725.7233],[2134.3112,713.2187],[2134.8112,700.3174],[2135.8112,686.6455],[2136.3112,673.7671],[2138.3112,647.4915],[2138.8112,635.0785],[2139.8112,622.6044],[2140.3112,609.4666],[2141.3112,596.3898],[2141.8112,583.87],[2142.8112,570.9305],[2143.3112,558.5251],[2144.3112,545.4789],[2144.8112,532.9209],[2145.8112,520.4621],[2146.3112,508.0872],[2147.3112,495.7657],[2147.8112,483.1467],[2148.8112,470.5658],[2149.3112,458.4046],[2150.3112,446.2586],[2150.8112,433.9981],[2151.8112,421.4707],[2152.3112,408.8288],[2153.3112,396.3547],[2153.8112,383.7204],[2154.8112,371.0862],[2155.3112,358.6579],[2156.3112,344.78],[2156.8112,331.8329],[2158.8112,306.3965],[2159.3112,292.7857],[2160.3112,280.2811],[2160.8112,267.1662],[2162.8112,239.8453],[2163.3112,226.4405],[2164.3112,213.8214],[2164.8112,200.531],[2165.8112,186.5158],[2166.8112,173.317],[2167.3112,159.5841],[2168.3112,146.3929],[2168.8112,133.5526],[2169.8112,120.0943],[2170.3112,107.1015],[2171.3112,94.3756],[2171.8112,82.222],[2172.8112,69.4657],[2173.8112,56.1447],[2174.3112,43.2282],[2175.3112,30.1667],[2175.8112,17.4027],[2176.8112,4.982],[2177.3112,-7.9879],[2178.3112,-20.4239],[2180.8112,-33.8745],[2181.8112,-77.9876],[2182.8112,-92.7276],[2183.3112,-107.0022],[2184.3112,-120.3155],[2184.8112,-133.2245],[2185.8112,-146.4462],[2186.8112,-158.6761],[2188.8112,-190.9256],[2189.8112,-207.4508],[2190.3112,-223.1598],[2193.3112,-271.0952],[2194.3112,-288.681],[2195.3112,-305.4962],[2198.3112,-352.9129],[2200.3112,-386.2686],[2201.3112,-402.1911],[2202,-403.2838]]
       }
       Rotation.X {type=MovieSceneScriptingDoubleChannel, default=0, const [0,3]}
       Rotation.Y {type=MovieSceneScriptingDoubleChannel, default=0, const [0,3]}
       Rotation.Z {
        type=MovieSceneScriptingDoubleChannel
        default=-105
        thinned_from=179
        keys [[1834.5,-152.7371],[1836,-152.426],[1837,-151.9594],[1865.5,-151.8038],[1866.5,-151.9594],[1868,-152.5815],[1892,-152.7371],[1894.5,-153.6703],[1909.5,-153.9814],[1912,-154.6036],[1916.5,-155.0702],[1933,-154.9146],[1934,-154.2925],[1934.5,-153.6703],[1936,-152.8926],[1937.5,-152.5815],[1968.5,-152.5815],[1974.5,-153.8258],[1976.5,-154.6036],[1987.5,-154.7591],[1994,-154.6036],[1995,-154.7591],[1995.5,-155.6923],[1998,-156.3145],[2034,-156.0034],[2035,-156.1589],[2035.5,-157.7143],[2036.5,-158.8031],[2037,-159.8919],[2038,-160.203],[2045.5,-160.5141],[2049.5,-160.0474],[2077,-161.4473],[2092,-161.2918],[2092.5,-160.6696],[2095,-159.5808],[2097.5,-159.2698],[2098,-158.8031],[2106.5,-158.492],[2109,-157.7143],[2110.5,-156.6256],[2111.5,-155.3812],[2112.5,-155.0702],[2115,-154.6036],[2127,-154.7591],[2127.5,-154.6036],[2128.5,-153.8258],[2129,-153.0481],[2130.5,-152.2705],[2150.5,-152.1149],[2151,-151.4927],[2152.5,-150.8706],[2154,-150.5595],[2155,-150.0929],[2158,-150.0929],[2159.5,-149.4707],[2162,-149.1596],[2163.5,-148.3819],[2165,-148.3819],[2172,-145.7378],[2174,-145.2712],[2174.5,-144.9601],[2177,-144.649],[2178.5,-144.0268],[2181,-143.5602],[2182,-142.7825],[2183.5,-142.1603],[2184.5,-142.1603],[2185,-141.8493],[2187,-141.6937],[2189,-140.6049],[2190,-140.2939],[2191.5,-139.2051],[2192,-146.6573],[2192.5,-138.4274],[2194.5,-137.0275],[2196.5,-134.6944],[2202,-129.0716]]
       }
       Scale.X {type=MovieSceneScriptingDoubleChannel, default=1}
       Scale.Y {type=MovieSceneScriptingDoubleChannel, default=1}
       Scale.Z {type=MovieSceneScriptingDoubleChannel, default=1}
      ]
     }
     {
      type=MovieScene3DTransformSection
      end_frame=7199
      start_frame=1799
      channels [
       Location.X {type=MovieSceneScriptingDoubleChannel, default=-293, keys [[2120.5,84.0192],[2202,-295],[2750,-631]]}
       Location.Y {type=MovieSceneScriptingDoubleChannel, default=669, keys [[2120.5,852.7858],[2202,582]]}
       Location.Z {type=MovieSceneScriptingDoubleChannel, default=54, keys [[2120.5,-100.1506],[2202,54]]}
       Rotation.X {type=MovieSceneScriptingDoubleChannel, default=0, const [0,2]}
       Rotation.Y {type=MovieSceneScriptingDoubleChannel, default=0, const [0,2]}
       Rotation.Z {type=MovieSceneScriptingDoubleChannel, default=0, const [0,2]}
       Scale.X {type=MovieSceneScriptingDoubleChannel, default=0, const [0,2]}
       Scale.Y {type=MovieSceneScriptingDoubleChannel, default=0, const [0,2]}
       Scale.Z {type=MovieSceneScriptingDoubleChannel, default=0, const [0,2]}
      ]
     }
    ]
   }
   {
    type=MovieSceneVisibilityTrack
    sections [{type=MovieSceneVisibilitySection, channels [None {type=MovieSceneScriptingBoolChannel, default=false, keys [[1799,false],[2097,true]]}]}]
   }
   {type=MovieSceneSkeletalAnimationTrack, sections [{type=MovieSceneSkeletalAnimationSection, end_frame=4705, start_frame=1799}]}
  ]
 }
 Mark_III_Glider_Jones {
  id=#3
  possessed_class=FortSkeletalMeshActor
  tracks [
   {type=MovieSceneSkeletalAnimationTrack, sections [{type=MovieSceneSkeletalAnimationSection, end_frame=2299, start_frame=1813}]}
   {
    type=MovieSceneVisibilityTrack
    sections [{
      type=MovieSceneVisibilitySection
      channels [None {type=MovieSceneScriptingBoolChannel, default=false, keys [[1800,false],[2098,true],[2195,true],[2203,false]]}]
     }]
   }
   {type=MovieScene3DAttachTrack, sections [{type=MovieScene3DAttachSection, end_frame=2201, start_frame=1800}]}
   {
    type=MovieScene3DTransformTrack
    sections [{
      type=MovieScene3DTransformSection
      channels [
       Location.X {type=MovieSceneScriptingDoubleChannel, default=-21, keys [[2095,0],[2191,0],[2195,-4.867],[2197,-5.6461],[2199,-19.759],[2200,-34.3082]]}
       Location.Y {
        type=MovieSceneScriptingDoubleChannel
        default=11
        keys [[2095,0],[2191,0],[2195,-3.5516],[2197,-7.1134],[2198,3.9854],[2199,21.4983],[2200,4.584]]
       }
       Location.Z {
        type=MovieSceneScriptingDoubleChannel
        default=38
        keys [[2095,0],[2191,0],[2193,4.1059],[2195,16.6417],[2197,29.2658],[2198,34.6741],[2199,20.6213],[2200,6.362]]
       }
       Rotation.X {type=MovieSceneScriptingDoubleChannel, default=12, thinned_from=6, keys [[2095,0],[2198,0],[2199,11.4918],[2200,34.8804]]}
       Rotation.Y {type=MovieSceneScriptingDoubleChannel, default=-110, keys [[2095,0],[2191,0],[2197,0.0893],[2199,-64.289],[2200,-112.7007]]}
       Rotation.Z {type=MovieSceneScriptingDoubleChannel, default=78, const [90,3]}
       Scale.X {type=MovieSceneScriptingDoubleChannel, default=0.1, keys [[2095,1],[2197,1],[2202,0.1]]}
       Scale.Y {type=MovieSceneScriptingDoubleChannel, default=0.1, keys [[2095,1],[2197,1],[2202,0.1]]}
       Scale.Z {type=MovieSceneScriptingDoubleChannel, default=0.1, keys [[2095,1],[2197,1],[2202,0.1]]}
      ]
     }]
   }
  ]
 }
]
display_rate {denominator=1, numerator=30}
tick_resolution {denominator=1, numerator=24000}
tracks []
