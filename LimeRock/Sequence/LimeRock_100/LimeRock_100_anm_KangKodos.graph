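# Level Sequence for LimeRock_100 that animates a single skeletal mesh actor, GoggleBud_Glider2, along a scripted flight at 60 fps.
# One transform track layers three sections: a fixed-position base with repeating height bob and tilt, large swooping moves that end flying far away, and a looping sway.
# The zero-scale offset sections suggest additive layering over the base.
# A visibility track and one skeletal animation section accompany the motion.
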LevelSequence /LimeRock/Sequence/LimeRock_100/LimeRock_100_anm_KangKodos
bindings [GoggleBud_Glider2 {
  id=#1
  possessed_class=FortSkeletalMeshActor
  tracks [
   {
    type=MovieScene3DTransformTrack
    sections [
     {
      type=MovieScene3DTransformSection
      channels [
       Weight {type=MovieSceneScriptingFloatChannel, default=1, const [1,9]}
       Location.X {type=MovieSceneScriptingDoubleChannel, default=-2955, const [-2955,8]}
       Location.Y {type=MovieSceneScriptingDoubleChannel, default=-6560, const [-6560,8]}
       Location.Z {
        type=MovieSceneScriptingDoubleChannel
        default=-698.6482
        thinned_from=98
        keys [[2879,4906.1128],[2913,4773.2557],[2966,4967.1649],[3025,4773.2557],[3085,4967.1649],[3147,4773.2557],[3207,4967.1649],[3269,4773.2557],[3329,4967.1649],[3391,4773.2557],[3451,4967.1649],[3513,4773.2557],[3573,4967.1649],[3597,4902.4917],[3635,4773.2557],[3695,4967.1649],[3747,4773.2557],[3805,4967.0625],[3879,4773.2557],[3939,4967.1649],[3980,4906.1128],[4014,4773.2557],[4067,4967.1649],[4126,4773.2557],[4186,4967.1649],[4248,4773.2557],[4264,4807.269],[4308,4967.1649],[4370,4773.2557],[4430,4967.1649],[4492,4773.2557],[4552,4967.1649],[4614,4773.2557],[4674,4967.1649],[4736,4773.2557],[4796,4967.1649],[4848,4773.2557],[4905,4967.1649],[4980,4773.2557],[5040,4967.1649],[5081,4906.1128],[5115,4773.2557],[5168,4967.1649],[5227,4773.2557],[5287,4967.1649],[5349,4773.2557],[5409,4967.1649],[5471,4773.2557],[5531,4967.1649],[5593,4773.2557],[5653,4967.1649],[5715,4773.2557],[5775,4967.1649],[5837,4773.2557],[5897,4967.1649],[5949,4773.2557],[6153,4967.1649],[6228,4773.2557],[6288,4967.1649],[6329,4906.1128],[6363,4773.2557],[6416,4967.1649],[6475,4773.2557],[6535,4967.1649],[6597,4773.2557],[6657,4967.1649],[6719,4773.2557],[6779,4967.1649],[6841,4773.2557],[6901,4967.1649],[6963,4773.2557],[7023,4967.1649],[7085,4773.2557],[7145,4967.1649],[7197,4773.2557],[7254,4967.1649],[7329,4773.2557],[7389,4967.1649],[7430,4906.1128],[7464,4773.2557],[7576,4773.2557],[7636,4967.1649],[7698,4773.2557],[7758,4967.1649],[7820,4773.2557],[7880,4967.1649],[7942,4773.2557],[8002,4967.1649],[8064,4773.2557],[8124,4967.1649],[8186,4773.2557],[8246,4967.1649],[8298,4773.2557],[8355,4967.1649],[8430,4773.2557],[8490,4967.1649],[8531,4906.1128]]
       }
       Rotation.X {
        type=MovieSceneScriptingDoubleChannel
        default=0
        keys [[2879,-10.5817],[2930,-15.9016],[2988,-10.4178],[3046,-15.9016],[3104,-10.4178],[3169,-15.5845],[3232,-10.2593],[3288,-15.9016],[3342,-10.4178],[3407,-15.7424],[3471,-10.4178],[3537,-15.7424],[3597,-11.0547],[3663,-15.7424],[3723,-10.577],[3784,-15.2647],[3805,-13.8933],[3845,-10.2586],[3914,-15.1055],[3980,-10.5817],[4031,-15.9016],[4089,-10.4178],[4147,-15.9016],[4205,-10.4178],[4264,-15.4606],[4270,-15.5845],[4333,-10.2593],[4389,-15.9016],[4443,-10.4178],[4508,-15.7424],[4572,-10.4178],[4638,-15.7424],[4698,-11.0547],[4764,-15.7424],[4824,-10.577],[4885,-15.2647],[4946,-10.2586],[5015,-15.1055],[5081,-10.5817],[5132,-15.9016],[5190,-10.4178],[5248,-15.9016],[5306,-10.4178],[5371,-15.5845],[5434,-10.2593],[5490,-15.9016],[5544,-10.4178],[5609,-15.7424],[5673,-10.4178],[5739,-15.7424],[5799,-11.0547],[5865,-15.7424],[5925,-10.577],[6133,-15.2647],[6194,-10.2586],[6263,-15.1055],[6329,-10.5817],[6380,-15.9016],[6438,-10.4178],[6496,-15.9016],[6554,-10.4178],[6619,-15.5845],[6682,-10.2593],[6738,-15.9016],[6792,-10.4178],[6857,-15.7424],[6921,-10.4178],[6987,-15.7424],[7047,-11.0547],[7113,-15.7424],[7173,-10.577],[7234,-15.2647],[7295,-10.2586],[7364,-15.1055],[7430,-10.5817],[7481,-15.9016],[7539,-10.4178],[7597,-15.9016],[7655,-10.4178],[7720,-15.5845],[7783,-10.2593],[7839,-15.9016],[7893,-10.4178],[7958,-15.7424],[8022,-10.4178],[8088,-15.7424],[8148,-11.0547],[8214,-15.7424],[8274,-10.577],[8335,-15.2647],[8396,-10.2586],[8465,-15.1055],[8531,-10.5817]]
       }
       Rotation.Y {
        type=MovieSceneScriptingDoubleChannel
        default=0
        keys [[2879,-1.6548],[2988,-3.3404],[3061,-1.6543],[3172,-3.3404],[3245,-1.6543],[3356,-3.3404],[3429,-1.6543],[3540,-3.3404],[3597,-1.8618],[3613,-1.6543],[3717,-3.3404],[3784,-1.6543],[3805,-1.783],[3908,-3.3404],[3980,-1.6548],[4089,-3.3404],[4162,-1.6543],[4264,-3.309],[4273,-3.3404],[4346,-1.6543],[4457,-3.3404],[4530,-1.6543],[4641,-3.3404],[4714,-1.6543],[4818,-3.3404],[4885,-1.6543],[5009,-3.3404],[5081,-1.6548],[5190,-3.3404],[5263,-1.6543],[5374,-3.3404],[5447,-1.6543],[5558,-3.3404],[5631,-1.6543],[5742,-3.3404],[5815,-1.6543],[5919,-3.3404],[6133,-1.6543],[6257,-3.3404],[6329,-1.6548],[6438,-3.3404],[6511,-1.6543],[6622,-3.3404],[6695,-1.6543],[6806,-3.3404],[6879,-1.6543],[6990,-3.3404],[7063,-1.6543],[7167,-3.3404],[7234,-1.6543],[7358,-3.3404],[7430,-1.6548],[7539,-3.3404],[7612,-1.6543],[7723,-3.3404],[7796,-1.6543],[7907,-3.3404],[7980,-1.6543],[8091,-3.3404],[8164,-1.6543],[8268,-3.3404],[8335,-1.6543],[8459,-3.3404],[8531,-1.6548]]
       }
       Rotation.Z {type=MovieSceneScriptingDoubleChannel, default=0, thinned_from=14, keys [[2879,-989.5686],[8531,-989.5686]]}
      ]
     }
     {
      type=MovieScene3DTransformSection
      channels [
       Weight {type=MovieSceneScriptingFloatChannel, default=1, const [1,13]}
       Location.X {
        type=MovieSceneScriptingDoubleChannel
        default=-5645
        thinned_from=23
        keys [[4527,104.8001],[4565,160.1095],[4599,1182.4676],[4661,4913.4681],[4718,5327.4651],[4745,5355.8398],[4784,5217.1572],[4799,4913.8281],[4858,-2395.2515],[4872,-3019.9307],[4926,-110.7634],[5002,-1593.5439],[6257,-1596.3251],[7214,-1585.263],[7370,-1515.6815],[7404,-2754.9133],[7455,568.3422],[7509,8827.114],[7529,8958.5451],[7630,-19756.1026],[12035,-981291.6309]]
       }
       Location.Y {
        type=MovieSceneScriptingDoubleChannel
        default=3959
        keys [[4527,5883.0769],[4566,5811.4646],[4586,5409.9968],[4634,3646.4841],[4665,4528.3823],[4696,6105.7466],[4729,6329.7696],[4756,6187.3182],[4791,6675.2007],[4826,8147.8213],[4840,8386.5533],[4904,3793.6439],[4990,-380.3048],[5000,-437.8339],[5047,-246.018],[5100,-363.4334],[6257,-268.9348],[7191,-194.4379],[7198,-202.2611],[7270,-920.725],[7351,-931.5789],[7363,-960.5366],[7393,-1870.899],[7399,-1767.0651],[7448,3187.3615],[7496,6384.2378],[7565,8811.9971],[7592,8908.536],[7839,36169.8881],[8593,81105.6767]]
       }
       Location.Z {
        type=MovieSceneScriptingDoubleChannel
        default=100
        keys [[4527,-700.4705],[4568,-700.5296],[4582,-853.2269],[4641,-4015.5772],[4691,-4900.9416],[4745,-4900.9416],[4787,-4824.765],[4811,-4552.4788],[4855,-3381.6938],[4908,-2905.4433],[4944,-1982.9363],[4980,-316.2337],[5021,55.139],[6257,60.7454],[7199,7.5387],[7211,-93.2394],[7229,96.6939],[7251,-126.0804],[7270,11.92],[7320,3.4924],[7337,9.3198],[7369,-175.2462],[7388,-549.3895],[7469,-4217.7539],[7484,-4747.2506],[7499,-5255.9543],[7538,-4608.5844],[7566,-3496.8428],[7609,2471.3044],[7766,98636.3594],[7986,307771.7081]]
       }
       Rotation.X {
        type=MovieSceneScriptingDoubleChannel
        default=2
        thinned_from=46
        keys [[4527,4.2525],[4552,3.8732],[4572,1.2786],[4620,-11.7342],[4662,12.222],[4715,1.1452],[4763,-9.9053],[4773,-9.1373],[4811,6.0568],[4819,9.1956],[4839,12.6236],[4921,16.8669],[5029,13.0379],[6257,12.9404],[7079,13.0707],[7168,12.6522],[7199,12.6284],[7234,35.277],[7272,20.0663],[7289,14.5614],[7362,14.6974],[7398,6.2349],[7414,4.6321],[7469,-2.6608],[7496,-0.5847],[7539,10.9905],[7643,76.1718],[8016,96.0962],[8347,96.0962],[8502,-8.5371],[8676,-8.5371],[8738,7],[8822,0.5083],[8960,13],[9418,13],[9573,-8.5371],[9747,-8.5371],[9809,7],[9893,0.5083],[10031,13],[10188,13]]
       }
       Rotation.Y {
        type=MovieSceneScriptingDoubleChannel
        default=-1
        keys [[4527,1.3952],[4576,0.796],[4628,-22.6983],[4715,14.2252],[4754,-1.0212],[4794,3.2855],[4814,6.7319],[4870,-29.6767],[4877,-32.2165],[4917,-2.4144],[4930,4.2071],[4962,-1.7599],[5029,5.2099],[6257,4.4407],[7199,4.8619],[7216,3.4387],[7230,27.4885],[7244,-0.4568],[7258,16.5535],[7279,8.2316],[7328,5.1464],[7338,4.9787],[7360,3.5619],[7399,-9.2404],[7426,9.7429],[7435,15.6868],[7448,15.2297],[7469,6.7099],[7487,-3.1569],[7521,-16.1003],[7531,-12.2493],[7575,11.7867],[7633,17.3313],[7692,19.4018],[7793,5.4268],[7816,4.4377],[7889,14.5167],[7945,3],[8002,-7.814],[8016,-5],[8046,-5],[8094,14.5167],[8150,3],[8207,-7.814],[8347,-7.814],[8433,-22.958],[8526,14.2252],[8579,-0.5132],[8632,5.846],[8679,-1.0212],[8734,3.2855],[8785,1.5859],[8859,-4.0109],[8882,-5],[8960,14.5167],[9016,3],[9073,-7.814],[9087,-5],[9117,-5],[9165,14.5167],[9221,3],[9278,-7.814],[9418,-7.814],[9504,-22.958],[9597,14.2252],[9650,-0.5132],[9703,5.846],[9750,-1.0212],[9805,3.2855],[9856,1.5859],[9930,-4.0109],[9953,-5],[10031,14.5167],[10087,3],[10144,-7.814],[10188,-5]]
       }
       Rotation.Z {
        type=MovieSceneScriptingDoubleChannel
        default=-63
        thinned_from=19
        keys [[4527,-1435.1513],[4723,-1435.2295],[4781,-1455.8864],[4854,-1542.8625],[4922,-1559.0153],[4999,-1669.2811],[5070,-1693.4548],[5134,-1683.6759],[7361,-1689.3365],[7377,-1697.2211],[7457,-1778.7265],[7481,-1786.4367],[7526,-1662.7935],[7571,-1613.8924],[7616,-1613.4635]]
       }
       Scale.X {type=MovieSceneScriptingDoubleChannel, default=0, const [0,14]}
       Scale.Y {type=MovieSceneScriptingDoubleChannel, default=0, const [0,14]}
       Scale.Z {type=MovieSceneScriptingDoubleChannel, default=0, const [0,14]}
      ]
     }
     {
      type=MovieScene3DTransformSection
      channels [
       Location.X {
        type=MovieSceneScriptingDoubleChannel
        default=1773
        thinned_from=22
        keys [[2709,249.6141],[2816,1625.8337],[2923,2104.2413],[3063,840.902],[3174,0],[3289,1625.8337],[3395,2167.2803],[3512,863.1628],[3600,180.6667],[3659,0],[3775,1538.193],[3841,2167.2803],[3869,2213.5244],[4022,930.0212],[4124,0],[4257,1625.8337],[4360,2167.2803],[4469,863.1628],[4575,73.5882],[4631,249.6141],[7519,249.6141]]
       }
       Location.Y {
        type=MovieSceneScriptingDoubleChannel
        default=0
        thinned_from=24
        keys [[2709,-512.0766],[2803,-1115.145],[2914,35.9207],[3006,1130.501],[3067,959.0658],[3174,-512.0766],[3276,-1115.145],[3395,35.9207],[3491,1130.501],[3553,959.0658],[3600,475.169],[3659,-512.0766],[3757,-1115.145],[3860,35.9207],[3956,1130.501],[4019,959.0658],[4124,-512.0766],[4228,-1288.5825],[4353,35.9207],[4450,1130.501],[4505,959.0658],[4670,-512.0766],[7519,-512.0766]]
       }
       Location.Z {
        type=MovieSceneScriptingDoubleChannel
        default=0
        thinned_from=20
        keys [[2709,0],[2818,-2.1515],[2914,-43.6846],[3025,136.8689],[3174,48.5817],[3291,-18.3455],[3395,-43.6846],[3512,136.8689],[3600,91.6176],[3659,48.5817],[3770,-12.7872],[3821,-31.2763],[3979,136.8689],[4124,48.5817],[4241,-18.3455],[4360,-43.6846],[4469,136.8689],[4631,0],[7519,0]]
       }
       Rotation.X {
        type=MovieSceneScriptingDoubleChannel
        default=0
        thinned_from=23
        keys [[2709,-2.7256],[2818,0],[2891,0.9501],[2957,15.7412],[3065,-2.2254],[3165,-5.6962],[3329,6.2425],[3374,5.5589],[3495,-8.3684],[3553,-9.1337],[3600,-8.8802],[3632,-8.5382],[3815,-0.3625],[3866,0],[4047,-5.1962],[4130,-6.1514],[4166,-5.7714],[4307,-0.0392],[4450,-5.7507],[4552,-7.322],[4631,-2.7256],[7519,-2.7256]]
       }
       Rotation.Y {
        type=MovieSceneScriptingDoubleChannel
        default=0
        thinned_from=22
        keys [[2709,0],[2812,-1.4676],[2871,-6.2305],[2969,6.2096],[3021,11.9725],[3197,-10.6332],[3291,-8.6424],[3374,-2.8426],[3447,4.5107],[3600,0.8033],[3632,0],[3740,-1.774],[3817,-2.2155],[3964,-1.4211],[4124,-6.1884],[4223,-7.415],[4358,-6.4468],[4403,-7.1623],[4527,-11.3475],[4631,0],[7519,0]]
       }
       Rotation.Z {
        type=MovieSceneScriptingDoubleChannel
        default=0
        thinned_from=19
        keys [[2709,0],[2814,61.1547],[2876,99.432],[3065,-67.367],[3208,-3.0012],[3387,-43.6368],[3491,-27.2697],[3600,13.5904],[3619,22.3682],[3728,76.1499],[3800,99.432],[4017,-67.367],[4160,-3.0012],[4222,17.3137],[4358,-53.7855],[4407,-72.0469],[4631,0],[7519,0]]
       }
       Scale.X {type=MovieSceneScriptingDoubleChannel, default=0, const [0,20]}
       Scale.Y {type=MovieSceneScriptingDoubleChannel, default=0, const [0,20]}
       Scale.Z {type=MovieSceneScriptingDoubleChannel, default=0, const [0,20]}
      ]
     }
    ]
   }
   {
    type=MovieSceneVisibilityTrack
    sections [{type=MovieSceneVisibilitySection, channels [None {type=MovieSceneScriptingBoolChannel, default=true}]}]
   }
   {type=MovieSceneSkeletalAnimationTrack, sections [{type=MovieSceneSkeletalAnimationSection, end_frame=8199, start_frame=3037}]}
  ]
 }]
display_rate {denominator=1, numerator=60}
tick_resolution {denominator=1, numerator=24000}
tracks []
