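# Level Sequence for LimeRock_100 that animates four player mannequins, the local player's and three numbered ones.
# Each possessed FortPlayerMannequin has two back-to-back skeletal animation sections, a visibility track and a transform track.
# Keys at the animation section boundaries reposition the mannequins and change their scale from 1.5 to 1.
# A possessed B_LimeRock_MannequinManager carries an event trigger track and a fixed transform.
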LevelSequence /LimeRock/Sequence/LimeRock_100/LimeRock_100_anm_Players
bindings [
 FortPlayerMannequin - 02 {
  id=#1
  possessed_class=FortPlayerMannequin
  tracks [
   {
    type=MovieSceneSkeletalAnimationTrack
    sections [
     {type=MovieSceneSkeletalAnimationSection, end_frame=9049, start_frame=8905}
     {type=MovieSceneSkeletalAnimationSection, end_frame=9252, start_frame=9049}
    ]
   }
   {
    type=MovieScene3DTransformTrack
    sections [{
      type=MovieScene3DTransformSection
      channels [
       Location.X {type=MovieSceneScriptingDoubleChannel, default=-90942, thinned_from=3, keys [[3600,-97575.7031],[9049,-90701.3606]]}
       Location.Y {type=MovieSceneScriptingDoubleChannel, default=-95980, thinned_from=3, keys [[3600,-95866.1484],[9049,-95956]]}
       Location.Z {type=MovieSceneScriptingDoubleChannel, default=1006, keys [[3600,600.1835],[8905,1002],[9049,1006]]}
       Rotation.X {type=MovieSceneScriptingDoubleChannel, default=0, const [0,2]}
       Rotation.Y {type=MovieSceneScriptingDoubleChannel, default=0, const [0,2]}
       Rotation.Z {type=MovieSceneScriptingDoubleChannel, default=102, keys [[8905,90],[9049,102]]}
       Scale.X {type=MovieSceneScriptingDoubleChannel, default=1, keys [[8905,1.5],[9049,1]]}
       Scale.Y {type=MovieSceneScriptingDoubleChannel, default=1, keys [[8905,1.5],[9049,1]]}
       Scale.Z {type=MovieSceneScriptingDoubleChannel, default=1, keys [[8905,1.5],[9049,1]]}
      ]
     }]
   }
   {
    type=MovieSceneVisibilityTrack
    sections [{type=MovieSceneVisibilitySection, channels [None {type=MovieSceneScriptingBoolChannel, default=true, const [true,2]}]}]
   }
  ]
 }
 FortPlayerMannequin - 04 {
  id=#2
  possessed_class=FortPlayerMannequin
  tracks [
   {
    type=MovieSceneSkeletalAnimationTrack
    sections [
     {type=MovieSceneSkeletalAnimationSection, end_frame=9049, start_frame=8905}
     {type=MovieSceneSkeletalAnimationSection, end_frame=9251, start_frame=9049}
    ]
   }
   {
    type=MovieScene3DTransformTrack
    sections [{
      type=MovieScene3DTransformSection
      channels [
       Location.X {type=MovieSceneScriptingDoubleChannel, default=-90680, keys [[3600,-97727.7031],[8905,-90680],[9048,-90606.5355]]}
       Location.Y {type=MovieSceneScriptingDoubleChannel, default=-95907, thinned_from=3, keys [[3600,-95850.1484],[9049,-95855.3089]]}
       Location.Z {type=MovieSceneScriptingDoubleChannel, default=1006, keys [[3600,600.1835],[8905,1001],[9049,1006]]}
       Rotation.X {type=MovieSceneScriptingDoubleChannel, default=0, const [0,2]}
       Rotation.Y {type=MovieSceneScriptingDoubleChannel, default=0, const [0,2]}
       Rotation.Z {type=MovieSceneScriptingDoubleChannel, default=106, keys [[8905,90],[9049,106]]}
       Scale.X {type=MovieSceneScriptingDoubleChannel, default=1, keys [[8905,1.5],[9049,1]]}
       Scale.Y {type=MovieSceneScriptingDoubleChannel, default=1, keys [[8905,1.5],[9049,1]]}
       Scale.Z {type=MovieSceneScriptingDoubleChannel, default=1, keys [[8905,1.5],[9049,1]]}
      ]
     }]
   }
   {
    type=MovieSceneVisibilityTrack
    sections [{type=MovieSceneVisibilitySection, channels [None {type=MovieSceneScriptingBoolChannel, default=true, const [true,2]}]}]
   }
  ]
 }
 B_LimeRock_MannequinManager {
  id=#3
  possessed_class=B_LimeRock_MannequinManager_C
  tracks [
   {
    type=MovieSceneEventTrack
    sections [{
      type=MovieSceneEventTriggerSection
      channels [None {type=MovieSceneScriptingEventChannel, const ["<Struct 'MovieSceneEvent' (0x851351c2d00) {}>", 1]}]
     }]
   }
   {
    type=MovieScene3DTransformTrack
    sections [{
      type=MovieScene3DTransformSection
      channels [
       Location.X {type=MovieSceneScriptingDoubleChannel, default=-90836, const [-90836,1]}
       Location.Y {type=MovieSceneScriptingDoubleChannel, default=-95820, const [-95820,1]}
       Location.Z {type=MovieSceneScriptingDoubleChannel, default=1006.0774, const [1006.0775,1]}
       Rotation.X {type=MovieSceneScriptingDoubleChannel, default=0, const [0,1]}
       Rotation.Y {type=MovieSceneScriptingDoubleChannel, default=0, const [0,1]}
       Rotation.Z {type=MovieSceneScriptingDoubleChannel, default=90, const [90,1]}
       Scale.X {type=MovieSceneScriptingDoubleChannel, default=1, const [1,1]}
       Scale.Y {type=MovieSceneScriptingDoubleChannel, default=1, const [1,1]}
       Scale.Z {type=MovieSceneScriptingDoubleChannel, default=1, const [1,1]}
      ]
     }]
   }
  ]
 }
 FortPlayerMannequin - local {
  id=#4
  possessed_class=FortPlayerMannequin
  tracks [
   {
    type=MovieSceneSkeletalAnimationTrack
    sections [
     {type=MovieSceneSkeletalAnimationSection, end_frame=9049, start_frame=8905}
     {type=MovieSceneSkeletalAnimationSection, end_frame=9251, start_frame=9049}
    ]
   }
   {
    type=MovieScene3DTransformTrack
    sections [{
      type=MovieScene3DTransformSection
      channels [
       Location.X {type=MovieSceneScriptingDoubleChannel, default=-90927, thinned_from=3, keys [[3600,-97655.7031],[9049,-90694.9811]]}
       Location.Y {type=MovieSceneScriptingDoubleChannel, default=-95830, thinned_from=3, keys [[3600,-95858.1484],[9049,-95886.261]]}
       Location.Z {type=MovieSceneScriptingDoubleChannel, default=1006, keys [[3600,600.1835],[8905,1003],[9049,1006]]}
       Rotation.X {type=MovieSceneScriptingDoubleChannel, default=0, const [0,2]}
       Rotation.Y {type=MovieSceneScriptingDoubleChannel, default=0, const [0,2]}
       Rotation.Z {type=MovieSceneScriptingDoubleChannel, default=107, keys [[8905,107],[9049,109]]}
       Scale.X {type=MovieSceneScriptingDoubleChannel, default=1, keys [[8905,1.5],[9049,1]]}
       Scale.Y {type=MovieSceneScriptingDoubleChannel, default=1, keys [[8905,1.5],[9049,1]]}
       Scale.Z {type=MovieSceneScriptingDoubleChannel, default=1, keys [[8905,1.5],[9049,1]]}
      ]
     }]
   }
   {
    type=MovieSceneVisibilityTrack
    sections [{type=MovieSceneVisibilitySection, channels [None {type=MovieSceneScriptingBoolChannel, default=true, const [true,2]}]}]
   }
  ]
 }
 FortPlayerMannequin - 03 {
  id=#5
  possessed_class=FortPlayerMannequin
  tracks [
   {
    type=MovieSceneSkeletalAnimationTrack
    sections [
     {type=MovieSceneSkeletalAnimationSection, end_frame=9049, start_frame=8905}
     {type=MovieSceneSkeletalAnimationSection, end_frame=9252, start_frame=9049}
    ]
   }
   {
    type=MovieScene3DTransformTrack
    sections [{
      type=MovieScene3DTransformSection
      channels [
       Location.X {type=MovieSceneScriptingDoubleChannel, default=-90676, keys [[3600,-97499.7031],[8905,-90655.5753],[9049,-90628.1071]]}
       Location.Y {type=MovieSceneScriptingDoubleChannel, default=-96168, keys [[3600,-95886.1484],[8905,-96168],[9049,-95939.4213]]}
       Location.Z {type=MovieSceneScriptingDoubleChannel, default=1006, keys [[3600,600.1835],[8905,1000],[9049,1006]]}
       Rotation.X {type=MovieSceneScriptingDoubleChannel, default=0, const [0,2]}
       Rotation.Y {type=MovieSceneScriptingDoubleChannel, default=0, const [0,2]}
       Rotation.Z {type=MovieSceneScriptingDoubleChannel, default=90, keys [[8905,84],[9049,88.7821]]}
       Scale.X {type=MovieSceneScriptingDoubleChannel, default=1.5, keys [[8905,1.5],[9049,1]]}
       Scale.Y {type=MovieSceneScriptingDoubleChannel, default=1.5, keys [[8905,1.5],[9049,1]]}
       Scale.Z {type=MovieSceneScriptingDoubleChannel, default=1.5, keys [[8905,1.5],[9049,1]]}
      ]
     }]
   }
   {
    type=MovieSceneVisibilityTrack
    sections [{type=MovieSceneVisibilitySection, channels [None {type=MovieSceneScriptingBoolChannel, default=true, const [true,2]}]}]
   }
  ]
 }
]
display_rate {denominator=1, numerator=60}
tick_resolution {denominator=1, numerator=24000}
tracks []
