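# Audio Level Sequence for the LimeRock_100 scripted event, scoring a rift, portal, UFO glider and monster scene with Jones and Hope.
# Possessed actors carry sound and foley sections, while event triggers fire on ambient loops, shield rig, portal, knockback volume and AGV.
# BP_AudioManager_LimeRock holds volume-automated music and keyed bools switching event and end mixes.
# Unbound tracks add dialogue, footsteps, a flyby and two animatic audio tracks.
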LevelSequence /LimeRock/Sequence/LimeRock_100/LimeRock_100_aud
bindings [
 NS_LimeRock_Ground_Portal {id=#1, possessed_class=NiagaraActor, tracks [{type=MovieSceneAudioTrack, display_name=Audio, sections []}]}
 Tentacle_LimeRock_Creature2 {
  id=#2
  possessed_class=FortSkeletalMeshActor
  tracks [{
    type=MovieSceneAudioTrack
    display_name=Audio
    sections [
     {
      type=MovieSceneAudioSection
      end_frame=9381
      start_frame=9173
      channels [
       Volume {type=MovieSceneScriptingFloatChannel, default=1}
       Pitch {type=MovieSceneScriptingFloatChannel, default=1}
      ]
     }
     {
      type=MovieSceneAudioSection
      end_frame=9227
      start_frame=9056
      channels [
       Volume {type=MovieSceneScriptingFloatChannel, default=1}
       Pitch {type=MovieSceneScriptingFloatChannel, default=1}
      ]
     }
     {
      type=MovieSceneAudioSection
      end_frame=9239
      start_frame=8905
      channels [
       Volume {type=MovieSceneScriptingFloatChannel, default=1}
       Pitch {type=MovieSceneScriptingFloatChannel, default=1}
      ]
     }
     {
      type=MovieSceneAudioSection
      end_frame=9042
      start_frame=8667
      channels [
       Volume {type=MovieSceneScriptingFloatChannel, default=1}
       Pitch {type=MovieSceneScriptingFloatChannel, default=1}
      ]
     }
    ]
   }]
 }
 SC_LimeRock_100_UFO_Flying_LP {
  id=#3
  possessed_class=AmbientSound
  tracks [{
    type=MovieSceneEventTrack
    sections [{
      type=MovieSceneEventTriggerSection
      channels [None {type=MovieSceneScriptingEventChannel, const ["<Struct 'MovieSceneEvent' (0x851351c4380) {}>", 1]}]
     }]
   }]
 }
 NS_LimeRock_Barrier_Fence {
  id=#4
  possessed_class=BP_LimeRock_ShieldRig_C
  tracks [
   {
    type=MovieSceneAudioTrack
    display_name=Audio
    sections [
     {
      type=MovieSceneAudioSection
      end_frame=4384
      start_frame=4235
      channels [
       Volume {type=MovieSceneScriptingFloatChannel, default=1}
       Pitch {type=MovieSceneScriptingFloatChannel, default=1}
      ]
     }
     {
      type=MovieSceneAudioSection
      end_frame=4526
      start_frame=4355
      channels [
       Volume {type=MovieSceneScriptingFloatChannel, default=1}
       Pitch {type=MovieSceneScriptingFloatChannel, default=1}
      ]
     }
     {
      type=MovieSceneAudioSection
      end_frame=6057
      start_frame=5936
      channels [
       Volume {type=MovieSceneScriptingFloatChannel, default=1}
       Pitch {type=MovieSceneScriptingFloatChannel, default=1}
      ]
     }
    ]
   }
   {
    type=MovieSceneEventTrack
    sections [{
      type=MovieSceneEventTriggerSection
      channels [None {type=MovieSceneScriptingEventChannel, const ["<Struct 'MovieSceneEvent' (0x851351c49c0) {}>", 1]}]
     }]
   }
  ]
 }
 S_Rift_Tear {
  id=#5
  possessed_class=FortStaticMeshActor
  tracks [{
    type=MovieSceneAudioTrack
    display_name=OpeningRiftAudio
    sections [
     {
      type=MovieSceneAudioSection
      end_frame=4386
      start_frame=4193
      channels [
       Volume {type=MovieSceneScriptingFloatChannel, default=1}
       Pitch {type=MovieSceneScriptingFloatChannel, default=1}
      ]
     }
     {
      type=MovieSceneAudioSection
      end_frame=4213
      start_frame=3972
      channels [
       Volume {type=MovieSceneScriptingFloatChannel, default=1}
       Pitch {type=MovieSceneScriptingFloatChannel, default=1}
      ]
     }
    ]
   }]
 }
 Jones {
  id=#6
  possessed_class=FortPlayerMannequin
  tracks [{
    type=MovieSceneAudioTrack
    display_name=JonesFoley
    sections [
     {
      type=MovieSceneAudioSection
      end_frame=4849
      start_frame=4228
      channels [
       Volume {type=MovieSceneScriptingFloatChannel, default=1}
       Pitch {type=MovieSceneScriptingFloatChannel, default=1}
      ]
     }
     {
      type=MovieSceneAudioSection
      end_frame=9186
      start_frame=9049
      channels [
       Volume {type=MovieSceneScriptingFloatChannel, default=1}
       Pitch {type=MovieSceneScriptingFloatChannel, default=1}
      ]
     }
    ]
   }]
 }
 SC_LimeRock_100_TreeHouse_RiftEnergy_LP {
  id=#7
  possessed_class=AmbientSound
  tracks [{
    type=MovieSceneEventTrack
    sections [{
      type=MovieSceneEventTriggerSection
      channels [None {type=MovieSceneScriptingEventChannel, const ["<Struct 'MovieSceneEvent' (0x851351c58c0) {}>", 1]}]
     }]
   }]
 }
 BP_LimeRock_Portal {
  id=#8
  possessed_class=BP_LimeRock_Portal_C
  tracks [
   {
    type=MovieSceneAudioTrack
    display_name=Audio
    sections [
     {
      type=MovieSceneAudioSection
      end_frame=7485
      start_frame=7074
      channels [
       Volume {type=MovieSceneScriptingFloatChannel, default=1}
       Pitch {type=MovieSceneScriptingFloatChannel, default=1}
      ]
     }
     {
      type=MovieSceneAudioSection
      end_frame=5498
      start_frame=4972
      channels [
       Volume {type=MovieSceneScriptingFloatChannel, default=1}
       Pitch {type=MovieSceneScriptingFloatChannel, default=1}
      ]
     }
     {
      type=MovieSceneAudioSection
      end_frame=8207
      start_frame=7679
      channels [
       Volume {type=MovieSceneScriptingFloatChannel, default=1}
       Pitch {type=MovieSceneScriptingFloatChannel, default=1}
      ]
     }
     {
      type=MovieSceneAudioSection
      end_frame=8384
      start_frame=7735
      channels [
       Volume {type=MovieSceneScriptingFloatChannel, default=1.5, keys [[7735,0],[7765,0],[7864,1.5]]}
       Pitch {type=MovieSceneScriptingFloatChannel, default=1}
      ]
     }
     {
      type=MovieSceneAudioSection
      end_frame=8471
      start_frame=7698
      channels [
       Volume {type=MovieSceneScriptingFloatChannel, default=0, keys [[7698,0],[7734,0],[7871,1],[8046,1],[8413,0]]}
       Pitch {type=MovieSceneScriptingFloatChannel, default=1}
      ]
     }
     {
      type=MovieSceneAudioSection
      end_frame=7708
      start_frame=7597
      channels [
       Volume {type=MovieSceneScriptingFloatChannel, default=1}
       Pitch {type=MovieSceneScriptingFloatChannel, default=1}
      ]
     }
     {
      type=MovieSceneAudioSection
      end_frame=6212
      start_frame=5841
      channels [
       Volume {type=MovieSceneScriptingFloatChannel, default=1}
       Pitch {type=MovieSceneScriptingFloatChannel, default=1}
      ]
     }
     {
      type=MovieSceneAudioSection
      end_frame=5802
      start_frame=5470
      channels [
       Volume {type=MovieSceneScriptingFloatChannel, default=1}
       Pitch {type=MovieSceneScriptingFloatChannel, default=1}
      ]
     }
     {
      type=MovieSceneAudioSection
      end_frame=6727
      start_frame=6248
      channels [
       Volume {type=MovieSceneScriptingFloatChannel, default=1}
       Pitch {type=MovieSceneScriptingFloatChannel, default=1}
      ]
     }
     {
      type=MovieSceneAudioSection
      end_frame=7008
      start_frame=6774
      channels [
       Volume {type=MovieSceneScriptingFloatChannel, default=1}
       Pitch {type=MovieSceneScriptingFloatChannel, default=1}
      ]
     }
    ]
   }
   {
    type=MovieSceneEventTrack
    sections [{
      type=MovieSceneEventTriggerSection
      channels [None {type=MovieSceneScriptingEventChannel, const ["<Struct 'MovieSceneEvent' (0x851351c6b80) {}>", 1]}]
     }]
   }
  ]
 }
 NS_LR_Treehouse_Pulse_FX {id=#9, possessed_class=NiagaraActor, tracks []}
 NS_LimeRock_Treehouse_FX {
  id=#10
  possessed_class=NiagaraActor
  tracks [{
    type=MovieSceneAudioTrack
    display_name=TreeAudio
    sections [
     {
      type=MovieSceneAudioSection
      end_frame=7462
      start_frame=6715
      channels [
       Volume {type=MovieSceneScriptingFloatChannel, default=1}
       Pitch {type=MovieSceneScriptingFloatChannel, default=1}
      ]
     }
     {
      type=MovieSceneAudioSection
      end_frame=6344
      start_frame=5905
      channels [
       Volume {type=MovieSceneScriptingFloatChannel, default=1}
       Pitch {type=MovieSceneScriptingFloatChannel, default=1}
      ]
     }
     {
      type=MovieSceneAudioSection
      end_frame=6102
      start_frame=5655
      channels [
       Volume {type=MovieSceneScriptingFloatChannel, default=1}
       Pitch {type=MovieSceneScriptingFloatChannel, default=1}
      ]
     }
     {
      type=MovieSceneAudioSection
      end_frame=7683
      start_frame=7551
      channels [
       Volume {type=MovieSceneScriptingFloatChannel, default=1}
       Pitch {type=MovieSceneScriptingFloatChannel, default=1}
      ]
     }
     {
      type=MovieSceneAudioSection
      end_frame=4709
      start_frame=4589
      channels [
       Volume {type=MovieSceneScriptingFloatChannel, default=1}
       Pitch {type=MovieSceneScriptingFloatChannel, default=1}
      ]
     }
     {
      type=MovieSceneAudioSection
      end_frame=5118
      start_frame=4998
      channels [
       Volume {type=MovieSceneScriptingFloatChannel, default=1}
       Pitch {type=MovieSceneScriptingFloatChannel, default=1}
      ]
     }
    ]
   }]
 }
 BP_AudioManager_LimeRock {
  id=#11
  possessed_class=BP_AudioManager_LimeRock_C
  tracks [
   {
    type=MovieSceneAudioTrack
    display_name=Music
    sections [
     {
      type=MovieSceneAudioSection
      end_frame=8080
      start_frame=4940
      channels [
       Volume {type=MovieSceneScriptingFloatChannel, default=1, keys [[5429,1],[5471.345,0.7],[5533.37,1]]}
       Pitch {type=MovieSceneScriptingFloatChannel, default=1}
      ]
     }
     {
      type=MovieSceneAudioSection
      end_frame=9364
      start_frame=7609
      channels [
       Volume {
        type=MovieSceneScriptingFloatChannel
        default=0.35
        thinned_from=13
        keys [[7609,1],[7695,1.5],[7966,1.5],[8116.9675,1],[8570,1],[8594,0.35],[8700,0.35],[8719,1],[9060.315,1],[9187.97,1.3]]
       }
       Pitch {type=MovieSceneScriptingFloatChannel, default=1}
      ]
     }
     {
      type=MovieSceneAudioSection
      end_frame=5523
      start_frame=4599
      channels [
       Volume {type=MovieSceneScriptingFloatChannel, default=0.5, keys [[4599,1],[4920.175,1],[5010,0.5]]}
       Pitch {type=MovieSceneScriptingFloatChannel, default=1}
      ]
     }
     {
      type=MovieSceneAudioSection
      end_frame=4922
      start_frame=3985
      channels [
       Volume {type=MovieSceneScriptingFloatChannel, default=0.9, keys [[3985,0.7],[4249,0.7],[4633,0.9]]}
       Pitch {type=MovieSceneScriptingFloatChannel, default=1}
      ]
     }
    ]
   }
   {
    type=MovieSceneBoolTrack
    display_name=Event Mix
    sections [{
      type=MovieSceneBoolSection
      channels [None {type=MovieSceneScriptingBoolChannel, default=false, keys [[3493,false],[3600,true],[14808,false]]}]
     }]
   }
   {
    type=MovieSceneBoolTrack
    display_name=Event End Mix Active
    sections [{
      type=MovieSceneBoolSection
      channels [None {type=MovieSceneScriptingBoolChannel, default=false, keys [[3593,false],[9221,true],[14394,false]]}]
     }]
   }
  ]
 }
 BP_LimeRock_KnockbackVolume {
  id=#12
  possessed_class=BP_LimeRock_KnockbackVolume_C
  tracks [{
    type=MovieSceneEventTrack
    sections [{
      type=MovieSceneEventTriggerSection
      channels [None {type=MovieSceneScriptingEventChannel, const ["<Struct 'MovieSceneEvent' (0x851351c8ca0) {}>", 1]}]
     }]
   }]
 }
 AGV_LimeRock {
  id=#13
  possessed_class=AGV_LimeRock_C
  tracks [{
    type=MovieSceneEventTrack
    sections [{
      type=MovieSceneEventTriggerSection
      channels [None {type=MovieSceneScriptingEventChannel, const ["<Struct 'MovieSceneEvent' (0x851351c83e0) {}>", 1]}]
     }]
   }]
 }
 Tentacle_LimeRock_Creature_BG_1 {
  id=#14
  possessed_class=FortSkeletalMeshActor
  tracks [{type=MovieSceneAudioTrack, display_name=Tentacle_01_Audio, sections []}]
 }
 NS_LimeRock_SmokeMonsterMain2 {
  id=#15
  possessed_class=NiagaraActor
  tracks [{
    type=MovieSceneAudioTrack
    display_name=MonsterAudio
    sections [
     {
      type=MovieSceneAudioSection
      end_frame=8786
      start_frame=7968
      channels [
       Volume {type=MovieSceneScriptingFloatChannel, default=2, keys [[7976,1.3],[8088,0.6],[8160,0.6],[8202,1],[8336.275,1],[8384,2]]}
       Pitch {type=MovieSceneScriptingFloatChannel, default=1}
      ]
     }
     {
      type=MovieSceneAudioSection
      end_frame=8837
      start_frame=8494
      channels [
       Volume {type=MovieSceneScriptingFloatChannel, default=1}
       Pitch {type=MovieSceneScriptingFloatChannel, default=1}
      ]
     }
     {
      type=MovieSceneAudioSection
      end_frame=9087
      start_frame=7923
      channels [
       Volume {type=MovieSceneScriptingFloatChannel, default=1}
       Pitch {type=MovieSceneScriptingFloatChannel, default=1}
      ]
     }
     {
      type=MovieSceneAudioSection
      end_frame=9219
      start_frame=8391
      channels [
       Volume {type=MovieSceneScriptingFloatChannel, default=1}
       Pitch {type=MovieSceneScriptingFloatChannel, default=1}
      ]
     }
     {
      type=MovieSceneAudioSection
      end_frame=9254
      start_frame=9052
      channels [
       Volume {type=MovieSceneScriptingFloatChannel, default=1}
       Pitch {type=MovieSceneScriptingFloatChannel, default=1}
      ]
     }
     {
      type=MovieSceneAudioSection
      end_frame=8149
      start_frame=7899
      channels [
       Volume {type=MovieSceneScriptingFloatChannel, default=1}
       Pitch {type=MovieSceneScriptingFloatChannel, default=1}
      ]
     }
     {
      type=MovieSceneAudioSection
      end_frame=8474
      start_frame=8155
      channels [
       Volume {type=MovieSceneScriptingFloatChannel, default=1}
       Pitch {type=MovieSceneScriptingFloatChannel, default=1}
      ]
     }
     {
      type=MovieSceneAudioSection
      end_frame=8837
      start_frame=8565
      channels [
       Volume {type=MovieSceneScriptingFloatChannel, default=1}
       Pitch {type=MovieSceneScriptingFloatChannel, default=1}
      ]
     }
     {
      type=MovieSceneAudioSection
      end_frame=8472
      start_frame=8270
      channels [
       Volume {type=MovieSceneScriptingFloatChannel, default=1}
       Pitch {type=MovieSceneScriptingFloatChannel, default=1}
      ]
     }
    ]
   }]
 }
 SM_Utility_Laptop_A_Blueprint2 {
  id=#16
  possessed_class=BP_Utility_Laptop_A_C
  tracks [{
    type=MovieSceneAudioTrack
    display_name=Audio
    sections [
     {
      type=MovieSceneAudioSection
      end_frame=4746
      start_frame=4546
      channels [
       Volume {type=MovieSceneScriptingFloatChannel, default=1}
       Pitch {type=MovieSceneScriptingFloatChannel, default=1}
      ]
     }
     {
      type=MovieSceneAudioSection
      end_frame=6102
      start_frame=4527
      channels [
       Volume {type=MovieSceneScriptingFloatChannel, default=1}
       Pitch {type=MovieSceneScriptingFloatChannel, default=1}
      ]
     }
     {
      type=MovieSceneAudioSection
      end_frame=5454
      start_frame=5254
      channels [
       Volume {type=MovieSceneScriptingFloatChannel, default=0.6}
       Pitch {type=MovieSceneScriptingFloatChannel, default=1}
      ]
     }
     {
      type=MovieSceneAudioSection
      end_frame=5123
      start_frame=4923
      channels [
       Volume {type=MovieSceneScriptingFloatChannel, default=1}
       Pitch {type=MovieSceneScriptingFloatChannel, default=1}
      ]
     }
     {
      type=MovieSceneAudioSection
      end_frame=9206
      start_frame=6102
      channels [
       Volume {type=MovieSceneScriptingFloatChannel, default=1}
       Pitch {type=MovieSceneScriptingFloatChannel, default=1}
      ]
     }
    ]
   }]
 }
 Tentacle_LimeRock_Creature_BG_2 {
  id=#17
  possessed_class=FortSkeletalMeshActor
  tracks [{type=MovieSceneAudioTrack, display_name=Audio, sections []}]
 }
 GoggleBud_Glider {
  id=#18
  possessed_class=FortSkeletalMeshActor
  tracks [{
    type=MovieSceneAudioTrack
    display_name=UFO Audio
    sections [
     {
      type=MovieSceneAudioSection
      end_frame=5085
      start_frame=4563
      channels [
       Volume {type=MovieSceneScriptingFloatChannel, default=0.7, keys [[4567,1],[4752.7225,1],[4766,0.7],[4823,0.6],[4842,1]]}
       Pitch {type=MovieSceneScriptingFloatChannel, default=1}
      ]
     }
     {
      type=MovieSceneAudioSection
      end_frame=6102
      start_frame=3601
      channels [
       Volume {type=MovieSceneScriptingFloatChannel, default=0.4, thinned_from=6, keys [[3601,0],[3619,1],[4994,1],[5263,0.4],[6102,0.6]]}
       Pitch {type=MovieSceneScriptingFloatChannel, default=1}
      ]
     }
     {
      type=MovieSceneAudioSection
      end_frame=4893
      start_frame=4793
      channels [
       Volume {type=MovieSceneScriptingFloatChannel, default=1}
       Pitch {type=MovieSceneScriptingFloatChannel, default=1}
      ]
     }
     {
      type=MovieSceneAudioSection
      end_frame=7861
      start_frame=7345
      channels [
       Volume {type=MovieSceneScriptingFloatChannel, default=0.5, keys [[7345,1],[7409,1],[7457.915,0.5],[7505,1]]}
       Pitch {type=MovieSceneScriptingFloatChannel, default=1}
      ]
     }
     {
      type=MovieSceneAudioSection
      end_frame=7967
      start_frame=6102
      channels [
       Volume {type=MovieSceneScriptingFloatChannel, default=0.6, keys [[6102,0.4],[7146,0.4],[7261,0.6],[7786,1],[7953,0]]}
       Pitch {type=MovieSceneScriptingFloatChannel, default=1}
      ]
     }
     {
      type=MovieSceneAudioSection
      end_frame=7295
      start_frame=7199
      channels [
       Volume {type=MovieSceneScriptingFloatChannel, default=1}
       Pitch {type=MovieSceneScriptingFloatChannel, default=1}
      ]
     }
     {
      type=MovieSceneAudioSection
      end_frame=4937
      start_frame=4662
      channels [
       Volume {type=MovieSceneScriptingFloatChannel, default=1}
       Pitch {type=MovieSceneScriptingFloatChannel, default=1}
      ]
     }
    ]
   }]
 }
 NS_LimeRock_EnergyBall {
  id=#19
  possessed_class=NiagaraActor
  tracks [{type=MovieSceneAudioTrack, display_name=RiftCracksInSky, sections []}]
 }
 Yogurt_Device_Gadget {
  id=#20
  possessed_class=FortSkeletalMeshActor
  tracks [{
    type=MovieSceneAudioTrack
    display_name=Audio
    sections [
     {
      type=MovieSceneAudioSection
      end_frame=6941
      start_frame=6663
      channels [
       Volume {type=MovieSceneScriptingFloatChannel, default=1}
       Pitch {type=MovieSceneScriptingFloatChannel, default=1}
      ]
     }
     {
      type=MovieSceneAudioSection
      end_frame=5375
      start_frame=5284
      channels [
       Volume {type=MovieSceneScriptingFloatChannel, default=1}
       Pitch {type=MovieSceneScriptingFloatChannel, default=1}
      ]
     }
     {
      type=MovieSceneAudioSection
      end_frame=6004
      start_frame=5781
      channels [
       Volume {type=MovieSceneScriptingFloatChannel, default=1}
       Pitch {type=MovieSceneScriptingFloatChannel, default=1}
      ]
     }
    ]
   }]
 }
 Hope {
  id=#21
  possessed_class=FortPlayerMannequin
  tracks [{
    type=MovieSceneAudioTrack
    display_name=HopeFoley
    sections [
     {
      type=MovieSceneAudioSection
      end_frame=4846
      start_frame=4095
      channels [
       Volume {type=MovieSceneScriptingFloatChannel, default=1}
       Pitch {type=MovieSceneScriptingFloatChannel, default=1}
      ]
     }
     {
      type=MovieSceneAudioSection
      end_frame=8989
      start_frame=8855
      channels [
       Volume {type=MovieSceneScriptingFloatChannel, default=1}
       Pitch {type=MovieSceneScriptingFloatChannel, default=1}
      ]
     }
     {
      type=MovieSceneAudioSection
      end_frame=9250
      start_frame=9085
      channels [
       Volume {type=MovieSceneScriptingFloatChannel, default=1}
       Pitch {type=MovieSceneScriptingFloatChannel, default=1}
      ]
     }
    ]
   }]
 }
]
display_rate {denominator=1, numerator=60}
tick_resolution {denominator=1, numerator=24000}
tracks [
 {
  type=MovieSceneAudioTrack
  display_name=Jones_VO
  sections [
   {
    type=MovieSceneAudioSection
    end_frame=6294
    start_frame=5690
    channels [
     Volume {type=MovieSceneScriptingFloatChannel, default=1}
     Pitch {type=MovieSceneScriptingFloatChannel, default=1}
     AttachActor {type=MovieSceneScriptingActorReferenceChannel}
    ]
   }
   {
    type=MovieSceneAudioSection
    end_frame=5256
    start_frame=4688
    channels [
     Volume {type=MovieSceneScriptingFloatChannel, default=1}
     Pitch {type=MovieSceneScriptingFloatChannel, default=1}
     AttachActor {type=MovieSceneScriptingActorReferenceChannel}
    ]
   }
   {
    type=MovieSceneAudioSection
    end_frame=5687
    start_frame=5263
    channels [
     Volume {type=MovieSceneScriptingFloatChannel, default=1}
     Pitch {type=MovieSceneScriptingFloatChannel, default=1}
     AttachActor {type=MovieSceneScriptingActorReferenceChannel}
    ]
   }
   {
    type=MovieSceneAudioSection
    end_frame=7206
    start_frame=6296
    channels [
     Volume {type=MovieSceneScriptingFloatChannel, default=1}
     Pitch {type=MovieSceneScriptingFloatChannel, default=1}
     AttachActor {type=MovieSceneScriptingActorReferenceChannel}
    ]
   }
   {
    type=MovieSceneAudioSection
    end_frame=8084
    start_frame=7207
    channels [
     Volume {type=MovieSceneScriptingFloatChannel, default=1}
     Pitch {type=MovieSceneScriptingFloatChannel, default=1}
     AttachActor {type=MovieSceneScriptingActorReferenceChannel}
    ]
   }
   {
    type=MovieSceneAudioSection
    end_frame=8710
    start_frame=8085
    channels [
     Volume {type=MovieSceneScriptingFloatChannel, default=1}
     Pitch {type=MovieSceneScriptingFloatChannel, default=1}
     AttachActor {type=MovieSceneScriptingActorReferenceChannel}
    ]
   }
   {
    type=MovieSceneAudioSection
    end_frame=9873
    start_frame=8711
    channels [
     Volume {type=MovieSceneScriptingFloatChannel, default=1}
     Pitch {type=MovieSceneScriptingFloatChannel, default=1}
     AttachActor {type=MovieSceneScriptingActorReferenceChannel}
    ]
   }
  ]
 }
 {
  type=MovieSceneAudioTrack
  display_name=Hope_VO
  sections [
   {
    type=MovieSceneAudioSection
    end_frame=4937
    start_frame=4491
    channels [
     Volume {type=MovieSceneScriptingFloatChannel, default=1}
     Pitch {type=MovieSceneScriptingFloatChannel, default=1}
     AttachActor {type=MovieSceneScriptingActorReferenceChannel}
    ]
   }
   {
    type=MovieSceneAudioSection
    end_frame=6100
    start_frame=5533
    channels [
     Volume {type=MovieSceneScriptingFloatChannel, default=1}
     Pitch {type=MovieSceneScriptingFloatChannel, default=1}
     AttachActor {type=MovieSceneScriptingActorReferenceChannel}
    ]
   }
   {
    type=MovieSceneAudioSection
    end_frame=7445
    start_frame=6102
    channels [
     Volume {type=MovieSceneScriptingFloatChannel, default=1}
     Pitch {type=MovieSceneScriptingFloatChannel, default=1}
     AttachActor {type=MovieSceneScriptingActorReferenceChannel}
    ]
   }
   {
    type=MovieSceneAudioSection
    end_frame=5531
    start_frame=4941
    channels [
     Volume {type=MovieSceneScriptingFloatChannel, default=1}
     Pitch {type=MovieSceneScriptingFloatChannel, default=1}
     AttachActor {type=MovieSceneScriptingActorReferenceChannel}
    ]
   }
   {
    type=MovieSceneAudioSection
    end_frame=9146
    start_frame=7446
    channels [
     Volume {type=MovieSceneScriptingFloatChannel, default=1}
     Pitch {type=MovieSceneScriptingFloatChannel, default=1}
     AttachActor {type=MovieSceneScriptingActorReferenceChannel}
    ]
   }
   {
    type=MovieSceneAudioSection
    end_frame=10268
    start_frame=9147
    channels [
     Volume {type=MovieSceneScriptingFloatChannel, default=1}
     Pitch {type=MovieSceneScriptingFloatChannel, default=1}
     AttachActor {type=MovieSceneScriptingActorReferenceChannel}
    ]
   }
  ]
 }
 {
  type=MovieSceneAudioTrack
  display_name=Flyby
  sections [{
    type=MovieSceneAudioSection
    end_frame=7788
    start_frame=7440
    channels [
     Volume {type=MovieSceneScriptingFloatChannel, default=1, keys [[7440,0.5],[7511,1]]}
     Pitch {type=MovieSceneScriptingFloatChannel, default=1}
     AttachActor {type=MovieSceneScriptingActorReferenceChannel}
    ]
   }]
 }
 {
  type=MovieSceneAudioTrack
  display_name=Animatic_V13
  sections [{
    type=MovieSceneAudioSection
    end_frame=9562
    start_frame=3273
    channels [
     Volume {type=MovieSceneScriptingFloatChannel, default=10}
     Pitch {type=MovieSceneScriptingFloatChannel, default=1}
     AttachActor {type=MovieSceneScriptingActorReferenceChannel, default="<Struct 'MovieSceneObjectBindingID' (0x850fe3b3f90) {}>"}
    ]
   }]
 }
 {
  type=MovieSceneAudioTrack
  display_name=Animatic_V15
  sections [{
    type=MovieSceneAudioSection
    end_frame=10017
    start_frame=3273
    channels [
     Volume {type=MovieSceneScriptingFloatChannel, default=40}
     Pitch {type=MovieSceneScriptingFloatChannel, default=1}
     AttachActor {type=MovieSceneScriptingActorReferenceChannel, default="<Struct 'MovieSceneObjectBindingID' (0x851351b42c0) {}>"}
    ]
   }]
 }
 {
  type=MovieSceneAudioTrack
  display_name=Jones Foley Footsteps
  sections [
   {
    type=MovieSceneAudioSection
    end_frame=5493
    start_frame=4436
    channels [
     Volume {type=MovieSceneScriptingFloatChannel, default=1}
     Pitch {type=MovieSceneScriptingFloatChannel, default=1}
     AttachActor {type=MovieSceneScriptingActorReferenceChannel}
    ]
   }
   {
    type=MovieSceneAudioSection
    end_frame=5790
    start_frame=5658
    channels [
     Volume {type=MovieSceneScriptingFloatChannel, default=1}
     Pitch {type=MovieSceneScriptingFloatChannel, default=1}
     AttachActor {type=MovieSceneScriptingActorReferenceChannel}
    ]
   }
   {
    type=MovieSceneAudioSection
    end_frame=6694
    start_frame=6179
    channels [
     Volume {type=MovieSceneScriptingFloatChannel, default=1}
     Pitch {type=MovieSceneScriptingFloatChannel, default=1}
     AttachActor {type=MovieSceneScriptingActorReferenceChannel}
    ]
   }
   {
    type=MovieSceneAudioSection
    end_frame=7335
    start_frame=7154
    channels [
     Volume {type=MovieSceneScriptingFloatChannel, default=1}
     Pitch {type=MovieSceneScriptingFloatChannel, default=1}
     AttachActor {type=MovieSceneScriptingActorReferenceChannel}
    ]
   }
   {
    type=MovieSceneAudioSection
    end_frame=8227
    start_frame=7835
    channels [
     Volume {type=MovieSceneScriptingFloatChannel, default=1}
     Pitch {type=MovieSceneScriptingFloatChannel, default=1}
     AttachActor {type=MovieSceneScriptingActorReferenceChannel}
    ]
   }
   {
    type=MovieSceneAudioSection
    end_frame=8605
    start_frame=8465
    channels [
     Volume {type=MovieSceneScriptingFloatChannel, default=1}
     Pitch {type=MovieSceneScriptingFloatChannel, default=1}
     AttachActor {type=MovieSceneScriptingActorReferenceChannel}
    ]
   }
  ]
 }
 {
  type=MovieSceneAudioTrack
  display_name=Audio
  sections [
   {
    type=MovieSceneAudioSection
    end_frame=8633
    start_frame=8379
    channels [
     Volume {type=MovieSceneScriptingFloatChannel, default=1}
     Pitch {type=MovieSceneScriptingFloatChannel, default=1}
     AttachActor {type=MovieSceneScriptingActorReferenceChannel}
    ]
   }
   {
    type=MovieSceneAudioSection
    end_frame=8820
    start_frame=8488
    channels [
     Volume {type=MovieSceneScriptingFloatChannel, default=1}
     Pitch {type=MovieSceneScriptingFloatChannel, default=1}
     AttachActor {type=MovieSceneScriptingActorReferenceChannel}
    ]
   }
  ]
 }
]
