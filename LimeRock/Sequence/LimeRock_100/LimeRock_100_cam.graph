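# Camera Level Sequence for LimeRock shot 100, apparently a live-event cinematic with rifts, a rising monster and a final slam.
# Two camera cuts line up with keyed moves and focal-length changes on CineCam A, then B, which layer extra transform sections; B's move coincides with a fade-out.
# A camera-force Blueprint switches shake and force-feedback assets on event beats, with matching lobby event triggers.
# Jones and Hope face cameras use attach tracks.
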LevelSequence /LimeRock/Sequence/LimeRock_100/LimeRock_100_cam
bindings [
 CineCam_B {
  id=#1
  possessed_class=CineCameraActor
  tracks [{
    type=MovieScene3DTransformTrack
    sections [
     {
      type=MovieScene3DTransformSection
      channels [
       Location.X {type=MovieSceneScriptingDoubleChannel, default=4868, keys [[9049,4833],[9209,4868]]}
       Location.Y {type=MovieSceneScriptingDoubleChannel, default=-680, keys [[9049,-676.2614],[9209,-680]]}
       Location.Z {type=MovieSceneScriptingDoubleChannel, default=-387, keys [[9049,-366],[9209,-387]]}
       Rotation.X {type=MovieSceneScriptingDoubleChannel, default=0, const [0,1]}
       Rotation.Y {type=MovieSceneScriptingDoubleChannel, default=36, keys [[9049,36],[9209,32.0849]]}
       Rotation.Z {type=MovieSceneScriptingDoubleChannel, default=-174.4076, const [-174.4077,1]}
       Scale.X {type=MovieSceneScriptingDoubleChannel, default=1, const [1,1]}
       Scale.Y {type=MovieSceneScriptingDoubleChannel, default=1, const [1,1]}
       Scale.Z {type=MovieSceneScriptingDoubleChannel, default=1, const [1,1]}
      ]
     }
     {
      type=MovieScene3DTransformSection
      channels [
       Location.X {type=MovieSceneScriptingDoubleChannel, default=-428, const [-428,1]}
       Location.Y {type=MovieSceneScriptingDoubleChannel, default=-471, const [-471,1]}
       Location.Z {type=MovieSceneScriptingDoubleChannel, default=0, const [0,1]}
       Rotation.X {type=MovieSceneScriptingDoubleChannel, default=0, const [0,1]}
       Rotation.Y {type=MovieSceneScriptingDoubleChannel, default=0, const [1.2,1]}
       Rotation.Z {type=MovieSceneScriptingDoubleChannel, default=5, const [5.6,1]}
       Scale.X {type=MovieSceneScriptingDoubleChannel, default=0, const [0,1]}
       Scale.Y {type=MovieSceneScriptingDoubleChannel, default=0, const [0,1]}
       Scale.Z {type=MovieSceneScriptingDoubleChannel, default=0, const [0,1]}
      ]
     }
     {
      type=MovieScene3DTransformSection
      end_frame=14400
      start_frame=9125
      channels [
       Location.X {type=MovieSceneScriptingDoubleChannel, default=0}
       Location.Y {type=MovieSceneScriptingDoubleChannel, default=0}
       Location.Z {type=MovieSceneScriptingDoubleChannel, default=0}
       Rotation.Z {type=MovieSceneScriptingDoubleChannel, default=0}
       Scale.X {type=MovieSceneScriptingDoubleChannel, default=0}
       Scale.Y {type=MovieSceneScriptingDoubleChannel, default=0}
       Scale.Z {type=MovieSceneScriptingDoubleChannel, default=0}
      ]
     }
     {
      type=MovieScene3DTransformSection
      end_frame=9209
      start_frame=3600
      channels [
       Location.X {type=MovieSceneScriptingDoubleChannel, default=0}
       Location.Y {type=MovieSceneScriptingDoubleChannel, default=0}
       Location.Z {type=MovieSceneScriptingDoubleChannel, default=0}
       Rotation.Z {type=MovieSceneScriptingDoubleChannel, default=0}
       Scale.X {type=MovieSceneScriptingDoubleChannel, default=0}
       Scale.Y {type=MovieSceneScriptingDoubleChannel, default=0}
       Scale.Z {type=MovieSceneScriptingDoubleChannel, default=0}
      ]
     }
    ]
   }]
 }
 CineCam_A {
  id=#2
  possessed_class=CineCameraActor
  tracks [{
    type=MovieScene3DTransformTrack
    sections [
     {
      type=MovieScene3DTransformSection
      channels [
       Location.X {type=MovieSceneScriptingDoubleChannel, default=3850, keys [[8905,3993.1279],[8949,3909.7915],[9049,3823.4956]]}
       Location.Y {type=MovieSceneScriptingDoubleChannel, default=-1169.1966, keys [[8905,-745.7999],[9049,-745.7999]]}
       Location.Z {type=MovieSceneScriptingDoubleChannel, default=347, keys [[8905,569.3437],[9049,1456.755]]}
       Rotation.X {type=MovieSceneScriptingDoubleChannel, default=0, const [0,1]}
       Rotation.Y {type=MovieSceneScriptingDoubleChannel, default=-65, keys [[8905,-56.978],[9049,-66.6771]]}
       Rotation.Z {type=MovieSceneScriptingDoubleChannel, default=-170, const [-357.8076,1]}
       Scale.X {type=MovieSceneScriptingDoubleChannel, default=1}
       Scale.Y {type=MovieSceneScriptingDoubleChannel, default=1}
       Scale.Z {type=MovieSceneScriptingDoubleChannel, default=1}
      ]
     }
     {
      type=MovieScene3DTransformSection
      channels [
       Location.X {type=MovieSceneScriptingDoubleChannel, default=-423, const [-423,1]}
       Location.Y {type=MovieSceneScriptingDoubleChannel, default=-494, const [-494,1]}
       Location.Z {type=MovieSceneScriptingDoubleChannel, default=0, const [0,1]}
       Rotation.X {type=MovieSceneScriptingDoubleChannel, default=0, const [0,1]}
       Rotation.Y {type=MovieSceneScriptingDoubleChannel, default=0, const [0,1]}
       Rotation.Z {type=MovieSceneScriptingDoubleChannel, default=0, const [0,1]}
       Scale.X {type=MovieSceneScriptingDoubleChannel, default=0, const [0,1]}
       Scale.Y {type=MovieSceneScriptingDoubleChannel, default=0, const [0,1]}
       Scale.Z {type=MovieSceneScriptingDoubleChannel, default=0, const [0,1]}
      ]
     }
     {
      type=MovieScene3DTransformSection
      end_frame=14400
      start_frame=3600
      channels [
       Location.X {type=MovieSceneScriptingDoubleChannel, default=0}
       Location.Y {type=MovieSceneScriptingDoubleChannel, default=0}
       Location.Z {type=MovieSceneScriptingDoubleChannel, default=0}
       Rotation.Z {type=MovieSceneScriptingDoubleChannel, default=0}
       Scale.X {type=MovieSceneScriptingDoubleChannel, default=0}
       Scale.Y {type=MovieSceneScriptingDoubleChannel, default=0}
       Scale.Z {type=MovieSceneScriptingDoubleChannel, default=0}
      ]
     }
    ]
   }]
 }
 Events_Shared_cameraForce_BP {
  id=#3
  possessed_class=Events_Shared_cameraForce_BP_C
  tracks [
   {type=MovieSceneSpawnTrack, sections [{type=MovieSceneSpawnSection, channels [None {type=MovieSceneScriptingBoolChannel, default=true}]}]}
   {
    type=MovieScene3DTransformTrack
    sections [{
      type=MovieScene3DTransformSection
      channels [
       Location.X {type=MovieSceneScriptingDoubleChannel, default=0}
       Location.Y {type=MovieSceneScriptingDoubleChannel, default=0}
       Location.Z {type=MovieSceneScriptingDoubleChannel, default=0}
       Rotation.X {type=MovieSceneScriptingDoubleChannel, default=0}
       Rotation.Y {type=MovieSceneScriptingDoubleChannel, default=0}
       Rotation.Z {type=MovieSceneScriptingDoubleChannel, default=0}
       Scale.X {type=MovieSceneScriptingDoubleChannel, default=1}
       Scale.Y {type=MovieSceneScriptingDoubleChannel, default=1}
       Scale.Z {type=MovieSceneScriptingDoubleChannel, default=1}
      ]
     }]
   }
   {
    type=MovieSceneObjectPropertyTrack
    display_name=Cam Shake Base
    sections [{
      type=MovieSceneObjectPropertySection
      channels [None {
        type=MovieSceneScriptingObjectPathChannel
        default="<Object '&Events_Shared_00small_csb_C' (0x84c3c6a8a00) Class 'BlueprintGeneratedClass'>"
        const [
         "<Object '&Events_Shared_00small_csb_C' (0x84c3c6a8a00) Class 'BlueprintGeneratedClass'>"
         1
        ]
       }]
     }]
   }
  ]
 }
 CameraComponent {
  id=#4
  parent_id=#1
  possessed_class=CineCameraComponent
  tracks [
   {
    type=MovieSceneFloatTrack
    display_name=Current Focal Length
    sections [{type=MovieSceneFloatSection, channels [None {type=MovieSceneScriptingFloatChannel, default=16, keys [[9049,16],[9209,14]]}]}]
   }
   {
    type=MovieSceneFloatTrack
    display_name="Manual Focus Distance (Focus Settings)"
    sections [{type=MovieSceneFloatSection, channels [None {type=MovieSceneScriptingFloatChannel, default=100000}]}]
   }
   {
    type=MovieSceneFloatTrack
    display_name=Current Aperture
    sections [{type=MovieSceneFloatSection, channels [None {type=MovieSceneScriptingFloatChannel, default=2.8}]}]
   }
  ]
 }
 FaceCam_Jones {
  id=#5
  possessed_class=CineCameraActor
  tracks [
   {
    type=MovieScene3DTransformTrack
    sections [{
      type=MovieScene3DTransformSection
      channels [
       Location.X {type=MovieSceneScriptingDoubleChannel, default=-26}
       Location.Y {type=MovieSceneScriptingDoubleChannel, default=147}
       Location.Z {type=MovieSceneScriptingDoubleChannel, default=-17}
       Rotation.X {type=MovieSceneScriptingDoubleChannel, default=-90}
       Rotation.Y {type=MovieSceneScriptingDoubleChannel, default=-186}
       Rotation.Z {type=MovieSceneScriptingDoubleChannel, default=100}
       Scale.X {type=MovieSceneScriptingDoubleChannel, default=1}
       Scale.Y {type=MovieSceneScriptingDoubleChannel, default=1}
       Scale.Z {type=MovieSceneScriptingDoubleChannel, default=1}
      ]
     }]
   }
   {type=MovieSceneSpawnTrack, sections [{type=MovieSceneSpawnSection, channels [None {type=MovieSceneScriptingBoolChannel, default=true}]}]}
   {type=MovieScene3DAttachTrack, sections [{type=MovieScene3DAttachSection, end_frame=14400, start_frame=3600}]}
  ]
 }
 CameraComponent {
  id=#6
  parent_id=#7
  possessed_class=CineCameraComponent
  tracks [
   {
    type=MovieSceneFloatTrack
    display_name=Current Focal Length
    sections [{type=MovieSceneFloatSection, channels [None {type=MovieSceneScriptingFloatChannel, default=50}]}]
   }
   {
    type=MovieSceneFloatTrack
    display_name="Manual Focus Distance (Focus Settings)"
    sections [{type=MovieSceneFloatSection, channels [None {type=MovieSceneScriptingFloatChannel, default=100000}]}]
   }
   {
    type=MovieSceneFloatTrack
    display_name=Current Aperture
    sections [{type=MovieSceneFloatSection, channels [None {type=MovieSceneScriptingFloatChannel, default=11}]}]
   }
  ]
 }
 CameraComponent {
  id=#8
  parent_id=#2
  possessed_class=CineCameraComponent
  tracks [
   {
    type=MovieSceneFloatTrack
    display_name=Current Focal Length
    sections [{type=MovieSceneFloatSection, channels [None {type=MovieSceneScriptingFloatChannel, default=35, keys [[8905,35],[9049,29.5845]]}]}]
   }
   {
    type=MovieSceneFloatTrack
    display_name="Manual Focus Distance (Focus Settings)"
    sections [{type=MovieSceneFloatSection, channels [None {type=MovieSceneScriptingFloatChannel, default=100000}]}]
   }
   {
    type=MovieSceneFloatTrack
    display_name=Current Aperture
    sections [{type=MovieSceneFloatSection, channels [None {type=MovieSceneScriptingFloatChannel, default=2.8}]}]
   }
  ]
 }
 LimeRock_cameraForce_BP {
  id=#9
  possessed_class=Events_Shared_cameraForce_BP_C
  tracks [
   {
    type=MovieSceneObjectPropertyTrack
    display_name=Force Effect
    sections [{
      type=MovieSceneObjectPropertySection
      channels [None {
        type=MovieSceneScriptingObjectPathChannel
        default="<Object '/LimeRock/ForceFeedback/FF_LimeRock_RiftOpen_01' (0x850b8173080) Class 'ForceFeedbackEffect'>"
        keys [
         [2504, "<Object '&Events_Shared_small_FF' (0x84c3c703480) Class 'ForceFeedbackEffect'>"]
         [4084, "<Object '&FF_LimeRock_RiftOpen_01' (0x850b8173080) Class 'ForceFeedbackEffect'>"]
         [4200, "<Object '/LimeRock/ForceFeedback/FF_LimeRock_RiftOpen_02' (0x850b8173180) Class 'ForceFeedbackEffect'>"]
         [4649, "<Object '/LimeRock/ForceFeedback/FF_LimeRock_Flyby_01' (0x850b8172880) Class 'ForceFeedbackEffect'>"]
         [4987, "<Object '&FF_LimeRock_RiftOpen_01' (0x850b8173080) Class 'ForceFeedbackEffect'>"]
         [5905, "<Object '/LimeRock/ForceFeedback/FF_LimeRock_TreePulse_01' (0x850b8173380) Class 'ForceFeedbackEffect'>"]
         [6694, "<Object '/LimeRock/ForceFeedback/FF_LimeRock_Beam' (0x850b8172e80) Class 'ForceFeedbackEffect'>"]
         [7078, "<Object '/LimeRock/ForceFeedback/FF_LimeRock_TreePulse_02' (0x850b8173300) Class 'ForceFeedbackEffect'>"]
         [7455, "<Object '/LimeRock/ForceFeedback/FF_LimeRock_Flyby_02' (0x850b8172c00) Class 'ForceFeedbackEffect'>"]
         [7661, "<Object '/LimeRock/ForceFeedback/FF_LimeRock_PortalOpen' (0x850b8173000) Class 'ForceFeedbackEffect'>"]
         [7751, "<Object '/LimeRock/ForceFeedback/FF_LimeRock_MonsterRising_01' (0x850b8173100) Class 'ForceFeedbackEffect'>"]
         [8557, "<Object '/LimeRock/ForceFeedback/FF_LimeRock_Roar' (0x850b8173280) Class 'ForceFeedbackEffect'>"]
         [8799, "<Object '/LimeRock/ForceFeedback/FF_LimeRock_MonsterRising_02' (0x850b8173200) Class 'ForceFeedbackEffect'>"]
         [9209, "<Object '/LimeRock/ForceFeedback/FF_LimeRock_Slam' (0x850b8173400) Class 'ForceFeedbackEffect'>"]
        ]
       }]
     }]
   }
   {
    type=MovieSceneObjectPropertyTrack
    display_name=Cam Shake Base
    sections [{
      type=MovieSceneObjectPropertySection
      channels [None {
        type=MovieSceneScriptingObjectPathChannel
        default="<Object '&Events_Shared_00small_csb_C' (0x84c3c6a8a00) Class 'BlueprintGeneratedClass'>"
        keys [
         [
          2504
          "<Object '&Events_Shared_00small_csb_C' (0x84c3c6a8a00) Class 'BlueprintGeneratedClass'>"
         ]
         [4649, "<Object '/LimeRock/Camera/CameraShakes/LimeRock_Hoverby_bp.LimeRock_Hoverby_bp_C' (0x850a44fae00) Class 'BlueprintGeneratedClass'>"]
         [
          5905
          "<Object '/LimeRock/Camera/CameraShakes/LimeRock_HoverPulse_csb.LimeRock_HoverPulse_csb_C' (0x850a44fa200) Class 'BlueprintGeneratedClass'>"
         ]
         [6694, "<Object '/LimeRock/Camera/CameraShakes/LimeRock_RiftGun_bp.LimeRock_RiftGun_bp_C' (0x850a44fd200) Class 'BlueprintGeneratedClass'>"]
         [
          7078
          "<Object '&LimeRock_HoverPulse_csb_C' (0x850a44fa200) Class 'BlueprintGeneratedClass'>"
         ]
         [7455, "<Object '/LimeRock/Camera/CameraShakes/LimeRock_Flyby_bp.LimeRock_Flyby_bp_C' (0x850a44fba00) Class 'BlueprintGeneratedClass'>"]
         [
          8557
          "<Object '/LimeRock/Camera/CameraShakes/LimeRock_MonsterRoar_csb.LimeRock_MonsterRoar_csb_C' (0x850a44f9600) Class 'BlueprintGeneratedClass'>"
         ]
         [
          10589
          "<Object '/LimeRock/Camera/CameraShakes/LimeRock_PortalOpen_bp.LimeRock_PortalOpen_bp_C' (0x850a44fc600) Class 'BlueprintGeneratedClass'>"
         ]
        ]
       }]
     }]
   }
   {
    type=MovieSceneEventTrack
    display_name=Cam Shake Radius Event Trigger
    sections [{
      type=MovieSceneEventTriggerSection
      channels [None {type=MovieSceneScriptingEventChannel, const ["<Struct 'MovieSceneEvent' (0x851351ce920) {}>", 1]}]
     }]
   }
   {type=MovieSceneSpawnTrack, sections [{type=MovieSceneSpawnSection, channels [None {type=MovieSceneScriptingBoolChannel, default=true}]}]}
   {
    type=MovieSceneEventTrack
    display_name=Force Feedback Radius Event Trigger
    sections [{
      type=MovieSceneEventTriggerSection
      channels [None {type=MovieSceneScriptingEventChannel, const ["<Struct 'MovieSceneEvent' (0x851351ce420) {}>", 1]}]
     }]
   }
   {
    type=MovieScene3DTransformTrack
    sections [{
      type=MovieScene3DTransformSection
      channels [
       Location.X {type=MovieSceneScriptingDoubleChannel, default=7084}
       Location.Y {type=MovieSceneScriptingDoubleChannel, default=1572}
       Location.Z {type=MovieSceneScriptingDoubleChannel, default=0}
       Rotation.X {type=MovieSceneScriptingDoubleChannel, default=0}
       Rotation.Y {type=MovieSceneScriptingDoubleChannel, default=0}
       Rotation.Z {type=MovieSceneScriptingDoubleChannel, default=0}
       Scale.X {type=MovieSceneScriptingDoubleChannel, default=1}
       Scale.Y {type=MovieSceneScriptingDoubleChannel, default=1}
       Scale.Z {type=MovieSceneScriptingDoubleChannel, default=1}
      ]
     }]
   }
   {
    type=MovieSceneIntegerTrack
    sections [{type=MovieSceneIntegerSection, channels [None {type=MovieSceneScriptingIntegerChannel, default=2, const [0,1]}]}]
   }
   {
    type=MovieSceneEventTrack
    display_name=Cam Shake Lobby Event Trigger
    sections [{
      type=MovieSceneEventTriggerSection
      channels [None {
        type=MovieSceneScriptingEventChannel
        keys [
         [2504, "<Struct 'MovieSceneEvent' (0x851351cdc00) {}>"]
         [4649, "<Struct 'MovieSceneEvent' (0x851351cdca0) {}>"]
         [5905, "<Struct 'MovieSceneEvent' (0x851351ce380) {}>"]
         [6694, "<Struct 'MovieSceneEvent' (0x851351cdfc0) {}>"]
         [7078, "<Struct 'MovieSceneEvent' (0x851351ce4c0) {}>"]
         [7455, "<Struct 'MovieSceneEvent' (0x851351ce560) {}>"]
         [8557, "<Struct 'MovieSceneEvent' (0x851351ce880) {}>"]
         [10589, "<Struct 'MovieSceneEvent' (0x851351cea60) {}>"]
        ]
       }]
     }]
   }
   {
    type=MovieSceneEventTrack
    display_name=Force Feedback Lobby Event Trigger
    sections [{
      type=MovieSceneEventTriggerSection
      channels [None {
        type=MovieSceneScriptingEventChannel
        keys [
         [2504, "<Struct 'MovieSceneEvent' (0x851351cef60) {}>"]
         [4084, "<Struct 'MovieSceneEvent' (0x851351cf000) {}>"]
         [4200, "<Struct 'MovieSceneEvent' (0x851351cf0a0) {}>"]
         [4649, "<Struct 'MovieSceneEvent' (0x851351cf140) {}>"]
         [4987, "<Struct 'MovieSceneEvent' (0x851351cf1e0) {}>"]
         [5905, "<Struct 'MovieSceneEvent' (0x851351cf280) {}>"]
         [6694, "<Struct 'MovieSceneEvent' (0x851351cf320) {}>"]
         [7078, "<Struct 'MovieSceneEvent' (0x851351cf3c0) {}>"]
         [7455, "<Struct 'MovieSceneEvent' (0x851351cf460) {}>"]
         [7661, "<Struct 'MovieSceneEvent' (0x851351cf500) {}>"]
         [7751, "<Struct 'MovieSceneEvent' (0x851351cf5a0) {}>"]
         [8557, "<Struct 'MovieSceneEvent' (0x851351cf640) {}>"]
         [8799, "<Struct 'MovieSceneEvent' (0x851351cf6e0) {}>"]
         [9209, "<Struct 'MovieSceneEvent' (0x851351cf780) {}>"]
        ]
       }]
     }]
   }
  ]
 }
 CineCameraActor2 {
  id=#10
  possessed_class=CineCameraActor
  tracks [
   {
    type=MovieScene3DTransformTrack
    sections [{
      type=MovieScene3DTransformSection
      channels [
       Location.X {type=MovieSceneScriptingDoubleChannel, default=-89945.1734, const [-89381.1094,1]}
       Location.Y {type=MovieSceneScriptingDoubleChannel, default=-94193.9972, const [-95357.75,1]}
       Location.Z {type=MovieSceneScriptingDoubleChannel, default=3781.8924, const [1278.05,1]}
       Rotation.X {type=MovieSceneScriptingDoubleChannel, default=0, const [0,1]}
       Rotation.Y {type=MovieSceneScriptingDoubleChannel, default=-49.4682, const [5.3318,1]}
       Rotation.Z {type=MovieSceneScriptingDoubleChannel, default=-184.0228, const [179.9772,1]}
       Scale.X {type=MovieSceneScriptingDoubleChannel, default=1, const [1,1]}
       Scale.Y {type=MovieSceneScriptingDoubleChannel, default=1, const [1,1]}
       Scale.Z {type=MovieSceneScriptingDoubleChannel, default=1, const [1,1]}
      ]
     }]
   }
   {type=MovieSceneSpawnTrack, sections [{type=MovieSceneSpawnSection, channels [None {type=MovieSceneScriptingBoolChannel, default=true}]}]}
  ]
 }
 Tentacle_LimeRock_Creature2 {
  id=#11
  possessed_class=FortSkeletalMeshActor
  tracks [
   {
    type=MovieScene3DTransformTrack
    sections [{
      type=MovieScene3DTransformSection
      channels [
       Location.X {type=MovieSceneScriptingDoubleChannel, default=-8008.1616}
       Location.Y {type=MovieSceneScriptingDoubleChannel, default=-5847.4937}
       Location.Z {type=MovieSceneScriptingDoubleChannel, default=918.2955}
       Rotation.X {type=MovieSceneScriptingDoubleChannel, default=42.0145}
       Rotation.Y {type=MovieSceneScriptingDoubleChannel, default=-3.1318}
       Rotation.Z {type=MovieSceneScriptingDoubleChannel, default=-108.1797}
       Scale.X {type=MovieSceneScriptingDoubleChannel, default=1}
       Scale.Y {type=MovieSceneScriptingDoubleChannel, default=1}
       Scale.Z {type=MovieSceneScriptingDoubleChannel, default=1}
      ]
     }]
   }
   {type=MovieSceneSkeletalAnimationTrack, sections []}
  ]
 }
 FaceCam_Hope {
  id=#7
  possessed_class=CineCameraActor
  tracks [
   {
    type=MovieScene3DTransformTrack
    sections [{
      type=MovieScene3DTransformSection
      channels [
       Location.X {type=MovieSceneScriptingDoubleChannel, default=-26}
       Location.Y {type=MovieSceneScriptingDoubleChannel, default=147}
       Location.Z {type=MovieSceneScriptingDoubleChannel, default=-17}
       Rotation.X {type=MovieSceneScriptingDoubleChannel, default=-90}
       Rotation.Y {type=MovieSceneScriptingDoubleChannel, default=-186}
       Rotation.Z {type=MovieSceneScriptingDoubleChannel, default=100}
       Scale.X {type=MovieSceneScriptingDoubleChannel, default=1}
       Scale.Y {type=MovieSceneScriptingDoubleChannel, default=1}
       Scale.Z {type=MovieSceneScriptingDoubleChannel, default=1}
      ]
     }]
   }
   {type=MovieSceneSpawnTrack, sections [{type=MovieSceneSpawnSection, channels [None {type=MovieSceneScriptingBoolChannel, default=true}]}]}
   {type=MovieScene3DAttachTrack, sections [{type=MovieScene3DAttachSection, end_frame=14400, start_frame=3600}]}
  ]
 }
 CameraComponent {
  id=#12
  parent_id=#5
  possessed_class=CineCameraComponent
  tracks [
   {
    type=MovieSceneFloatTrack
    display_name=Current Focal Length
    sections [{type=MovieSceneFloatSection, channels [None {type=MovieSceneScriptingFloatChannel, default=50}]}]
   }
   {
    type=MovieSceneFloatTrack
    display_name="Manual Focus Distance (Focus Settings)"
    sections [{type=MovieSceneFloatSection, channels [None {type=MovieSceneScriptingFloatChannel, default=100000}]}]
   }
   {
    type=MovieSceneFloatTrack
    display_name=Current Aperture
    sections [{type=MovieSceneFloatSection, channels [None {type=MovieSceneScriptingFloatChannel, default=11}]}]
   }
  ]
 }
 CameraComponent {
  id=#13
  parent_id=#10
  possessed_class=CineCameraComponent
  tracks [
   {
    type=MovieSceneFloatTrack
    display_name=Current Focal Length
    sections [{type=MovieSceneFloatSection, channels [None {type=MovieSceneScriptingFloatChannel, default=14}]}]
   }
   {
    type=MovieSceneFloatTrack
    display_name="Manual Focus Distance (Focus Settings)"
    sections [{type=MovieSceneFloatSection, channels [None {type=MovieSceneScriptingFloatChannel, default=100000}]}]
   }
   {
    type=MovieSceneFloatTrack
    display_name=Current Aperture
    sections [{type=MovieSceneFloatSection, channels [None {type=MovieSceneScriptingFloatChannel, default=2.8}]}]
   }
  ]
 }
]
display_rate {denominator=1, numerator=60}
tick_resolution {denominator=1, numerator=24000}
tracks [
 {
  type=MovieSceneFadeTrack
  sections [{type=MovieSceneFadeSection, channels [None {type=MovieSceneScriptingFloatChannel, default=0, keys [[9049,0],[9209,1]]}]}]
 }
 {
  type=MovieSceneCameraCutTrack
  sections [{type=MovieSceneCameraCutSection, end_frame=9049, start_frame=8905}, {type=MovieSceneCameraCutSection, end_frame=12302, start_frame=9049}]
 }
]
