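# LimeRock_100_gam is a Level Sequence that carries gameplay triggers for the LimeRock event.
# Event tracks fire timed calls on the sequence runner, including a separate UFO Flyby track, and on an attached focus point and a knockback volume.
# At the opening beat a fake countdown timer shrinks and hides, while a tentacle creature holds a fixed transform with an empty animation track.
# A data layer section spans the sequence.
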LevelSequence /LimeRock/Sequence/LimeRock_100/LimeRock_100_gam
bindings [
 BP_LimeRock_KnockbackVolume {
  id=#1
  possessed_class=BP_LimeRock_KnockbackVolume_C
  tracks [{
    type=MovieSceneEventTrack
    sections [{
      type=MovieSceneEventTriggerSection
      channels [None {type=MovieSceneScriptingEventChannel, const ["<Struct 'MovieSceneEvent' (0x851351cf5a0) {}>", 1]}]
     }]
   }]
 }
 NS_DeckBed_TimerBase {id=#2, parent_id=#3, possessed_class=NiagaraComponent, tracks []}
 BP_LimeRock_SequenceRunner {
  id=#4
  possessed_class=BP_LimeRock_SequenceRunner_C
  tracks [
   {
    type=MovieSceneEventTrack
    display_name=Event Track
    sections [{
      type=MovieSceneEventTriggerSection
      channels [None {
        type=MovieSceneScriptingEventChannel
        keys [
         [3600, "<Struct 'MovieSceneEvent' (0x851351cf280) {}>"]
         [5922, "<Struct 'MovieSceneEvent' (0x851351cf1e0) {}>"]
         [7098, "<Struct 'MovieSceneEvent' (0x851351cf140) {}>"]
         [7459, "<Struct 'MovieSceneEvent' (0x851351cf0a0) {}>"]
         [8582, "<Struct 'MovieSceneEvent' (0x851351cf000) {}>"]
         [8892, "<Struct 'MovieSceneEvent' (0x851351cef60) {}>"]
         [9322, "<Struct 'MovieSceneEvent' (0x851351ce880) {}>"]
        ]
       }]
     }]
   }
   {
    type=MovieSceneEventTrack
    display_name=UFO Flyby
    sections [{
      type=MovieSceneEventTriggerSection
      channels [None {
        type=MovieSceneScriptingEventChannel
        keys [
         [3600, "<Struct 'MovieSceneEvent' (0x851351cdc00) {}>"]
         [7426, "<Struct 'MovieSceneEvent' (0x851351ce420) {}>"]
         [7601, "<Struct 'MovieSceneEvent' (0x851351ce920) {}>"]
        ]
       }]
     }]
   }
  ]
 }
 Tentacle_LimeRock_Creature2 {
  id=#5
  possessed_class=FortSkeletalMeshActor
  tracks [
   {
    type=MovieScene3DTransformTrack
    sections [{
      type=MovieScene3DTransformSection
      channels [
       Location.X {type=MovieSceneScriptingDoubleChannel, default=-8008.1616}
       Location.Y {type=MovieSceneScriptingDoubleChannel, default=-5847.4937}
       Location.Z {type=MovieSceneScriptingDoubleChannel, default=918.2955}
       Rotation.X {type=MovieSceneScriptingDoubleChannel, default=42.0145}
       Rotation.Y {type=MovieSceneScriptingDoubleChannel, default=-3.1318}
       Rotation.Z {type=MovieSceneScriptingDoubleChannel, default=-108.1797}
       Scale.X {type=MovieSceneScriptingDoubleChannel, default=1}
       Scale.Y {type=MovieSceneScriptingDoubleChannel, default=1}
       Scale.Z {type=MovieSceneScriptingDoubleChannel, default=1}
      ]
     }]
   }
   {type=MovieSceneSkeletalAnimationTrack, sections []}
  ]
 }
 BP_EventsShared_FocusPoint {
  id=#6
  possessed_class=BP_EventsShared_FocusPoint_C
  tracks [
   {type=MovieScene3DAttachTrack, sections [{type=MovieScene3DAttachSection, end_frame=14400, start_frame=3600}]}
   {
    type=MovieSceneEventTrack
    sections [{
      type=MovieSceneEventTriggerSection
      channels [None {
        type=MovieSceneScriptingEventChannel
        keys [[4072, "<Struct 'MovieSceneEvent' (0x851351cdca0) {}>"], [8712, "<Struct 'MovieSceneEvent' (0x851351ce380) {}>"]]
       }]
     }]
   }
  ]
 }
 BP_LimeRock_CountdownTimer_Fake {
  id=#3
  possessed_class=BP_Generic_Game_Countdown_C
  tracks [
   {
    type=MovieScene3DTransformTrack
    sections [{
      type=MovieScene3DTransformSection
      channels [
       Scale.X {type=MovieSceneScriptingDoubleChannel, default=0.2, keys [[3600,0.2],[3618,0.01]]}
       Scale.Y {type=MovieSceneScriptingDoubleChannel, default=0.2, keys [[3600,0.2],[3607,0.2],[3622,0.01]]}
       Scale.Z {type=MovieSceneScriptingDoubleChannel, default=0.2, keys [[3600,0.2],[3607,0.2],[3622,0.01]]}
      ]
     }]
   }
   {
    type=MovieSceneVisibilityTrack
    sections [{type=MovieSceneVisibilitySection, channels [None {type=MovieSceneScriptingBoolChannel, default=true, keys [[3600,true],[3621,false]]}]}]
   }
  ]
 }
]
display_rate {denominator=1, numerator=60}
tick_resolution {denominator=1, numerator=24000}
tracks [{type=MovieSceneDataLayerTrack, display_name=Data Layer, sections [{type=MovieSceneDataLayerSection, end_frame=14547, start_frame=3267}]}]
