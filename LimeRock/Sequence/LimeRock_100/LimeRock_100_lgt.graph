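# LimeRock_100 lighting sequence that stages a timed tentacle-shadow effect and overrides time-of-day lighting.
# A DaySequenceModifier blends its user weight to full, then a red rim spotlight switches on and slightly widens its attenuation radius.
# Later a pink spotlight, a large static light-blocker plane and a shadow plane switch on together for a short window.
# The shadow plane drifts, rotates and grows, apparently as a gobo faking a moving tentacle shadow.
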LevelSequence /LimeRock/Sequence/LimeRock_100/LimeRock_100_lgt
bindings [
 SpotLight {
  id=#1
  possessed_class=SpotLight
  tracks [
   {
    type=MovieScene3DTransformTrack
    sections [{
      type=MovieScene3DTransformSection
      channels [
       Location.X {type=MovieSceneScriptingDoubleChannel, default=-91585.551, const [-91585.5547,2]}
       Location.Y {type=MovieSceneScriptingDoubleChannel, default=-95840.629, const [-95840.6328,2]}
       Location.Z {type=MovieSceneScriptingDoubleChannel, default=2670.2891, const [2670.2891,2]}
       Rotation.X {type=MovieSceneScriptingDoubleChannel, default=0, const [0,2]}
       Rotation.Y {type=MovieSceneScriptingDoubleChannel, default=-61.762, const [-61.762,2]}
       Rotation.Z {type=MovieSceneScriptingDoubleChannel, default=2.1924, const [2.1924,2]}
       Scale.X {type=MovieSceneScriptingDoubleChannel, default=1, const [1,2]}
       Scale.Y {type=MovieSceneScriptingDoubleChannel, default=1, const [1,2]}
       Scale.Z {type=MovieSceneScriptingDoubleChannel, default=1, const [1,2]}
      ]
     }]
   }
   {
    type=MovieSceneVisibilityTrack
    sections [{
      type=MovieSceneVisibilitySection
      channels [None {type=MovieSceneScriptingBoolChannel, default=false, keys [[3600,false],[8893,true],[9064,false]]}]
     }]
   }
   {
    type=MovieSceneSpawnTrack
    sections [{
      type=MovieSceneSpawnSection
      channels [None {type=MovieSceneScriptingBoolChannel, default=false, keys [[3600,false],[8893,true],[9064,false]]}]
     }]
   }
  ]
 }
 M_TentacleShadowPlane_LightBlocker {
  id=#2
  possessed_class=FortStaticMeshActor
  tracks [
   {
    type=MovieScene3DTransformTrack
    sections [{
      type=MovieScene3DTransformSection
      channels [
       Location.X {type=MovieSceneScriptingDoubleChannel, default=-91934.6166}
       Location.Y {type=MovieSceneScriptingDoubleChannel, default=-95875.4711}
       Location.Z {type=MovieSceneScriptingDoubleChannel, default=2860.9002}
       Rotation.X {type=MovieSceneScriptingDoubleChannel, default=35.768}
       Rotation.Y {type=MovieSceneScriptingDoubleChannel, default=0.981}
       Rotation.Z {type=MovieSceneScriptingDoubleChannel, default=88.6013}
       Scale.X {type=MovieSceneScriptingDoubleChannel, default=50}
       Scale.Y {type=MovieSceneScriptingDoubleChannel, default=50}
       Scale.Z {type=MovieSceneScriptingDoubleChannel, default=50}
      ]
     }]
   }
   {
    type=MovieSceneVisibilityTrack
    sections [{
      type=MovieSceneVisibilitySection
      channels [None {type=MovieSceneScriptingBoolChannel, default=false, keys [[3600,false],[8893,true],[9064,false]]}]
     }]
   }
   {
    type=MovieSceneSpawnTrack
    sections [{
      type=MovieSceneSpawnSection
      channels [None {type=MovieSceneScriptingBoolChannel, default=false, keys [[3600,false],[8893,true],[9064,false]]}]
     }]
   }
  ]
 }
 DaySequenceModifierVolume {id=#3, possessed_class=DaySequenceModifierVolume, tracks []}
 M_TentacleShadowPlane {
  id=#4
  possessed_class=FortStaticMeshActor
  tracks [
   {
    type=MovieScene3DTransformTrack
    sections [{
      type=MovieScene3DTransformSection
      channels [
       Location.X {type=MovieSceneScriptingDoubleChannel, default=-91552.268, keys [[8893,-91738.1875],[9064,-91686.7422]]}
       Location.Y {type=MovieSceneScriptingDoubleChannel, default=-95881.119, keys [[8893,-95874.7109],[9064,-95873.9766]]}
       Location.Z {type=MovieSceneScriptingDoubleChannel, default=2432.2532, keys [[8893,2297.709],[9064,2334.9421]]}
       Rotation.X {type=MovieSceneScriptingDoubleChannel, default=35.7804, keys [[8893,35.7804],[8974,35.4328],[9064,35.8906]]}
       Rotation.Y {type=MovieSceneScriptingDoubleChannel, default=-3.0187, keys [[8893,-3.0187],[8974,-5.9376],[9064,0.0334]]}
       Rotation.Z {type=MovieSceneScriptingDoubleChannel, default=94.1791, keys [[8893,94.1791],[8974,98.2726],[9064,89.9539]]}
       Scale.X {type=MovieSceneScriptingDoubleChannel, default=5, keys [[8893,5],[9064,7]]}
       Scale.Y {type=MovieSceneScriptingDoubleChannel, default=5, keys [[8893,5],[9064,7]]}
       Scale.Z {type=MovieSceneScriptingDoubleChannel, default=5, keys [[8893,5],[9064,7]]}
      ]
     }]
   }
   {
    type=MovieSceneVisibilityTrack
    sections [{
      type=MovieSceneVisibilitySection
      channels [None {type=MovieSceneScriptingBoolChannel, default=true, keys [[3600,false],[8893,true],[9064,false]]}]
     }]
   }
   {
    type=MovieSceneSpawnTrack
    sections [{
      type=MovieSceneSpawnSection
      channels [None {type=MovieSceneScriptingBoolChannel, default=false, keys [[3600,false],[8893,true],[9064,false]]}]
     }]
   }
  ]
 }
 SpotLight_RimLight {
  id=#5
  possessed_class=SpotLight
  tracks [
   {
    type=MovieSceneSpawnTrack
    sections [{type=MovieSceneSpawnSection, channels [None {type=MovieSceneScriptingBoolChannel, default=false, keys [[3600,false],[8227,true]]}]}]
   }
   {
    type=MovieScene3DTransformTrack
    sections [{
      type=MovieScene3DTransformSection
      channels [
       Location.X {type=MovieSceneScriptingDoubleChannel, default=-122927.4877, const [-115397.9062,1]}
       Location.Y {type=MovieSceneScriptingDoubleChannel, default=-98845.3337, const [-97057.3281,1]}
       Location.Z {type=MovieSceneScriptingDoubleChannel, default=15614.3961, const [22104.2676,1]}
       Rotation.X {type=MovieSceneScriptingDoubleChannel, default=0.4108, const [0.4108,1]}
       Rotation.Y {type=MovieSceneScriptingDoubleChannel, default=-50.0157, const [-50.0157,1]}
       Rotation.Z {type=MovieSceneScriptingDoubleChannel, default=12.8222, const [12.8222,1]}
       Scale.X {type=MovieSceneScriptingDoubleChannel, default=1, const [1,1]}
       Scale.Y {type=MovieSceneScriptingDoubleChannel, default=1, const [1,1]}
       Scale.Z {type=MovieSceneScriptingDoubleChannel, default=1, const [1,1]}
      ]
     }]
   }
  ]
 }
 LightComponent0 {
  id=#6
  parent_id=#5
  possessed_class=SpotLightComponent
  tracks [
   {
    type=MovieSceneFloatTrack
    display_name=Intensity
    sections [{type=MovieSceneFloatSection, channels [None {type=MovieSceneScriptingFloatChannel, default=4750000}]}]
   }
   {
    type=MovieSceneColorTrack
    display_name=Light Color
    sections [{
      type=MovieSceneColorSection
      channels [
       Color.R {type=MovieSceneScriptingFloatChannel, default=1}
       Color.G {type=MovieSceneScriptingFloatChannel, default=0.1441}
       Color.B {type=MovieSceneScriptingFloatChannel, default=0.1912}
       Color.A {type=MovieSceneScriptingFloatChannel, default=1}
      ]
     }]
   }
   {
    type=MovieSceneFloatTrack
    display_name=Attenuation Radius
    sections [{type=MovieSceneFloatSection, channels [None {type=MovieSceneScriptingFloatChannel, default=31000, keys [[8510,30000],[8825,31000]]}]}]
   }
  ]
 }
 DaySequenceModifier {
  id=#7
  parent_id=#3
  possessed_class=DaySequenceModifierComponent
  tracks [{
    type=MovieSceneFloatTrack
    display_name=User Blend Weight
    sections [{type=MovieSceneFloatSection, channels [None {type=MovieSceneScriptingFloatChannel, default=0, keys [[7688,0],[7925,1]]}]}]
   }]
 }
 LightComponent0 {
  id=#8
  parent_id=#1
  possessed_class=SpotLightComponent
  tracks [
   {
    type=MovieSceneFloatTrack
    display_name=Intensity
    sections [{type=MovieSceneFloatSection, channels [None {type=MovieSceneScriptingFloatChannel, default=2000, const [2000,1]}]}]
   }
   {
    type=MovieSceneColorTrack
    display_name=Light Color
    sections [{
      type=MovieSceneColorSection
      channels [
       Color.R {type=MovieSceneScriptingFloatChannel, default=1, const [1,1]}
       Color.G {type=MovieSceneScriptingFloatChannel, default=0.3231, const [0.3231,1]}
       Color.B {type=MovieSceneScriptingFloatChannel, default=0.7084, const [0.7084,1]}
       Color.A {type=MovieSceneScriptingFloatChannel, default=1, const [1,1]}
      ]
     }]
   }
  ]
 }
]
display_rate {denominator=1, numerator=60}
tick_resolution {denominator=1, numerator=24000}
tracks []
