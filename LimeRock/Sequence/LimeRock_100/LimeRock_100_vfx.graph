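# VFX Level Sequence for the LimeRock_100 cinematic, choreographing a reality-rift portal, treehouse pulses, smoke monster, splashes and UFO dust.
# Spawn tracks time each Niagara actor, with transform and attach tracks placing or moving them.
# Niagara parameter tracks animate rift jitter, pulses, barrier point force, monster glow and spawn rates, while particle tracks trigger or toggle systems.
# It also keys portal visibility and a camera rail's position.
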LevelSequence /LimeRock/Sequence/LimeRock_100/LimeRock_100_vfx
bindings [
 NS_LimeRock_Treehouse_Decal {
  id=#1
  possessed_class=NiagaraActor
  tracks [
   {
    type=MovieSceneSpawnTrack
    sections [{
      type=MovieSceneSpawnSection
      channels [None {type=MovieSceneScriptingBoolChannel, default=true, keys [[6447.04,false],[6707.925,true],[7087.925,true],[7088.925,false]]}]
     }]
   }
   {
    type=MovieScene3DTransformTrack
    sections [{
      type=MovieScene3DTransformSection
      channels [
       Location.X {type=MovieSceneScriptingDoubleChannel, default=-95190.9203}
       Location.Y {type=MovieSceneScriptingDoubleChannel, default=-95297.3748}
       Location.Z {type=MovieSceneScriptingDoubleChannel, default=2459}
       Rotation.X {type=MovieSceneScriptingDoubleChannel, default=0}
       Rotation.Y {type=MovieSceneScriptingDoubleChannel, default=0}
       Rotation.Z {type=MovieSceneScriptingDoubleChannel, default=0}
       Scale.X {type=MovieSceneScriptingDoubleChannel, default=1}
       Scale.Y {type=MovieSceneScriptingDoubleChannel, default=1}
       Scale.Z {type=MovieSceneScriptingDoubleChannel, default=1}
      ]
     }]
   }
  ]
 }
 NS_LimeRock_Barrier_Tree {id=#2, possessed_class=NiagaraActor, tracks []}
 LimeRock_SceneRoot {id=#3, possessed_class=Actor, tracks []}
 NS_LimeRock_Shockwave_Ground {
  id=#4
  possessed_class=NiagaraActor
  tracks [
   {
    type=MovieSceneSpawnTrack
    sections [{
      type=MovieSceneSpawnSection
      channels [None {type=MovieSceneScriptingBoolChannel, default=false, keys [[5851.4675,false],[5942.01,true],[7392.155,false]]}]
     }]
   }
   {
    type=MovieScene3DTransformTrack
    sections [{
      type=MovieScene3DTransformSection
      channels [
       Location.X {type=MovieSceneScriptingDoubleChannel, default=-90712.292}
       Location.Y {type=MovieSceneScriptingDoubleChannel, default=-96058.2861}
       Location.Z {type=MovieSceneScriptingDoubleChannel, default=1003.0898}
       Rotation.X {type=MovieSceneScriptingDoubleChannel, default=0}
       Rotation.Y {type=MovieSceneScriptingDoubleChannel, default=0}
       Rotation.Z {type=MovieSceneScriptingDoubleChannel, default=0}
       Scale.X {type=MovieSceneScriptingDoubleChannel, default=1}
       Scale.Y {type=MovieSceneScriptingDoubleChannel, default=1}
       Scale.Z {type=MovieSceneScriptingDoubleChannel, default=1}
      ]
     }]
   }
  ]
 }
 NS_LimeRock_MonsterSplash_Ripple {
  id=#5
  possessed_class=NiagaraActor
  tracks [
   {
    type=MovieSceneSpawnTrack
    sections [{
      type=MovieSceneSpawnSection
      channels [None {type=MovieSceneScriptingBoolChannel, default=false, keys [[8156,false],[8157,true],[8619.3175,false]]}]
     }]
   }
   {
    type=MovieScene3DTransformTrack
    sections [{
      type=MovieScene3DTransformSection
      channels [
       Location.X {type=MovieSceneScriptingDoubleChannel, default=-117454.1967}
       Location.Y {type=MovieSceneScriptingDoubleChannel, default=-103956.8684}
       Location.Z {type=MovieSceneScriptingDoubleChannel, default=1856.1605}
       Rotation.X {type=MovieSceneScriptingDoubleChannel, default=0}
       Rotation.Y {type=MovieSceneScriptingDoubleChannel, default=0}
       Rotation.Z {type=MovieSceneScriptingDoubleChannel, default=0}
       Scale.X {type=MovieSceneScriptingDoubleChannel, default=1}
       Scale.Y {type=MovieSceneScriptingDoubleChannel, default=1}
       Scale.Z {type=MovieSceneScriptingDoubleChannel, default=1}
      ]
     }]
   }
  ]
 }
 NS_LR_Treehouse_Pulse_FX {
  id=#6
  possessed_class=NiagaraActor
  tracks [
   {
    type=MovieSceneSpawnTrack
    sections [{
      type=MovieSceneSpawnSection
      channels [None {type=MovieSceneScriptingBoolChannel, default=false, keys [[5275,false],[5905,true],[6102,false]]}]
     }]
   }
   {
    type=MovieScene3DTransformTrack
    sections [{
      type=MovieScene3DTransformSection
      channels [
       Location.X {type=MovieSceneScriptingDoubleChannel, default=-99590.8668}
       Location.Y {type=MovieSceneScriptingDoubleChannel, default=-99071.4586}
       Location.Z {type=MovieSceneScriptingDoubleChannel, default=5932.4304}
       Rotation.X {type=MovieSceneScriptingDoubleChannel, default=-3.7211}
       Rotation.Y {type=MovieSceneScriptingDoubleChannel, default=-18.8522}
       Rotation.Z {type=MovieSceneScriptingDoubleChannel, default=21.6289}
       Scale.X {type=MovieSceneScriptingDoubleChannel, default=1}
       Scale.Y {type=MovieSceneScriptingDoubleChannel, default=1}
       Scale.Z {type=MovieSceneScriptingDoubleChannel, default=1}
      ]
     }]
   }
  ]
 }
 NiagaraComponent0 {
  id=#7
  parent_id=#8
  possessed_class=NiagaraComponent
  tracks [{
    type=MovieSceneNiagaraFloatParameterTrack
    display_name=SpawnRate
    sections [{
      type=MovieSceneFloatSection
      channels [None {type=MovieSceneScriptingFloatChannel, default=40, keys [[8133.9,0],[8214.3825,5],[8448.2875,0],[8546.3775,0],[8596.68,40],[8750.1025,0]]}]
     }]
   }]
 }
 NS_LimeRock_RealityTear {
  id=#9
  possessed_class=NiagaraActor
  tracks [
   {
    type=MovieSceneSpawnTrack
    sections [{type=MovieSceneSpawnSection, channels [None {type=MovieSceneScriptingBoolChannel, default=false, keys [[7874.3275,false],[8587.04,true]]}]}]
   }
   {
    type=MovieScene3DTransformTrack
    sections [{
      type=MovieScene3DTransformSection
      channels [
       Location.X {type=MovieSceneScriptingDoubleChannel, default=-161542.5475}
       Location.Y {type=MovieSceneScriptingDoubleChannel, default=-95942.0595}
       Location.Z {type=MovieSceneScriptingDoubleChannel, default=51475.3813}
       Rotation.X {type=MovieSceneScriptingDoubleChannel, default=0}
       Rotation.Y {type=MovieSceneScriptingDoubleChannel, default=0}
       Rotation.Z {type=MovieSceneScriptingDoubleChannel, default=0}
       Scale.X {type=MovieSceneScriptingDoubleChannel, default=1}
       Scale.Y {type=MovieSceneScriptingDoubleChannel, default=1}
       Scale.Z {type=MovieSceneScriptingDoubleChannel, default=1}
      ]
     }]
   }
  ]
 }
 NiagaraComponent0 {
  id=#10
  parent_id=#11
  possessed_class=NiagaraComponent
  tracks [
   {
    type=MovieSceneParticleTrack
    display_name=FX System
    sections [{type=MovieSceneParticleSection, channels [None {type=MovieSceneScriptingParticleChannel, const ["<ParticleKey.TRIGGER: 2>", 1]}]}]
   }
   {
    type=MovieSceneNiagaraVectorParameterTrack
    display_name=Origin
    sections [{
      type=MovieSceneFloatVectorSection
      channels [
       Vector.X {type=MovieSceneScriptingFloatChannel, default=-90577, const [-90945,1]}
       Vector.Y {type=MovieSceneScriptingFloatChannel, default=-94749, const [-94994,1]}
       Vector.Z {type=MovieSceneScriptingFloatChannel, default=1226, const [1226,1]}
      ]
     }]
   }
  ]
 }
 NS_LimeRock_TentacleTrails {id=#12, possessed_class=NiagaraActor, tracks []}
 NS_LimeRock_RiftDevice_Idle {
  id=#13
  possessed_class=NiagaraActor
  tracks [
   {
    type=MovieSceneSpawnTrack
    sections [{
      type=MovieSceneSpawnSection
      channels [None {type=MovieSceneScriptingBoolChannel, default=false, keys [[5181.5625,false],[5275.5625,true],[10664.5625,false]]}]
     }]
   }
   {
    type=MovieScene3DTransformTrack
    sections [{
      type=MovieScene3DTransformSection
      channels [
       Location.X {type=MovieSceneScriptingDoubleChannel, default=0}
       Location.Y {type=MovieSceneScriptingDoubleChannel, default=0}
       Location.Z {type=MovieSceneScriptingDoubleChannel, default=0}
       Rotation.X {type=MovieSceneScriptingDoubleChannel, default=0}
       Rotation.Y {type=MovieSceneScriptingDoubleChannel, default=0}
       Rotation.Z {type=MovieSceneScriptingDoubleChannel, default=0}
       Scale.X {type=MovieSceneScriptingDoubleChannel, default=1}
       Scale.Y {type=MovieSceneScriptingDoubleChannel, default=1}
       Scale.Z {type=MovieSceneScriptingDoubleChannel, default=1}
      ]
     }]
   }
   {type=MovieScene3DAttachTrack, sections [{type=MovieScene3DAttachSection, end_frame=14251, start_frame=5276}]}
  ]
 }
 NiagaraComponent0 {
  id=#14
  parent_id=#15
  possessed_class=NiagaraComponent
  tracks [
   {
    type=MovieSceneNiagaraFloatParameterTrack
    display_name=Spread
    sections [{
      type=MovieSceneFloatSection
      channels [None {type=MovieSceneScriptingFloatChannel, default=0.7, keys [[7964.3275,0],[8202,0.5],[8381.84,1.15],[8541.3275,1.1],[8882.51,1.1],[9059.4,0.7]]}]
     }]
   }
   {
    type=MovieSceneNiagaraFloatParameterTrack
    display_name=HeartGlow
    sections [{
      type=MovieSceneFloatSection
      channels [None {type=MovieSceneScriptingFloatChannel, default=0.5, keys [[8381.84,0],[8495.3275,0.8],[8556.12,0.8545],[8593.12,1.26],[8726.12,1]]}]
     }]
   }
  ]
 }
 NS_LimeRock_RiftIn_Buildup {
  id=#16
  possessed_class=NiagaraActor
  tracks [
   {
    type=MovieSceneSpawnTrack
    sections [{
      type=MovieSceneSpawnSection
      channels [None {type=MovieSceneScriptingBoolChannel, default=false, keys [[3708.98,false],[3969.395,true],[4242.395,false]]}]
     }]
   }
   {
    type=MovieScene3DTransformTrack
    sections [{
      type=MovieScene3DTransformSection
      channels [
       Location.X {type=MovieSceneScriptingDoubleChannel, default=-92987.5289}
       Location.Y {type=MovieSceneScriptingDoubleChannel, default=-90983.0746}
       Location.Z {type=MovieSceneScriptingDoubleChannel, default=2927.7}
       Rotation.X {type=MovieSceneScriptingDoubleChannel, default=0}
       Rotation.Y {type=MovieSceneScriptingDoubleChannel, default=0}
       Rotation.Z {type=MovieSceneScriptingDoubleChannel, default=0}
       Scale.X {type=MovieSceneScriptingDoubleChannel, default=1}
       Scale.Y {type=MovieSceneScriptingDoubleChannel, default=1}
       Scale.Z {type=MovieSceneScriptingDoubleChannel, default=1}
      ]
     }]
   }
  ]
 }
 NiagaraComponent0 {
  id=#17
  parent_id=#6
  possessed_class=NiagaraComponent
  tracks [{
    type=MovieSceneParticleTrack
    display_name=FX System
    sections [{type=MovieSceneParticleSection, channels [None {type=MovieSceneScriptingParticleChannel, const ["<ParticleKey.TRIGGER: 2>", 1]}]}]
   }]
 }
 NS_LimeRock_SmokeMonsterMain {
  id=#15
  possessed_class=NiagaraActor
  tracks [
   {
    type=MovieSceneSpawnTrack
    sections [{type=MovieSceneSpawnSection, channels [None {type=MovieSceneScriptingBoolChannel, default=false, keys [[7164.84,false],[7694.84,true]]}]}]
   }
   {
    type=MovieScene3DTransformTrack
    sections [{
      type=MovieScene3DTransformSection
      channels [
       Location.X {type=MovieSceneScriptingDoubleChannel, default=0, keys [[8745.4925,0],[8905.18,13418.4463],[9050.255,7164.1284],[9197.11,8828.292]]}
       Location.Y {
        type=MovieSceneScriptingDoubleChannel
        default=-95479.9999
        keys [[7694.84,0],[8745.4925,0],[8905.18,-1990.7545],[9050.255,1.7377],[9197.11,443.0892]]
       }
       Location.Z {type=MovieSceneScriptingDoubleChannel, default=0, keys [[8745.4925,0],[8905.18,-9620.7373],[9050.255,-5288.4292],[9197.11,-3127.5078]]}
       Rotation.X {
        type=MovieSceneScriptingDoubleChannel
        default=0
        keys [[7694.84,0.3549],[7865.3275,0],[8024.8225,0],[8183.58,-2.5308],[8293.4875,-3.9147],[8459.915,1.8917],[8572.2825,2.1729],[8745.4925,2.1729],[8905.18,8.3189],[9050.255,14.0319],[9197.11,15.3899]]
       }
       Rotation.Y {
        type=MovieSceneScriptingDoubleChannel
        default=0
        keys [[7694.84,-61.8951],[7865.3275,-9.1377],[8024.8225,0],[8162.3275,-38.9575],[8293.4875,-59.8042],[8408,-42.4074],[8470,-16.1767],[8504,-23.0429],[8542.3275,-18.4291],[8572,-32.8625],[8745.4925,-32.8625],[8905.18,-44.7996],[9050.255,-20.7389],[9197.11,-31.3044]]
       }
       Rotation.Z {
        type=MovieSceneScriptingDoubleChannel
        default=0
        keys [[7694.84,9.8823],[7865.3275,7.5793],[8024.8225,0],[8183.58,11.298],[8293.4875,6.2682],[8459.915,-10.694],[8572.2825,-11.3757],[8745.4925,-11.3757],[8905.18,-13.4353],[9050.255,-7.1943],[9197.11,-10.2761]]
       }
       Scale.X {type=MovieSceneScriptingDoubleChannel, default=1, const [1,1]}
       Scale.Y {type=MovieSceneScriptingDoubleChannel, default=1, const [1,1]}
       Scale.Z {type=MovieSceneScriptingDoubleChannel, default=1, const [1,1]}
      ]
     }]
   }
   {type=MovieScene3DAttachTrack, sections [{type=MovieScene3DAttachSection, end_frame=13451, start_frame=7694}]}
  ]
 }
 NS_LimeRock_RiftDevice_Burst {
  id=#18
  possessed_class=NiagaraActor
  tracks [
   {
    type=MovieSceneSpawnTrack
    sections [{
      type=MovieSceneSpawnSection
      channels [None {type=MovieSceneScriptingBoolChannel, default=false, keys [[5761.5625,false],[5794.855,true],[5998.1675,false]]}]
     }]
   }
   {
    type=MovieScene3DTransformTrack
    sections [{
      type=MovieScene3DTransformSection
      channels [
       Location.X {type=MovieSceneScriptingDoubleChannel, default=0}
       Location.Y {type=MovieSceneScriptingDoubleChannel, default=0}
       Location.Z {type=MovieSceneScriptingDoubleChannel, default=0}
       Rotation.X {type=MovieSceneScriptingDoubleChannel, default=0}
       Rotation.Y {type=MovieSceneScriptingDoubleChannel, default=0}
       Rotation.Z {type=MovieSceneScriptingDoubleChannel, default=0}
       Scale.X {type=MovieSceneScriptingDoubleChannel, default=1}
       Scale.Y {type=MovieSceneScriptingDoubleChannel, default=1}
       Scale.Z {type=MovieSceneScriptingDoubleChannel, default=1}
      ]
     }]
   }
   {type=MovieScene3DAttachTrack, sections [{type=MovieScene3DAttachSection, end_frame=14159, start_frame=5762}]}
  ]
 }
 NS_LimeRock_Treehouse_Decal_02 {
  id=#19
  possessed_class=NiagaraActor
  tracks [
   {
    type=MovieSceneSpawnTrack
    sections [{
      type=MovieSceneSpawnSection
      channels [None {type=MovieSceneScriptingBoolChannel, default=false, keys [[6447.04,false],[6707.925,true],[7062.925,true],[7063.925,false]]}]
     }]
   }
   {
    type=MovieScene3DTransformTrack
    sections [{
      type=MovieScene3DTransformSection
      channels [
       Location.X {type=MovieSceneScriptingDoubleChannel, default=-95190.9203}
       Location.Y {type=MovieSceneScriptingDoubleChannel, default=-95297.3748}
       Location.Z {type=MovieSceneScriptingDoubleChannel, default=2769, keys [[6707.925,3073],[6729.925,3073],[6849,2769],[7004,1974],[7053,1787]]}
       Rotation.X {type=MovieSceneScriptingDoubleChannel, default=0}
       Rotation.Y {type=MovieSceneScriptingDoubleChannel, default=0}
       Rotation.Z {type=MovieSceneScriptingDoubleChannel, default=0}
       Scale.X {type=MovieSceneScriptingDoubleChannel, default=1}
       Scale.Y {type=MovieSceneScriptingDoubleChannel, default=1}
       Scale.Z {type=MovieSceneScriptingDoubleChannel, default=1}
      ]
     }]
   }
  ]
 }
 BP_LimeRock_Portal {
  id=#20
  possessed_class=BP_LimeRock_Portal_C
  tracks [
   {
    type=MovieSceneVisibilityTrack
    sections [{type=MovieSceneVisibilitySection, channels [None {type=MovieSceneScriptingBoolChannel, default=true, keys [[3789,true],[7329.7475,false]]}]}]
   }
   {type=MovieSceneEventTrack, sections []}
   {
    type=MovieSceneFloatTrack
    display_name=Portal Radius
    sections [{type=MovieSceneFloatSection, channels [None {type=MovieSceneScriptingFloatChannel, default=0}]}]
   }
  ]
 }
 NiagaraComponent0 {
  id=#21
  parent_id=#22
  possessed_class=NiagaraComponent
  tracks [
   {
    type=MovieScene3DTransformTrack
    sections [{
      type=MovieScene3DTransformSection
      channels [
       Location.X {type=MovieSceneScriptingDoubleChannel, default=-104625.4313}
       Location.Y {type=MovieSceneScriptingDoubleChannel, default=-109411.8726}
       Location.Z {type=MovieSceneScriptingDoubleChannel, default=-1077.5996}
       Rotation.X {type=MovieSceneScriptingDoubleChannel, default=0}
       Rotation.Y {type=MovieSceneScriptingDoubleChannel, default=0}
       Rotation.Z {type=MovieSceneScriptingDoubleChannel, default=0}
       Scale.X {type=MovieSceneScriptingDoubleChannel, default=1}
       Scale.Y {type=MovieSceneScriptingDoubleChannel, default=1}
       Scale.Z {type=MovieSceneScriptingDoubleChannel, default=1}
      ]
     }]
   }
   {
    type=MovieSceneFloatTrack
    display_name=Translucency Sort Distance Offset
    sections [{type=MovieSceneFloatSection, channels [None {type=MovieSceneScriptingFloatChannel, default=0}]}]
   }
  ]
 }
 NiagaraComponent0 {
  id=#23
  parent_id=#24
  possessed_class=NiagaraComponent
  tracks [
   {
    type=MovieSceneParticleTrack
    display_name=FX System
    sections [{type=MovieSceneParticleSection, channels [None {type=MovieSceneScriptingParticleChannel, const ["<ParticleKey.TRIGGER: 2>", 2]}]}]
   }
   {
    type=MovieSceneNiagaraVectorParameterTrack
    display_name=Origin
    sections [{
      type=MovieSceneFloatVectorSection
      channels [
       Vector.X {type=MovieSceneScriptingFloatChannel, default=-90577, const [-90577,1]}
       Vector.Y {type=MovieSceneScriptingFloatChannel, default=-94749, const [-94749,1]}
       Vector.Z {type=MovieSceneScriptingFloatChannel, default=1226, const [1226,1]}
      ]
     }]
   }
  ]
 }
 NS_LimeRock_RealityTear_Portal {
  id=#25
  parent_id=#20
  possessed_class=NiagaraComponent
  tracks [
   {
    type=MovieSceneNiagaraFloatParameterTrack
    display_name=RiftOpen
    sections [{type=MovieSceneFloatSection, channels [None {type=MovieSceneScriptingFloatChannel, default=0.4, keys [[7144,1],[7164.84,0]]}]}]
   }
   {
    type=MovieSceneNiagaraFloatParameterTrack
    display_name=RiftJitter
    sections [{
      type=MovieSceneFloatSection
      channels [None {
        type=MovieSceneScriptingFloatChannel
        default=0
        keys [[4450.4875,0],[4618.3225,1],[4706.0625,0],[4867.4875,0],[4986.4475,1],[5014.57,1],[5100.55,0],[5117.98,1],[5252.5875,0],[5609.7275,0],[5834.495,1],[5844.0375,0],[5899,0],[5908,1],[5953,0.7696],[6149,0.5],[6711.8225,0.5],[6759.0125,1]]
       }]
     }]
   }
   {
    type=MovieSceneParticleTrack
    display_name=FX System
    sections [{
      type=MovieSceneParticleSection
      channels [None {
        type=MovieSceneScriptingParticleChannel
        keys [[3303.8725, "<ParticleKey.DEACTIVATE: 1>"], [3600, "<ParticleKey.ACTIVATE: 0>"], [7164.84, "<ParticleKey.DEACTIVATE: 1>"]]
       }]
     }]
   }
   {
    type=MovieSceneNiagaraFloatParameterTrack
    display_name=PulseTravel
    sections [{
      type=MovieSceneFloatSection
      channels [None {
        type=MovieSceneScriptingFloatChannel
        default=1
        keys [[4587.105,0],[4688.0475,1],[4981.5425,0],[5044.5575,1],[5814.6375,0],[5916,1],[6720.82,0],[6773.8525,1],[6788.6925,0],[6844.04,1],[6858.59,0],[6918.645,1],[6930.4375,0],[6990.4925,1],[7005.58,0],[7065.635,1],[7084.945,0],[7145,1]]
       }]
     }]
   }
   {
    type=MovieSceneNiagaraFloatParameterTrack
    display_name=Intensity
    sections [{
      type=MovieSceneFloatSection
      channels [None {
        type=MovieSceneScriptingFloatChannel
        default=2
        keys [[4653.7175,0],[4738.7075,0.3],[4897.2775,0.3],[5015.4475,1],[5162.4675,0.6],[5738.58,0.6],[5850.015,1.5],[5995,2],[6711.8225,1],[6759.0125,0.5],[6844.04,0.8382],[6872.3625,0.3],[6918.7,0.5641],[7098.4,0.2]]
       }]
     }]
   }
   {
    type=MovieSceneNiagaraFloatParameterTrack
    display_name=PulseColorShift
    sections [{type=MovieSceneFloatSection, channels [None {type=MovieSceneScriptingFloatChannel, default=1, keys [[4560,0],[6711.8225,1]]}]}]
   }
  ]
 }
 NS_LimeRock_UFODust_Kickup {
  id=#26
  possessed_class=NiagaraActor
  tracks [
   {
    type=MovieSceneSpawnTrack
    sections [{type=MovieSceneSpawnSection, channels [None {type=MovieSceneScriptingBoolChannel, default=true, keys [[7476,false],[7481,true]]}]}]
   }
   {
    type=MovieScene3DTransformTrack
    sections [{
      type=MovieScene3DTransformSection
      channels [
       Location.X {type=MovieSceneScriptingDoubleChannel, default=-92024.8878}
       Location.Y {type=MovieSceneScriptingDoubleChannel, default=-95906.1084}
       Location.Z {type=MovieSceneScriptingDoubleChannel, default=1063.2823}
       Rotation.X {type=MovieSceneScriptingDoubleChannel, default=0}
       Rotation.Y {type=MovieSceneScriptingDoubleChannel, default=0}
       Rotation.Z {type=MovieSceneScriptingDoubleChannel, default=82.6}
       Scale.X {type=MovieSceneScriptingDoubleChannel, default=1}
       Scale.Y {type=MovieSceneScriptingDoubleChannel, default=1}
       Scale.Z {type=MovieSceneScriptingDoubleChannel, default=1}
      ]
     }]
   }
  ]
 }
 NS_LimeRock_Ground_Portal_SmokePour {
  id=#27
  possessed_class=NiagaraActor
  tracks [
   {
    type=MovieSceneSpawnTrack
    sections [{
      type=MovieSceneSpawnSection
      channels [None {type=MovieSceneScriptingBoolChannel, default=false, keys [[7164.84,false],[7688.84,true],[8228.4,false]]}]
     }]
   }
   {
    type=MovieScene3DTransformTrack
    sections [{
      type=MovieScene3DTransformSection
      channels [
       Location.X {type=MovieSceneScriptingDoubleChannel, default=-115767.2656}
       Location.Y {type=MovieSceneScriptingDoubleChannel, default=-103432.7498}
       Location.Z {type=MovieSceneScriptingDoubleChannel, default=12342.2148}
       Rotation.X {type=MovieSceneScriptingDoubleChannel, default=-178.2944}
       Rotation.Y {type=MovieSceneScriptingDoubleChannel, default=-72.6844}
       Rotation.Z {type=MovieSceneScriptingDoubleChannel, default=-164.8185}
       Scale.X {type=MovieSceneScriptingDoubleChannel, default=1}
       Scale.Y {type=MovieSceneScriptingDoubleChannel, default=1}
       Scale.Z {type=MovieSceneScriptingDoubleChannel, default=1}
      ]
     }]
   }
  ]
 }
 NS_LimeRock_BeamEnergyUp {
  id=#28
  possessed_class=NiagaraActor
  tracks [
   {
    type=MovieSceneSpawnTrack
    sections [{
      type=MovieSceneSpawnSection
      channels [None {type=MovieSceneScriptingBoolChannel, default=false, keys [[6938.925,false],[7076,true],[7314.925,false]]}]
     }]
   }
   {
    type=MovieScene3DTransformTrack
    sections [{
      type=MovieScene3DTransformSection
      channels [
       Location.X {type=MovieSceneScriptingDoubleChannel, default=-95141.428}
       Location.Y {type=MovieSceneScriptingDoubleChannel, default=-95296.1959}
       Location.Z {type=MovieSceneScriptingDoubleChannel, default=1709.4284}
       Rotation.X {type=MovieSceneScriptingDoubleChannel, default=0}
       Rotation.Y {type=MovieSceneScriptingDoubleChannel, default=0}
       Rotation.Z {type=MovieSceneScriptingDoubleChannel, default=0}
       Scale.X {type=MovieSceneScriptingDoubleChannel, default=1}
       Scale.Y {type=MovieSceneScriptingDoubleChannel, default=1}
       Scale.Z {type=MovieSceneScriptingDoubleChannel, default=1}
      ]
     }]
   }
  ]
 }
 NiagaraComponent0 {
  id=#29
  parent_id=#30
  possessed_class=NiagaraComponent
  tracks [
   {
    type=MovieSceneNiagaraFloatParameterTrack
    display_name=Spread
    sections [{
      type=MovieSceneFloatSection
      channels [None {type=MovieSceneScriptingFloatChannel, default=0.7, keys [[7964.3275,0],[8202,0.5],[8279,1.1],[8541.3275,1],[8882.51,1.1],[9059.4,0.7]]}]
     }]
   }
   {
    type=MovieSceneNiagaraFloatParameterTrack
    display_name=HeartGlow
    sections [{
      type=MovieSceneFloatSection
      channels [None {type=MovieSceneScriptingFloatChannel, default=0.5, keys [[8381.84,0],[8495.3275,0.8],[8556.12,0.8545],[8593.12,1.26],[8726.12,1]]}]
     }]
   }
  ]
 }
 NiagaraComponent0 {
  id=#31
  parent_id=#12
  possessed_class=NiagaraComponent
  tracks [{
    type=MovieSceneParticleTrack
    display_name=FX System
    sections [{type=MovieSceneParticleSection, channels [None {type=MovieSceneScriptingParticleChannel, const ["<ParticleKey.TRIGGER: 2>", 1]}]}]
   }]
 }
 NS_LimeRock_Smoke_Passby {
  id=#32
  possessed_class=NiagaraActor
  tracks [
   {
    type=MovieSceneSpawnTrack
    sections [{
      type=MovieSceneSpawnSection
      channels [None {type=MovieSceneScriptingBoolChannel, default=false, keys [[8734.53,false],[8882.51,true],[9050.255,false]]}]
     }]
   }
   {
    type=MovieScene3DTransformTrack
    sections [{
      type=MovieScene3DTransformSection
      channels [
       Location.X {type=MovieSceneScriptingDoubleChannel, default=-91727.8731}
       Location.Y {type=MovieSceneScriptingDoubleChannel, default=-95785.9133}
       Location.Z {type=MovieSceneScriptingDoubleChannel, default=1009.5242}
       Rotation.X {type=MovieSceneScriptingDoubleChannel, default=0}
       Rotation.Y {type=MovieSceneScriptingDoubleChannel, default=-5.2241}
       Rotation.Z {type=MovieSceneScriptingDoubleChannel, default=0}
       Scale.X {type=MovieSceneScriptingDoubleChannel, default=1}
       Scale.Y {type=MovieSceneScriptingDoubleChannel, default=1}
       Scale.Z {type=MovieSceneScriptingDoubleChannel, default=1}
      ]
     }]
   }
  ]
 }
 NiagaraComponent0 {
  id=#33
  parent_id=#34
  possessed_class=NiagaraComponent
  tracks [{
    type=MovieSceneParticleTrack
    display_name=FX System
    sections [{type=MovieSceneParticleSection, channels [None {type=MovieSceneScriptingParticleChannel, const ["<ParticleKey.TRIGGER: 2>", 1]}]}]
   }]
 }
 NiagaraComponent0 {id=#35, parent_id=#5, possessed_class=NiagaraComponent, tracks []}
 NiagaraComponent0 {
  id=#36
  parent_id=#37
  possessed_class=NiagaraComponent
  tracks [
   {
    type=MovieSceneNiagaraFloatParameterTrack
    display_name=Splash_SpawnRate
    sections [{type=MovieSceneFloatSection, channels [None {type=MovieSceneScriptingFloatChannel, default=20, keys [[8060,20],[8160,70],[8182,0]]}]}]
   }
   {
    type=MovieSceneNiagaraFloatParameterTrack
    display_name=Splash_SpawnRadius
    sections [{type=MovieSceneFloatSection, channels [None {type=MovieSceneScriptingFloatChannel, default=5000, keys [[8060,5000],[8170,12000]]}]}]
   }
   {
    type=MovieSceneNiagaraFloatParameterTrack
    display_name=Splash_SizeMult
    sections [{type=MovieSceneFloatSection, channels [None {type=MovieSceneScriptingFloatChannel, default=20, keys [[8060,10],[8170,20]]}]}]
   }
  ]
 }
 NS_LImeRock_AR_TracersHope {
  id=#24
  possessed_class=NiagaraActor
  tracks [
   {
    type=MovieSceneSpawnTrack
    sections [{
      type=MovieSceneSpawnSection
      channels [None {type=MovieSceneScriptingBoolChannel, default=false, keys [[8380.1425,false],[8865.295,true],[9224.545,false]]}]
     }]
   }
   {
    type=MovieScene3DTransformTrack
    sections [{
      type=MovieScene3DTransformSection
      channels [
       Location.X {
        type=MovieSceneScriptingDoubleChannel
        default=-92707.8659
        keys [[8865.295,-92707.8672],[8937.5575,-93262.1797],[9097.7625,-95606.0547],[9127.8725,-95796.1797],[9160.0475,-94376.6562]]
       }
       Location.Y {
        type=MovieSceneScriptingDoubleChannel
        default=-96184.9075
        thinned_from=5
        keys [[8865.295,-96184.9062],[8937.5575,-96524.0391],[9160.0475,-96314.0859]]
       }
       Location.Z {
        type=MovieSceneScriptingDoubleChannel
        default=3043.8228
        keys [[8865.295,3043.8228],[8937.5575,3806.7795],[9097.7625,4217.5923],[9127.8725,4974.3794],[9160.0475,4917.2573]]
       }
       Rotation.X {
        type=MovieSceneScriptingDoubleChannel
        default=0.4657
        keys [[8865.295,0.4657],[8937.5575,6.146],[9097.7625,-6.9565],[9127.8725,-6.8619],[9160.0475,-16.2731]]
       }
       Rotation.Y {
        type=MovieSceneScriptingDoubleChannel
        default=-30.1813
        keys [[8865.295,-30.1813],[8937.5575,4.3891],[9097.7625,14.3145],[9127.8725,-10.8121],[9160.0475,-25.0946]]
       }
       Rotation.Z {
        type=MovieSceneScriptingDoubleChannel
        default=31.6971
        keys [[8865.295,31.6971],[8937.5575,15.0568],[9097.7625,6.176],[9127.8725,9.1971],[9160.0475,26.6451]]
       }
       Scale.X {type=MovieSceneScriptingDoubleChannel, default=1, const [1,1]}
       Scale.Y {type=MovieSceneScriptingDoubleChannel, default=1, const [1,1]}
       Scale.Z {type=MovieSceneScriptingDoubleChannel, default=1, const [1,1]}
      ]
     }]
   }
  ]
 }
 NS_LimeRock_RiftIn_Mini {
  id=#38
  possessed_class=NiagaraActor
  tracks [
   {
    type=MovieSceneSpawnTrack
    sections [{
      type=MovieSceneSpawnSection
      channels [None {type=MovieSceneScriptingBoolChannel, default=false, keys [[3812,false],[4089.3975,true],[4398,false]]}]
     }]
   }
   {
    type=MovieScene3DTransformTrack
    sections [{
      type=MovieScene3DTransformSection
      channels [
       Location.X {type=MovieSceneScriptingDoubleChannel, default=-92987.5289}
       Location.Y {type=MovieSceneScriptingDoubleChannel, default=-90983.0746}
       Location.Z {type=MovieSceneScriptingDoubleChannel, default=2927.7}
       Rotation.X {type=MovieSceneScriptingDoubleChannel, default=0}
       Rotation.Y {type=MovieSceneScriptingDoubleChannel, default=0}
       Rotation.Z {type=MovieSceneScriptingDoubleChannel, default=-56.1574}
       Scale.X {type=MovieSceneScriptingDoubleChannel, default=1}
       Scale.Y {type=MovieSceneScriptingDoubleChannel, default=1}
       Scale.Z {type=MovieSceneScriptingDoubleChannel, default=1}
      ]
     }]
   }
  ]
 }
 NiagaraComponent0 {id=#39, parent_id=#40, possessed_class=NiagaraComponent, tracks []}
 NS_LimeRock_Barrier_Fence {id=#41, possessed_class=BP_LimeRock_ShieldRig_C, tracks []}
 NS_LimeRock_RealityTear_Drain {
  id=#42
  possessed_class=NiagaraActor
  tracks [
   {
    type=MovieSceneSpawnTrack
    sections [{
      type=MovieSceneSpawnSection
      channels [None {type=MovieSceneScriptingBoolChannel, default=false, keys [[5597.2225,false],[5805,true],[6154.435,false]]}]
     }]
   }
   {
    type=MovieScene3DTransformTrack
    sections [{
      type=MovieScene3DTransformSection
      channels [
       Location.X {type=MovieSceneScriptingDoubleChannel, default=-100970.5751}
       Location.Y {type=MovieSceneScriptingDoubleChannel, default=-99577.7682}
       Location.Z {type=MovieSceneScriptingDoubleChannel, default=6679.692}
       Rotation.X {type=MovieSceneScriptingDoubleChannel, default=1.7056}
       Rotation.Y {type=MovieSceneScriptingDoubleChannel, default=-17.3156}
       Rotation.Z {type=MovieSceneScriptingDoubleChannel, default=15.1815}
       Scale.X {type=MovieSceneScriptingDoubleChannel, default=1}
       Scale.Y {type=MovieSceneScriptingDoubleChannel, default=1}
       Scale.Z {type=MovieSceneScriptingDoubleChannel, default=1}
      ]
     }]
   }
  ]
 }
 CameraRig_Rail {
  id=#43
  possessed_class=CameraRig_Rail
  tracks [
   {
    type=MovieSceneSpawnTrack
    sections [{type=MovieSceneSpawnSection, channels [None {type=MovieSceneScriptingBoolChannel, default=false, keys [[7164.84,false],[7688.84,true]]}]}]
   }
   {
    type=MovieScene3DTransformTrack
    sections [{
      type=MovieScene3DTransformSection
      channels [
       Location.X {type=MovieSceneScriptingDoubleChannel, default=-128539.0253}
       Location.Y {type=MovieSceneScriptingDoubleChannel, default=-106358.7593}
       Location.Z {type=MovieSceneScriptingDoubleChannel, default=7666.9914}
       Rotation.X {type=MovieSceneScriptingDoubleChannel, default=0}
       Rotation.Y {type=MovieSceneScriptingDoubleChannel, default=0}
       Rotation.Z {type=MovieSceneScriptingDoubleChannel, default=14.0669}
       Scale.X {type=MovieSceneScriptingDoubleChannel, default=1}
       Scale.Y {type=MovieSceneScriptingDoubleChannel, default=1}
       Scale.Z {type=MovieSceneScriptingDoubleChannel, default=1}
      ]
     }]
   }
   {
    type=MovieSceneFloatTrack
    display_name=Current Position on Rail
    sections [{
      type=MovieSceneFloatSection
      channels [None {type=MovieSceneScriptingFloatChannel, default=0.2, keys [[7716.84,0],[7912.84,0.2],[8026.3275,0.25],[8281.3275,0.5796],[8552.3275,1]]}]
     }]
   }
  ]
 }
 NiagaraComponent0 {
  id=#44
  parent_id=#2
  possessed_class=NiagaraComponent
  tracks [
   {
    type=MovieSceneNiagaraFloatParameterTrack
    display_name=Intensity
    sections [{
      type=MovieSceneFloatSection
      channels [None {
        type=MovieSceneScriptingFloatChannel
        default=1
        keys [[4507.15,0],[4691.225,0.3],[4993.9175,0.3],[5129.5175,0.5],[5821.4175,0.5],[5911.1025,1],[6874,1],[6999.735,0],[7616.2575,0],[7702.6725,1]]
       }]
     }]
   }
   {
    type=MovieSceneNiagaraFloatParameterTrack
    display_name=PointForce
    sections [{
      type=MovieSceneFloatSection
      channels [None {
        type=MovieSceneScriptingFloatChannel
        default=0
        keys [[5692.355,0],[5894.6975,-0.25],[5911.1025,1],[5930,1],[6027,0],[7677.2425,0],[7702.6725,1],[7788.795,0],[8551.9475,0],[8571.9975,1],[8666.0725,1],[8803.3275,0]]
       }]
     }]
   }
   {
    type=MovieSceneNiagaraVectorParameterTrack
    display_name=PointForceOrigin
    sections [{
      type=MovieSceneFloatVectorSection
      channels [
       Vector.X {type=MovieSceneScriptingFloatChannel, default=-109771, keys [[3686.605,-101065],[7616.2575,-115767],[8513.3925,-109771]]}
       Vector.Y {type=MovieSceneScriptingFloatChannel, default=-96559, keys [[3686.605,-99603],[7616.2575,-103432],[8513.3925,-96559]]}
       Vector.Z {type=MovieSceneScriptingFloatChannel, default=15634, keys [[3686.605,6710],[7616.2575,12342],[8513.3925,15634]]}
      ]
     }]
   }
  ]
 }
 NS_LimeRock_Ground_Portal {
  id=#45
  possessed_class=NiagaraActor
  tracks [
   {
    type=MovieSceneSpawnTrack
    sections [{
      type=MovieSceneSpawnSection
      channels [None {type=MovieSceneScriptingBoolChannel, default=false, keys [[7164.84,false],[7688.84,true],[8582,false]]}]
     }]
   }
   {
    type=MovieScene3DTransformTrack
    sections [{
      type=MovieScene3DTransformSection
      channels [
       Location.X {type=MovieSceneScriptingDoubleChannel, default=-115767.2656}
       Location.Y {type=MovieSceneScriptingDoubleChannel, default=-103432.7498}
       Location.Z {type=MovieSceneScriptingDoubleChannel, default=12342.2148}
       Rotation.X {type=MovieSceneScriptingDoubleChannel, default=-178.2944}
       Rotation.Y {type=MovieSceneScriptingDoubleChannel, default=-72.6844}
       Rotation.Z {type=MovieSceneScriptingDoubleChannel, default=195.1815}
       Scale.X {type=MovieSceneScriptingDoubleChannel, default=1}
       Scale.Y {type=MovieSceneScriptingDoubleChannel, default=1}
       Scale.Z {type=MovieSceneScriptingDoubleChannel, default=1}
      ]
     }]
   }
  ]
 }
 NiagaraComponent0 {
  id=#46
  parent_id=#47
  possessed_class=NiagaraComponent
  tracks [{
    type=MovieSceneNiagaraFloatParameterTrack
    display_name=Smoke_SpawnRate
    sections [{type=MovieSceneFloatSection, channels [None {type=MovieSceneScriptingFloatChannel, default=150, keys [[7481,150],[7497,150],[7501,0]]}]}]
   }]
 }
 NS_LimeRock_UFODust_ExtraSmoke {
  id=#47
  possessed_class=NiagaraActor
  tracks [
   {
    type=MovieSceneSpawnTrack
    sections [{type=MovieSceneSpawnSection, channels [None {type=MovieSceneScriptingBoolChannel, default=true, keys [[7476,false],[7481,true]]}]}]
   }
   {
    type=MovieScene3DTransformTrack
    sections [{
      type=MovieScene3DTransformSection
      channels [
       Location.X {type=MovieSceneScriptingDoubleChannel, default=-92024.8878, keys [[7481,-92024.8906],[7500,-89610.6719]]}
       Location.Y {type=MovieSceneScriptingDoubleChannel, default=-95906.1084, keys [[7481,-95906.1094],[7500,-94614.1875]]}
       Location.Z {type=MovieSceneScriptingDoubleChannel, default=1063.2823, const [1063.2822,2]}
       Rotation.X {type=MovieSceneScriptingDoubleChannel, default=0}
       Rotation.Y {type=MovieSceneScriptingDoubleChannel, default=0}
       Rotation.Z {type=MovieSceneScriptingDoubleChannel, default=82.6}
       Scale.X {type=MovieSceneScriptingDoubleChannel, default=1}
       Scale.Y {type=MovieSceneScriptingDoubleChannel, default=1}
       Scale.Z {type=MovieSceneScriptingDoubleChannel, default=1}
      ]
     }]
   }
  ]
 }
 NiagaraComponent0 {
  id=#48
  parent_id=#49
  possessed_class=NiagaraComponent
  tracks [{
    type=MovieSceneParticleTrack
    display_name=FX System
    sections [{type=MovieSceneParticleSection, channels [None {type=MovieSceneScriptingParticleChannel, const ["<ParticleKey.TRIGGER: 2>", 1]}]}]
   }]
 }
 NS_LimeRock_LensWind {
  id=#50
  possessed_class=NiagaraActor
  tracks [
   {
    type=MovieSceneSpawnTrack
    sections [{
      type=MovieSceneSpawnSection
      channels [None {type=MovieSceneScriptingBoolChannel, default=false, keys [[8502.705,false],[8557.3275,true],[8877.205,false]]}]
     }]
   }
   {
    type=MovieScene3DTransformTrack
    sections [{
      type=MovieScene3DTransformSection
      channels [
       Location.X {type=MovieSceneScriptingDoubleChannel, default=-87809.4893}
       Location.Y {type=MovieSceneScriptingDoubleChannel, default=-94803.4447}
       Location.Z {type=MovieSceneScriptingDoubleChannel, default=1272}
       Rotation.X {type=MovieSceneScriptingDoubleChannel, default=0}
       Rotation.Y {type=MovieSceneScriptingDoubleChannel, default=0}
       Rotation.Z {type=MovieSceneScriptingDoubleChannel, default=0}
       Scale.X {type=MovieSceneScriptingDoubleChannel, default=1}
       Scale.Y {type=MovieSceneScriptingDoubleChannel, default=1}
       Scale.Z {type=MovieSceneScriptingDoubleChannel, default=1}
      ]
     }]
   }
  ]
 }
 NS_LimeRock_RealityTear_Zap {
  id=#51
  possessed_class=NiagaraActor
  tracks [
   {
    type=MovieSceneSpawnTrack
    sections [{
      type=MovieSceneSpawnSection
      channels [None {type=MovieSceneScriptingBoolChannel, default=false, keys [[7164.84,false],[7568.8375,true],[7815.0975,false]]}]
     }]
   }
   {
    type=MovieScene3DTransformTrack
    sections [{
      type=MovieScene3DTransformSection
      channels [
       Location.X {type=MovieSceneScriptingDoubleChannel, default=-115767.2656}
       Location.Y {type=MovieSceneScriptingDoubleChannel, default=-103432.7498}
       Location.Z {type=MovieSceneScriptingDoubleChannel, default=12342.2148}
       Rotation.X {type=MovieSceneScriptingDoubleChannel, default=-0.5317}
       Rotation.Y {type=MovieSceneScriptingDoubleChannel, default=-17.3077}
       Rotation.Z {type=MovieSceneScriptingDoubleChannel, default=16.968}
       Scale.X {type=MovieSceneScriptingDoubleChannel, default=1}
       Scale.Y {type=MovieSceneScriptingDoubleChannel, default=1}
       Scale.Z {type=MovieSceneScriptingDoubleChannel, default=1}
      ]
     }]
   }
  ]
 }
 NS_LimeRock_Portal_Close {
  id=#52
  possessed_class=NiagaraActor
  tracks [
   {
    type=MovieSceneSpawnTrack
    sections [{
      type=MovieSceneSpawnSection
      channels [None {type=MovieSceneScriptingBoolChannel, default=false, keys [[6614.285,false],[7164.84,true],[7540.2175,false]]}]
     }]
   }
   {
    type=MovieScene3DTransformTrack
    sections [{
      type=MovieScene3DTransformSection
      channels [
       Location.X {type=MovieSceneScriptingDoubleChannel, default=-101065.1944}
       Location.Y {type=MovieSceneScriptingDoubleChannel, default=-99603.4429}
       Location.Z {type=MovieSceneScriptingDoubleChannel, default=6710.2577}
       Rotation.X {type=MovieSceneScriptingDoubleChannel, default=1.7056}
       Rotation.Y {type=MovieSceneScriptingDoubleChannel, default=-17.3156}
       Rotation.Z {type=MovieSceneScriptingDoubleChannel, default=15.1815}
       Scale.X {type=MovieSceneScriptingDoubleChannel, default=1}
       Scale.Y {type=MovieSceneScriptingDoubleChannel, default=1}
       Scale.Z {type=MovieSceneScriptingDoubleChannel, default=1}
      ]
     }]
   }
  ]
 }
 NS_LimeRock_MonsterSplash {
  id=#22
  possessed_class=NiagaraActor
  tracks [
   {
    type=MovieSceneSpawnTrack
    sections [{
      type=MovieSceneSpawnSection
      channels [None {type=MovieSceneScriptingBoolChannel, default=false, keys [[8036,false],[8166,true],[8593.055,false]]}]
     }]
   }
   {
    type=MovieScene3DTransformTrack
    sections [{
      type=MovieScene3DTransformSection
      channels [
       Location.X {type=MovieSceneScriptingDoubleChannel, default=-125152.1866}
       Location.Y {type=MovieSceneScriptingDoubleChannel, default=-95517.4496}
       Location.Z {type=MovieSceneScriptingDoubleChannel, default=1762.4003}
       Rotation.X {type=MovieSceneScriptingDoubleChannel, default=0}
       Rotation.Y {type=MovieSceneScriptingDoubleChannel, default=0}
       Rotation.Z {type=MovieSceneScriptingDoubleChannel, default=0}
       Scale.X {type=MovieSceneScriptingDoubleChannel, default=1}
       Scale.Y {type=MovieSceneScriptingDoubleChannel, default=1}
       Scale.Z {type=MovieSceneScriptingDoubleChannel, default=1}
      ]
     }]
   }
  ]
 }
 NS_LimeRock_Leaf {
  id=#40
  possessed_class=NiagaraActor
  tracks [
   {
    type=MovieSceneSpawnTrack
    sections [{
      type=MovieSceneSpawnSection
      channels [None {type=MovieSceneScriptingBoolChannel, default=false, keys [[8557,false],[8558,true],[9045,true],[9046,false]]}]
     }]
   }
   {
    type=MovieScene3DTransformTrack
    sections [{
      type=MovieScene3DTransformSection
      channels [
       Location.X {type=MovieSceneScriptingDoubleChannel, default=-95190.9203}
       Location.Y {type=MovieSceneScriptingDoubleChannel, default=-95297.3748}
       Location.Z {type=MovieSceneScriptingDoubleChannel, default=1590.4032}
       Rotation.X {type=MovieSceneScriptingDoubleChannel, default=0}
       Rotation.Y {type=MovieSceneScriptingDoubleChannel, default=0}
       Rotation.Z {type=MovieSceneScriptingDoubleChannel, default=0}
       Scale.X {type=MovieSceneScriptingDoubleChannel, default=1}
       Scale.Y {type=MovieSceneScriptingDoubleChannel, default=1}
       Scale.Z {type=MovieSceneScriptingDoubleChannel, default=1}
      ]
     }]
   }
  ]
 }
 NiagaraComponent0 {
  id=#53
  parent_id=#54
  possessed_class=NiagaraComponent
  tracks [
   {
    type=MovieSceneNiagaraFloatParameterTrack
    display_name=Ribbon_SpawnRate
    sections [{type=MovieSceneFloatSection, channels [None {type=MovieSceneScriptingFloatChannel, default=40, keys [[7589,40],[7617,20]]}]}]
   }
   {
    type=MovieSceneNiagaraFloatParameterTrack
    display_name=Ribbon_OpacityMult
    sections [{
      type=MovieSceneFloatSection
      channels [None {type=MovieSceneScriptingFloatChannel, default=0.3, keys [[7451,0],[7469,1],[7569,1],[7593,0.3]]}]
     }]
   }
  ]
 }
 NiagaraComponent0 {id=#55, parent_id=#45, possessed_class=NiagaraComponent, tracks []}
 NS_LimeRock_MonsterSplash_Small {
  id=#37
  possessed_class=NiagaraActor
  tracks [
   {
    type=MovieSceneSpawnTrack
    sections [{
      type=MovieSceneSpawnSection
      channels [None {type=MovieSceneScriptingBoolChannel, default=false, keys [[8046,false],[8060,true],[8212,true],[8213,false]]}]
     }]
   }
   {
    type=MovieScene3DTransformTrack
    sections [{
      type=MovieScene3DTransformSection
      channels [
       Location.X {type=MovieSceneScriptingDoubleChannel, default=-117454.1967}
       Location.Y {type=MovieSceneScriptingDoubleChannel, default=-103956.8684}
       Location.Z {type=MovieSceneScriptingDoubleChannel, default=1856.1605}
       Rotation.X {type=MovieSceneScriptingDoubleChannel, default=0}
       Rotation.Y {type=MovieSceneScriptingDoubleChannel, default=0}
       Rotation.Z {type=MovieSceneScriptingDoubleChannel, default=0}
       Scale.X {type=MovieSceneScriptingDoubleChannel, default=1}
       Scale.Y {type=MovieSceneScriptingDoubleChannel, default=1}
       Scale.Z {type=MovieSceneScriptingDoubleChannel, default=1}
      ]
     }]
   }
  ]
 }
 NS_LimeRock_Treehouse_Energy_Supressed {
  id=#34
  possessed_class=NiagaraActor
  tracks [
   {
    type=MovieSceneSpawnTrack
    sections [{
      type=MovieSceneSpawnSection
      channels [None {type=MovieSceneScriptingBoolChannel, default=false, keys [[6260.92,false],[6481.92,true],[8878.92,false]]}]
     }]
   }
   {
    type=MovieScene3DTransformTrack
    sections [{
      type=MovieScene3DTransformSection
      channels [
       Location.X {type=MovieSceneScriptingDoubleChannel, default=-94411.9644, keys [[6702.92,-94400],[7047.92,-94900]]}
       Location.Y {type=MovieSceneScriptingDoubleChannel, default=-95308.1747, const [-95300,2]}
       Location.Z {type=MovieSceneScriptingDoubleChannel, default=2746.4608, keys [[6702.92,3000],[7047.92,1800]]}
       Rotation.X {type=MovieSceneScriptingDoubleChannel, default=0, const [0,2]}
       Rotation.Y {type=MovieSceneScriptingDoubleChannel, default=0, const [0,2]}
       Rotation.Z {type=MovieSceneScriptingDoubleChannel, default=0, const [0,2]}
       Scale.X {type=MovieSceneScriptingDoubleChannel, default=1, const [1,2]}
       Scale.Y {type=MovieSceneScriptingDoubleChannel, default=1, const [1,2]}
       Scale.Z {type=MovieSceneScriptingDoubleChannel, default=1, const [1,2]}
      ]
     }]
   }
  ]
 }
 NiagaraComponent0 {id=#56, parent_id=#50, possessed_class=NiagaraComponent, tracks []}
 NS_LimeRock_Portal_Shockwave_02 {
  id=#57
  possessed_class=NiagaraActor
  tracks [
   {
    type=MovieScene3DTransformTrack
    sections [{
      type=MovieScene3DTransformSection
      channels [
       Location.X {type=MovieSceneScriptingDoubleChannel, default=-115767.2656}
       Location.Y {type=MovieSceneScriptingDoubleChannel, default=-103432.7498}
       Location.Z {type=MovieSceneScriptingDoubleChannel, default=12342.2148}
       Rotation.X {type=MovieSceneScriptingDoubleChannel, default=0}
       Rotation.Y {type=MovieSceneScriptingDoubleChannel, default=0}
       Rotation.Z {type=MovieSceneScriptingDoubleChannel, default=11.6248}
       Scale.X {type=MovieSceneScriptingDoubleChannel, default=1}
       Scale.Y {type=MovieSceneScriptingDoubleChannel, default=1}
       Scale.Z {type=MovieSceneScriptingDoubleChannel, default=1}
      ]
     }]
   }
   {
    type=MovieSceneSpawnTrack
    sections [{
      type=MovieSceneSpawnSection
      channels [None {type=MovieSceneScriptingBoolChannel, default=false, keys [[6108,false],[7688.84,true],[8260.305,false]]}]
     }]
   }
  ]
 }
 NS_LimeRock_Treehouse_FX_2 {
  id=#58
  possessed_class=NiagaraActor
  tracks [{
    type=MovieScene3DTransformTrack
    sections [{
      type=MovieScene3DTransformSection
      channels [
       Location.X {type=MovieSceneScriptingDoubleChannel, default=-95190.9203, const [-95190.9219,1]}
       Location.Y {type=MovieSceneScriptingDoubleChannel, default=-95297.3748, const [-95297.375,1]}
       Location.Z {type=MovieSceneScriptingDoubleChannel, default=1590.4032, const [1590.4032,1]}
       Rotation.X {type=MovieSceneScriptingDoubleChannel, default=0, const [0,1]}
       Rotation.Y {type=MovieSceneScriptingDoubleChannel, default=0, const [0,1]}
       Rotation.Z {type=MovieSceneScriptingDoubleChannel, default=0, const [0,1]}
       Scale.X {type=MovieSceneScriptingDoubleChannel, default=1, const [1,1]}
       Scale.Y {type=MovieSceneScriptingDoubleChannel, default=1, const [1,1]}
       Scale.Z {type=MovieSceneScriptingDoubleChannel, default=1, const [1,1]}
      ]
     }]
   }]
 }
 NS_LimeRock_HorizonLightning {
  id=#8
  possessed_class=NiagaraActor
  tracks [
   {
    type=MovieSceneSpawnTrack
    sections [{
      type=MovieSceneSpawnSection
      channels [None {type=MovieSceneScriptingBoolChannel, default=false, keys [[7308.9425,false],[7899.9925,true],[9341.335,false]]}]
     }]
   }
   {
    type=MovieScene3DTransformTrack
    sections [{
      type=MovieScene3DTransformSection
      channels [
       Location.X {type=MovieSceneScriptingDoubleChannel, default=-129309.4606}
       Location.Y {type=MovieSceneScriptingDoubleChannel, default=-97819.2812}
       Location.Z {type=MovieSceneScriptingDoubleChannel, default=2309.4353}
       Rotation.X {type=MovieSceneScriptingDoubleChannel, default=0}
       Rotation.Y {type=MovieSceneScriptingDoubleChannel, default=0}
       Rotation.Z {type=MovieSceneScriptingDoubleChannel, default=3.4613}
       Scale.X {type=MovieSceneScriptingDoubleChannel, default=1}
       Scale.Y {type=MovieSceneScriptingDoubleChannel, default=1}
       Scale.Z {type=MovieSceneScriptingDoubleChannel, default=1}
      ]
     }]
   }
  ]
 }
 NS_LImeRock_AR_TracersJones {
  id=#11
  possessed_class=NiagaraActor
  tracks [
   {
    type=MovieSceneSpawnTrack
    sections [{
      type=MovieSceneSpawnSection
      channels [None {type=MovieSceneScriptingBoolChannel, default=false, keys [[8380.1425,false],[9051.3475,true],[9224.545,false]]}]
     }]
   }
   {
    type=MovieScene3DTransformTrack
    sections [{
      type=MovieScene3DTransformSection
      channels [
       Location.X {
        type=MovieSceneScriptingDoubleChannel
        default=-92707.8659
        keys [[9051.3475,-95288.7812],[9097.7625,-95684.8672],[9123.6525,-96063.4609],[9144.225,-96063.4609],[9168.4875,-93704.5547]]
       }
       Location.Y {
        type=MovieSceneScriptingDoubleChannel
        default=-96184.9075
        keys [[9051.3475,-96910.2734],[9097.7625,-97070.4609],[9123.6525,-97469.4609],[9144.225,-97469.4609],[9168.4875,-96511.4453]]
       }
       Location.Z {
        type=MovieSceneScriptingDoubleChannel
        default=3043.8228
        keys [[9051.3475,3857.9526],[9097.7625,3777.4946],[9123.6525,4276.1475],[9144.225,4276.1475],[9168.4875,4833.5088]]
       }
       Rotation.X {
        type=MovieSceneScriptingDoubleChannel
        default=0.4657
        keys [[9051.3475,-7.9337],[9097.7625,-2.5501],[9123.6525,-3.4327],[9144.225,-3.4327],[9168.4875,18.9778]]
       }
       Rotation.Y {
        type=MovieSceneScriptingDoubleChannel
        default=-30.1813
        keys [[9051.3475,17.3504],[9097.7625,8.7849],[9123.6525,8.4812],[9144.225,8.4812],[9168.4875,-19.8099]]
       }
       Rotation.Z {
        type=MovieSceneScriptingDoubleChannel
        default=31.6971
        keys [[9051.3475,-1.6048],[9097.7625,-19.837],[9123.6525,-25.7118],[9144.225,-25.7118],[9168.4875,-17.8542]]
       }
       Scale.X {type=MovieSceneScriptingDoubleChannel, default=1, const [1,1]}
       Scale.Y {type=MovieSceneScriptingDoubleChannel, default=1, const [1,1]}
       Scale.Z {type=MovieSceneScriptingDoubleChannel, default=1, const [1,1]}
      ]
     }]
   }
  ]
 }
 NiagaraComponent0 {
  id=#59
  parent_id=#60
  possessed_class=NiagaraComponent
  tracks [{
    type=MovieSceneParticleTrack
    display_name=FX System
    sections [{type=MovieSceneParticleSection, channels [None {type=MovieSceneScriptingParticleChannel, const ["<ParticleKey.TRIGGER: 2>", 1]}]}]
   }]
 }
 NiagaraComponent0 {
  id=#61
  parent_id=#1
  possessed_class=NiagaraComponent
  tracks [
   {
    type=MovieSceneNiagaraFloatParameterTrack
    display_name=V_Mask
    sections [{
      type=MovieSceneFloatSection
      channels [None {type=MovieSceneScriptingFloatChannel, default=120, keys [[6707.925,120],[6786.925,7],[7051.925,3]]}]
     }]
   }
   {
    type=MovieSceneNiagaraFloatParameterTrack
    display_name=Opacity
    sections [{
      type=MovieSceneFloatSection
      channels [None {type=MovieSceneScriptingFloatChannel, default=0.5, keys [[6707.925,0],[6836.925,0.5],[7051.925,0.65],[7082.925,0]]}]
     }]
   }
  ]
 }
 NiagaraComponent0 {
  id=#62
  parent_id=#19
  possessed_class=NiagaraComponent
  tracks [
   {
    type=MovieSceneNiagaraFloatParameterTrack
    display_name=Opacity
    sections [{
      type=MovieSceneFloatSection
      channels [None {type=MovieSceneScriptingFloatChannel, default=1, keys [[6722,0],[6783,0.3],[7049,0.3],[7064,0]]}]
     }]
   }
   {
    type=MovieSceneNiagaraFloatParameterTrack
    display_name=V_Mask
    sections [{type=MovieSceneFloatSection, channels [None {type=MovieSceneScriptingFloatChannel, default=600, keys [[6946,300],[6994,600]]}]}]
   }
   {
    type=MovieSceneNiagaraFloatParameterTrack
    display_name=Decal_Size
    sections [{type=MovieSceneFloatSection, channels [None {type=MovieSceneScriptingFloatChannel, default=1500, keys [[6982,3000],[6992,1500]]}]}]
   }
  ]
 }
 NS_LR_Treehouse_Pulse_Small {
  id=#49
  possessed_class=NiagaraActor
  tracks [
   {
    type=MovieSceneSpawnTrack
    sections [{
      type=MovieSceneSpawnSection
      channels [None {type=MovieSceneScriptingBoolChannel, default=false, keys [[3953,false],[4583,true],[4744,false]]}]
     }]
   }
   {
    type=MovieScene3DTransformTrack
    sections [{
      type=MovieScene3DTransformSection
      channels [
       Location.X {type=MovieSceneScriptingDoubleChannel, default=-95171.9329}
       Location.Y {type=MovieSceneScriptingDoubleChannel, default=-95296}
       Location.Z {type=MovieSceneScriptingDoubleChannel, default=2400}
       Rotation.X {type=MovieSceneScriptingDoubleChannel, default=0}
       Rotation.Y {type=MovieSceneScriptingDoubleChannel, default=-10}
       Rotation.Z {type=MovieSceneScriptingDoubleChannel, default=0}
       Scale.X {type=MovieSceneScriptingDoubleChannel, default=1}
       Scale.Y {type=MovieSceneScriptingDoubleChannel, default=1}
       Scale.Z {type=MovieSceneScriptingDoubleChannel, default=1}
      ]
     }]
   }
  ]
 }
 NS_LimeRock_Dust_Kickup {
  id=#63
  possessed_class=NiagaraActor
  tracks [
   {
    type=MovieSceneSpawnTrack
    sections [{
      type=MovieSceneSpawnSection
      channels [None {type=MovieSceneScriptingBoolChannel, default=false, keys [[5071,false],[5925,true],[6195.52,false]]}]
     }]
   }
   {
    type=MovieScene3DTransformTrack
    sections [{
      type=MovieScene3DTransformSection
      channels [
       Location.X {type=MovieSceneScriptingDoubleChannel, default=-94282.4154}
       Location.Y {type=MovieSceneScriptingDoubleChannel, default=-95479.4086}
       Location.Z {type=MovieSceneScriptingDoubleChannel, default=2157.555}
       Rotation.X {type=MovieSceneScriptingDoubleChannel, default=0}
       Rotation.Y {type=MovieSceneScriptingDoubleChannel, default=0}
       Rotation.Z {type=MovieSceneScriptingDoubleChannel, default=0}
       Scale.X {type=MovieSceneScriptingDoubleChannel, default=1}
       Scale.Y {type=MovieSceneScriptingDoubleChannel, default=1}
       Scale.Z {type=MovieSceneScriptingDoubleChannel, default=1}
      ]
     }]
   }
  ]
 }
 NS_LimeRock_RiftIn_Exit {
  id=#64
  possessed_class=NiagaraActor
  tracks [
   {
    type=MovieSceneSpawnTrack
    sections [{
      type=MovieSceneSpawnSection
      channels [None {type=MovieSceneScriptingBoolChannel, default=false, keys [[3812,false],[4200.6575,true],[4398,false]]}]
     }]
   }
   {
    type=MovieScene3DTransformTrack
    sections [{
      type=MovieScene3DTransformSection
      channels [
       Location.X {type=MovieSceneScriptingDoubleChannel, default=-92987.5289}
       Location.Y {type=MovieSceneScriptingDoubleChannel, default=-90983.0746}
       Location.Z {type=MovieSceneScriptingDoubleChannel, default=2927.7}
       Rotation.X {type=MovieSceneScriptingDoubleChannel, default=0}
       Rotation.Y {type=MovieSceneScriptingDoubleChannel, default=-4.0492}
       Rotation.Z {type=MovieSceneScriptingDoubleChannel, default=-57.9377}
       Scale.X {type=MovieSceneScriptingDoubleChannel, default=1}
       Scale.Y {type=MovieSceneScriptingDoubleChannel, default=1}
       Scale.Z {type=MovieSceneScriptingDoubleChannel, default=1}
      ]
     }]
   }
  ]
 }
 NS_LR_Treehouse_Pulse_Medium {
  id=#60
  possessed_class=NiagaraActor
  tracks [
   {
    type=MovieSceneSpawnTrack
    sections [{
      type=MovieSceneSpawnSection
      channels [None {type=MovieSceneScriptingBoolChannel, default=false, keys [[3953,false],[4981.5425,true],[5224.5425,false]]}]
     }]
   }
   {
    type=MovieScene3DTransformTrack
    sections [{
      type=MovieScene3DTransformSection
      channels [
       Location.X {type=MovieSceneScriptingDoubleChannel, default=-95171.9329}
       Location.Y {type=MovieSceneScriptingDoubleChannel, default=-95296}
       Location.Z {type=MovieSceneScriptingDoubleChannel, default=2400}
       Rotation.X {type=MovieSceneScriptingDoubleChannel, default=0}
       Rotation.Y {type=MovieSceneScriptingDoubleChannel, default=-10}
       Rotation.Z {type=MovieSceneScriptingDoubleChannel, default=0}
       Scale.X {type=MovieSceneScriptingDoubleChannel, default=1}
       Scale.Y {type=MovieSceneScriptingDoubleChannel, default=1}
       Scale.Z {type=MovieSceneScriptingDoubleChannel, default=1}
      ]
     }]
   }
  ]
 }
 NS_LimeRock_UFO_WindTrail {
  id=#54
  possessed_class=NiagaraActor
  tracks [
   {
    type=MovieSceneSpawnTrack
    sections [{
      type=MovieSceneSpawnSection
      channels [None {type=MovieSceneScriptingBoolChannel, default=false, keys [[7450,false],[7451,true],[7910,true],[7913,false]]}]
     }]
   }
   {
    type=MovieScene3DTransformTrack
    sections [{
      type=MovieScene3DTransformSection
      channels [
       Location.X {type=MovieSceneScriptingDoubleChannel, default=0}
       Location.Y {type=MovieSceneScriptingDoubleChannel, default=0}
       Location.Z {type=MovieSceneScriptingDoubleChannel, default=0}
       Rotation.X {type=MovieSceneScriptingDoubleChannel, default=0}
       Rotation.Y {type=MovieSceneScriptingDoubleChannel, default=0}
       Rotation.Z {type=MovieSceneScriptingDoubleChannel, default=0}
       Scale.X {type=MovieSceneScriptingDoubleChannel, default=1}
       Scale.Y {type=MovieSceneScriptingDoubleChannel, default=1}
       Scale.Z {type=MovieSceneScriptingDoubleChannel, default=1}
      ]
     }]
   }
   {type=MovieScene3DAttachTrack, sections [{type=MovieScene3DAttachSection, end_frame=14400, start_frame=7451}]}
  ]
 }
 NS_LimeRock_Barrier_Fence {
  id=#65
  parent_id=#41
  possessed_class=NiagaraComponent
  tracks [
   {
    type=MovieSceneNiagaraFloatParameterTrack
    display_name=Opacity
    sections [{
      type=MovieSceneFloatSection
      channels [None {
        type=MovieSceneScriptingFloatChannel
        default=0
        keys [[4230.3275,0],[4235.47,0.5],[4272.35,0],[4352.56,0],[4357.7025,0.5],[4394.5825,0],[5926,0],[5941.2,1],[6018.3425,0]]
       }]
     }]
   }
   {
    type=MovieSceneNiagaraFloatParameterTrack
    display_name=Ripple
    sections [{
      type=MovieSceneFloatSection
      channels [None {type=MovieSceneScriptingFloatChannel, default=1, keys [[4230.3275,0],[4286.325,1],[4352.56,0],[4407.9225,1],[5926,0],[6003.49,1]]}]
     }]
   }
   {
    type=MovieSceneNiagaraVectorParameterTrack
    display_name=RipplePosition
    sections [{
      type=MovieSceneFloatVectorSection
      channels [
       Vector.X {type=MovieSceneScriptingFloatChannel, default=-91600, keys [[4230.3275,-91195],[4352.56,-91326],[5892.845,-91600]]}
       Vector.Y {type=MovieSceneScriptingFloatChannel, default=-94900, keys [[4230.3275,-94334],[4352.56,-94560],[5892.845,-94900]]}
       Vector.Z {type=MovieSceneScriptingFloatChannel, default=1500, keys [[4230.3275,1695],[4352.56,1632],[5892.845,1500]]}
      ]
     }]
   }
   {
    type=MovieSceneNiagaraFloatParameterTrack
    display_name=BigRipple
    sections [{type=MovieSceneFloatSection, channels [None {type=MovieSceneScriptingFloatChannel, default=1, keys [[4230.3275,0],[5892.845,1]]}]}]
   }
  ]
 }
 NS_LimeRock_Boss_Roar {
  id=#66
  possessed_class=NiagaraActor
  tracks [
   {
    type=MovieSceneSpawnTrack
    sections [{
      type=MovieSceneSpawnSection
      channels [None {type=MovieSceneScriptingBoolChannel, default=false, keys [[8497.705,false],[8556.12,true],[8782.1025,false]]}]
     }]
   }
   {
    type=MovieScene3DTransformTrack
    sections [{
      type=MovieScene3DTransformSection
      channels [
       Location.X {type=MovieSceneScriptingDoubleChannel, default=-109771.0378}
       Location.Y {type=MovieSceneScriptingDoubleChannel, default=-96559.5618}
       Location.Z {type=MovieSceneScriptingDoubleChannel, default=15634.4003}
       Rotation.X {type=MovieSceneScriptingDoubleChannel, default=-1.1082}
       Rotation.Y {type=MovieSceneScriptingDoubleChannel, default=-22.9081}
       Rotation.Z {type=MovieSceneScriptingDoubleChannel, default=6.5242}
       Scale.X {type=MovieSceneScriptingDoubleChannel, default=1}
       Scale.Y {type=MovieSceneScriptingDoubleChannel, default=1}
       Scale.Z {type=MovieSceneScriptingDoubleChannel, default=1}
      ]
     }]
   }
  ]
 }
 NS_LimeRock_RiftDevice_Beam {
  id=#67
  possessed_class=NiagaraActor
  tracks [
   {
    type=MovieSceneSpawnTrack
    sections [{
      type=MovieSceneSpawnSection
      channels [None {type=MovieSceneScriptingBoolChannel, default=false, keys [[6447.04,false],[6693.925,true],[7078.3675,false]]}]
     }]
   }
   {
    type=MovieScene3DTransformTrack
    sections [{
      type=MovieScene3DTransformSection
      channels [
       Location.X {type=MovieSceneScriptingDoubleChannel, default=0}
       Location.Y {type=MovieSceneScriptingDoubleChannel, default=0}
       Location.Z {type=MovieSceneScriptingDoubleChannel, default=0}
       Rotation.X {type=MovieSceneScriptingDoubleChannel, default=0}
       Rotation.Y {type=MovieSceneScriptingDoubleChannel, default=0}
       Rotation.Z {type=MovieSceneScriptingDoubleChannel, default=90}
       Scale.X {type=MovieSceneScriptingDoubleChannel, default=1}
       Scale.Y {type=MovieSceneScriptingDoubleChannel, default=1}
       Scale.Z {type=MovieSceneScriptingDoubleChannel, default=1}
      ]
     }]
   }
   {type=MovieScene3DAttachTrack, sections [{type=MovieScene3DAttachSection, end_frame=12990, start_frame=6665}]}
  ]
 }
 Tentacle_LimeRock_Creature {
  id=#68
  parent_id=#20
  possessed_class=SkeletalMeshComponent
  tracks [{
    type=MovieSceneVisibilityTrack
    sections [{
      type=MovieSceneVisibilitySection
      channels [None {type=MovieSceneScriptingBoolChannel, default=false, keys [[4511.5925,false],[4989.11,true],[7329.7475,false]]}]
     }]
   }]
 }
 BP_LimeRock_Smoke_Env_Spline {
  id=#69
  possessed_class=BP_LimeRock_Smoke_Env_Spline_C
  tracks [
   {
    type=MovieSceneSpawnTrack
    sections [{type=MovieSceneSpawnSection, channels [None {type=MovieSceneScriptingBoolChannel, default=false, keys [[7865.3275,false],[8228.4,true]]}]}]
   }
   {
    type=MovieScene3DTransformTrack
    sections [{
      type=MovieScene3DTransformSection
      channels [
       Location.X {type=MovieSceneScriptingDoubleChannel, default=-100417.8386}
       Location.Y {type=MovieSceneScriptingDoubleChannel, default=-95388.3712}
       Location.Z {type=MovieSceneScriptingDoubleChannel, default=448.5651}
       Rotation.X {type=MovieSceneScriptingDoubleChannel, default=0}
       Rotation.Y {type=MovieSceneScriptingDoubleChannel, default=0}
       Rotation.Z {type=MovieSceneScriptingDoubleChannel, default=0}
       Scale.X {type=MovieSceneScriptingDoubleChannel, default=1}
       Scale.Y {type=MovieSceneScriptingDoubleChannel, default=1}
       Scale.Z {type=MovieSceneScriptingDoubleChannel, default=1}
      ]
     }]
   }
  ]
 }
 NiagaraComponent0 {
  id=#70
  parent_id=#67
  possessed_class=NiagaraComponent
  tracks [
   {
    type=MovieSceneNiagaraVectorParameterTrack
    display_name=EndPoint
    sections [{
      type=MovieSceneFloatVectorSection
      channels [
       Vector.X {type=MovieSceneScriptingFloatChannel, default=-94700.6328, thinned_from=3, keys [[6707.925,-94400],[7052.925,-94900]]}
       Vector.Y {type=MovieSceneScriptingFloatChannel, default=-95300}
       Vector.Z {type=MovieSceneScriptingFloatChannel, default=1800, keys [[6707.925,3000],[7052.925,1800]]}
      ]
     }]
   }
   {
    type=MovieSceneParticleTrack
    display_name=FX System
    sections [{
      type=MovieSceneParticleSection
      channels [None {type=MovieSceneScriptingParticleChannel, keys [[6693.925, "<ParticleKey.ACTIVATE: 0>"], [7064, "<ParticleKey.DEACTIVATE: 1>"]]}]
     }]
   }
   {
    type=MovieSceneNiagaraFloatParameterTrack
    display_name=Opacity
    sections [{type=MovieSceneFloatSection, channels [None {type=MovieSceneScriptingFloatChannel, default=0, keys [[7052.925,1],[7077.6925,0]]}]}]
   }
  ]
 }
 NiagaraComponent0 {
  id=#71
  parent_id=#58
  possessed_class=NiagaraComponent
  tracks [
   {
    type=MovieSceneNiagaraFloatParameterTrack
    display_name=Lightning Spawn Rate
    sections [{
      type=MovieSceneFloatSection
      channels [None {
        type=MovieSceneScriptingFloatChannel
        default=5
        keys [[5572,2],[5845,5],[5880,2],[6691,2],[6795,20],[6835,19.1558],[7118,0],[7299,0.5],[7516,0.5],[7563,5],[7605.895,5],[7612.895,0.5],[7627.3475,2],[7639.895,0]]
       }]
     }]
   }
   {
    type=MovieSceneNiagaraFloatParameterTrack
    display_name=Fade out
    sections [{
      type=MovieSceneFloatSection
      channels [None {
        type=MovieSceneScriptingFloatChannel
        default=0.65
        keys [[3600,0.3],[4532.1075,0.3],[4583,0.65],[4606,0.5],[4894.5425,0.5],[4981.5425,0.85],[5025.5425,0.7],[5572,1],[5825,2],[5853,1],[6691,1],[6835,1.2613],[7078,2],[7105,0],[7526.72,0],[7571,1],[7598,0]]
       }]
     }]
   }
   {
    type=MovieSceneNiagaraFloatParameterTrack
    display_name=ColorShifting
    sections [{
      type=MovieSceneFloatSection
      channels [None {type=MovieSceneScriptingFloatChannel, default=0, keys [[6764,0],[6835,0.1303],[7078,1],[7571,1],[7614.6275,0]]}]
     }]
   }
   {
    type=MovieSceneNiagaraFloatParameterTrack
    display_name=TreeEnergy_Brightness
    sections [{
      type=MovieSceneFloatSection
      channels [None {
        type=MovieSceneScriptingFloatChannel
        default=7
        thinned_from=13
        keys [[3600,0.3],[4532.1075,0.3],[4583,1],[4606,0.5],[4894.5425,0.5],[4981.5425,1],[5089.5425,0.7],[5572,1],[5845,7],[5880,1],[7541.8375,1],[7605.895,5]]
       }]
     }]
   }
   {
    type=MovieSceneNiagaraColorParameterTrack
    display_name=OverallColorSwitch
    sections [{
      type=MovieSceneColorSection
      channels [
       Color.R {type=MovieSceneScriptingFloatChannel, default=0.2, keys [[6711,0.9357],[6835,0.7725],[7118,0.2],[7541.8375,0.2009],[7603.895,0.9357]]}
       Color.G {type=MovieSceneScriptingFloatChannel, default=0.3, keys [[6711,0.3786],[6835,0.3611],[7118,0.3],[7541.8375,0.3001],[7603.895,0.3786]]}
       Color.B {type=MovieSceneScriptingFloatChannel, default=1, const [1,5]}
       Color.A {type=MovieSceneScriptingFloatChannel, default=1, const [1,3]}
      ]
     }]
   }
   {
    type=MovieSceneNiagaraFloatParameterTrack
    display_name=Energy_SpawnRate
    sections [{
      type=MovieSceneFloatSection
      channels [None {
        type=MovieSceneScriptingFloatChannel
        default=2
        thinned_from=9
        keys [[5571,0],[5572,1],[5824,3],[5825,0],[7571,0],[7572,2],[7602.895,5],[7605.895,0]]
       }]
     }]
   }
   {
    type=MovieSceneNiagaraFloatParameterTrack
    display_name=Energy_SpawnTimeMult
    sections [{
      type=MovieSceneFloatSection
      channels [None {
        type=MovieSceneScriptingFloatChannel
        default=0.3
        thinned_from=7
        keys [[5571,3],[5572,3],[5824,0.3],[6835,0.4263],[7542.8375,0.5],[7605.895,0.25]]
       }]
     }]
   }
   {
    type=MovieSceneNiagaraFloatParameterTrack
    display_name=TreeTrunk_LightningRate
    sections [{
      type=MovieSceneFloatSection
      channels [None {type=MovieSceneScriptingFloatChannel, default=5, thinned_from=6, keys [[5571,1],[5824,5],[5880,1],[7541.8375,1],[7605.895,3]]}]
     }]
   }
   {
    type=MovieSceneNiagaraFloatParameterTrack
    display_name=Light_Intensity
    sections [{
      type=MovieSceneFloatSection
      channels [None {
        type=MovieSceneScriptingFloatChannel
        default=1.5
        thinned_from=17
        keys [[3600,0.3],[4532.1075,0.3],[4583,0.65],[4606,0.5],[4894.5425,0.5],[4981.5425,0.85],[5025.5425,0.7],[5572,1],[5845,4],[5880,1.5],[7004,1.5],[7120.1175,0],[7541,0],[7571,2],[7638,2],[7666,0]]
       }]
     }]
   }
   {
    type=MovieSceneNiagaraFloatParameterTrack
    display_name=Light_Radius
    sections [{
      type=MovieSceneFloatSection
      channels [None {type=MovieSceneScriptingFloatChannel, default=1, thinned_from=8, keys [[5572,1],[5845,2],[5880,1],[7541,1],[7571,3],[7638,1.5],[7667,0]]}]
     }]
   }
   {
    type=MovieSceneNiagaraFloatParameterTrack
    display_name=LightColor_Switch
    sections [{
      type=MovieSceneFloatSection
      channels [None {type=MovieSceneScriptingFloatChannel, default=0, keys [[6786,0],[6835,0.1289],[7004,1],[7571,1],[7616,0]]}]
     }]
   }
   {
    type=MovieSceneNiagaraFloatParameterTrack
    display_name=TreeHouse_Spline_VerticalOpacityPower
    sections [{
      type=MovieSceneFloatSection
      channels [None {
        type=MovieSceneScriptingFloatChannel
        default=0.6
        keys [[3600,3],[4532.1075,3],[4583,1],[4691,2],[4894.5425,2],[4981.5425,0.6],[5089.5425,1],[5563,1],[5918,0.6],[6835,0.6]]
       }]
     }]
   }
   {
    type=MovieSceneNiagaraFloatParameterTrack
    display_name=TreeHouse_Spline_ColorLerp
    sections [{
      type=MovieSceneFloatSection
      channels [None {type=MovieSceneScriptingFloatChannel, default=0, keys [[6835,0],[6906,0],[7033,1],[7571,1],[7598,0]]}]
     }]
   }
   {
    type=MovieSceneNiagaraFloatParameterTrack
    display_name=GroundDecal_Intensity
    sections [{
      type=MovieSceneFloatSection
      channels [None {
        type=MovieSceneScriptingFloatChannel
        default=0
        thinned_from=12
        keys [[3600,0.2],[4532,0.2],[4583,0.5],[4606,0.4],[4895,0.4],[4982,0.5],[5026,0.4],[5825,0.5],[5880,0.3],[7004,0.3],[7120.1175,0]]
       }]
     }]
   }
   {
    type=MovieSceneNiagaraIntegerParameterTrack
    display_name=TreeHouse_TopEnergy_MeshDecal_Spawn
    sections [{
      type=MovieSceneIntegerSection
      channels [None {type=MovieSceneScriptingIntegerChannel, default=1, keys [[3600,0],[4567,0],[4571,1],[6613,1],[6617,0]]}]
     }]
   }
  ]
 }
 NS_LimeRock_SmokeMonsterMain2 {
  id=#30
  possessed_class=NiagaraActor
  tracks [
   {
    type=MovieSceneSpawnTrack
    sections [{type=MovieSceneSpawnSection, channels [None {type=MovieSceneScriptingBoolChannel, default=false, keys [[7164.84,false],[7694.84,true]]}]}]
   }
   {
    type=MovieScene3DTransformTrack
    sections [{
      type=MovieScene3DTransformSection
      channels [
       Location.X {type=MovieSceneScriptingDoubleChannel, default=0}
       Location.Y {type=MovieSceneScriptingDoubleChannel, default=0}
       Location.Z {type=MovieSceneScriptingDoubleChannel, default=0}
       Rotation.X {type=MovieSceneScriptingDoubleChannel, default=0}
       Rotation.Y {type=MovieSceneScriptingDoubleChannel, default=0}
       Rotation.Z {type=MovieSceneScriptingDoubleChannel, default=0}
       Scale.X {type=MovieSceneScriptingDoubleChannel, default=1}
       Scale.Y {type=MovieSceneScriptingDoubleChannel, default=1}
       Scale.Z {type=MovieSceneScriptingDoubleChannel, default=1}
      ]
     }]
   }
   {type=MovieScene3DAttachTrack, sections [{type=MovieScene3DAttachSection, end_frame=13451, start_frame=7694}]}
  ]
 }
 NiagaraComponent0 {
  id=#72
  parent_id=#13
  possessed_class=NiagaraComponent
  tracks [
   {
    type=MovieSceneNiagaraFloatParameterTrack
    display_name=ColorShift
    sections [{
      type=MovieSceneFloatSection
      channels [None {
        type=MovieSceneScriptingFloatChannel
        default=1
        keys [[5767.4375,0],[5768.7525,0.4],[5770.9425,1],[5773.01,0.3],[5775.08,1],[5789.2925,0],[5790.9625,0.4],[5793.555,0],[5826,1],[5830,0],[5845,1],[5848,0],[5861,1],[5863,0],[5880,1],[5881,0],[5897,1],[5898,0],[5905,0.5],[5916,0],[6038.4175,0],[6042.4175,0.4],[6049.4175,0],[6054.4175,0],[6060.4175,0.3],[6073.4175,0]]
       }]
     }]
   }
   {
    type=MovieSceneNiagaraFloatParameterTrack
    display_name=IntensityLerp
    sections [{
      type=MovieSceneFloatSection
      channels [None {
        type=MovieSceneScriptingFloatChannel
        default=0
        keys [[6638.925,0],[6693.925,1],[6694.925,3],[6707.925,1.5258],[6776.925,1],[7069.6375,1],[7086.5575,0]]
       }]
     }]
   }
  ]
 }
]
display_rate {denominator=1, numerator=60}
tick_resolution {denominator=1, numerator=24000}
tracks [{
  type=MovieSceneMaterialParameterCollectionTrack
  display_name=MPC_LimeRock
  sections [{
    type=MovieSceneParameterSection
    channels [
     PortalRadius {
      type=MovieSceneScriptingFloatChannel
      default=5
      keys [[4993.375,0],[5024.3025,240.1363],[5076.2425,380],[5161.4675,350],[5605,350],[5879,600],[5905,931.9117],[5916,1350],[5995,1000],[6712,984.1491],[6759,818.1586],[6813,891.2535],[6868,701.5089],[6919,779.576],[6970,558.8492],[7012.645,625.5732],[7045.645,500],[7114,145.2137],[7144,261.6614],[7164.84,0],[7688.84,0],[7705,2800],[7776,5565.396],[7974,7522.7729],[8216.5725,8000],[8582,5]]
     }
     TentacleEmissive {type=MovieSceneScriptingFloatChannel, default=0, keys [[9124.575,1],[9178.6825,0]]}
     PortalPositionRadiusR {type=MovieSceneScriptingFloatChannel, default=-122218, keys [[7164.84,-101065],[7688.84,-115767],[8244.9275,-115767],[8388.8375,-122218]]}
     PortalPositionRadiusG {type=MovieSceneScriptingFloatChannel, default=-105401, keys [[7164.84,-99603],[7688.84,-103432],[8244.9275,-103432],[8388.8375,-105401]]}
     PortalPositionRadiusB {type=MovieSceneScriptingFloatChannel, default=14443, keys [[7164.84,6710],[7688.84,12342],[8244.9275,12342],[8388.8375,14443]]}
     PortalPositionRadiusA {type=MovieSceneScriptingFloatChannel, default=0}
    ]
   }]
 }]
