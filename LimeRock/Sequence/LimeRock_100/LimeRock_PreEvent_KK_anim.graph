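# Level Sequence that animates a LimeRock pre-event scene with a Fishstix mannequin, UFO-tagged crates, a GoggleBud glider and a fake countdown timer.
# Keyed transforms and visibility lift and tumble crates, drop the big ones, flicker the timer, and move Fishstix down, hiding and revealing him with Z-scale squashes.
# The glider layers a flight path with looping sway and bob transform sections.
# Niagara triggers add impact and hiding effects, alongside three subsequences.
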LevelSequence /LimeRock/Sequence/LimeRock_100/LimeRock_PreEvent_KK_anim
bindings [
 FortPlayerMannequin - Fishstix {
  id=#1
  possessed_class=FortPlayerMannequin
  tracks [
   {
    type=MovieScene3DTransformTrack
    sections [
     {
      type=MovieScene3DTransformSection
      channels [
       Location.X {type=MovieSceneScriptingDoubleChannel, default=-90836, thinned_from=3, keys [[7116,-90896],[8095,-90844]]}
       Location.Y {type=MovieSceneScriptingDoubleChannel, default=-94432, keys [[7116,-94728],[7689,-94444],[8095,-94444]]}
       Location.Z {type=MovieSceneScriptingDoubleChannel, default=1104, keys [[7116,2140],[7689,1108],[8095,1108]]}
       Rotation.X {type=MovieSceneScriptingDoubleChannel, default=0, keys [[7116,0],[7689,-7.7385],[8095,-7.7385]]}
       Rotation.Y {type=MovieSceneScriptingDoubleChannel, default=0, keys [[7116,0],[7689,2.94],[8095,2.94]]}
       Rotation.Z {type=MovieSceneScriptingDoubleChannel, default=-179.8238, keys [[7116,9.4017],[7689,-173.0976],[8095,-173.0976]]}
      ]
     }
     {
      type=MovieScene3DTransformSection
      channels [
       Scale.X {type=MovieSceneScriptingDoubleChannel, default=1, const [1,5]}
       Scale.Y {type=MovieSceneScriptingDoubleChannel, default=1, const [1,5]}
       Scale.Z {type=MovieSceneScriptingDoubleChannel, default=1, keys [[7116,1],[8097,0.6627],[8122,1],[8634,1],[8659,0.6627]]}
      ]
     }
    ]
   }
   {
    type=MovieSceneSkeletalAnimationTrack
    sections [
     {type=MovieSceneSkeletalAnimationSection, end_frame=8082, start_frame=7114}
     {type=MovieSceneSkeletalAnimationSection, end_frame=8266, start_frame=8095}
     {type=MovieSceneSkeletalAnimationSection, end_frame=8682, start_frame=8266}
    ]
   }
   {
    type=MovieSceneVisibilityTrack
    sections [{
      type=MovieSceneVisibilitySection
      channels [None {type=MovieSceneScriptingBoolChannel, default=false, keys [[7080,false],[7116,true],[7686,false],[8098,true],[8660,false]]}]
     }]
   }
  ]
 }
 Ammo Crate 2 - UFO {
  id=#2
  possessed_class=FortStaticMeshActor
  tracks [
   {
    type=MovieScene3DTransformTrack
    sections [{
      type=MovieScene3DTransformSection
      channels [
       Location.X {type=MovieSceneScriptingDoubleChannel, default=-91037.9788, keys [[7161,-91037.9766],[7345,-90701.1562]]}
       Location.Y {type=MovieSceneScriptingDoubleChannel, default=-94850.9827, keys [[7161,-94850.9844],[7345,-94718.9453]]}
       Location.Z {type=MovieSceneScriptingDoubleChannel, default=1102.0885, keys [[7161,1102.0885],[7345,2178.0884]]}
       Rotation.X {type=MovieSceneScriptingDoubleChannel, default=0, keys [[7161,0],[7345,76.3005]]}
       Rotation.Y {type=MovieSceneScriptingDoubleChannel, default=0, keys [[7161,0],[7345,-26.6343]]}
       Rotation.Z {type=MovieSceneScriptingDoubleChannel, default=-33.4985, keys [[7161,-33.4985],[7345,-207.2621]]}
       Scale.X {type=MovieSceneScriptingDoubleChannel, default=1, const [1,2]}
       Scale.Y {type=MovieSceneScriptingDoubleChannel, default=1, const [1,2]}
       Scale.Z {type=MovieSceneScriptingDoubleChannel, default=1, const [1,2]}
      ]
     }]
   }
   {
    type=MovieSceneVisibilityTrack
    sections [{type=MovieSceneVisibilitySection, channels [None {type=MovieSceneScriptingBoolChannel, default=false, keys [[7075,true],[7329,false]]}]}]
   }
  ]
 }
 Big Crate 1 - UFO {
  id=#3
  possessed_class=FortStaticMeshActor
  tracks [
   {
    type=MovieScene3DTransformTrack
    sections [{
      type=MovieScene3DTransformSection
      channels [
       Location.X {type=MovieSceneScriptingDoubleChannel, default=3939.2887, thinned_from=4, keys [[7190,3939.2888],[7569,4187.2891],[7606,4187.2891]]}
       Location.Y {type=MovieSceneScriptingDoubleChannel, default=-192.3957, const [-192.3957,2]}
       Location.Z {type=MovieSceneScriptingDoubleChannel, default=-353.2664, keys [[7190,-353.2664],[7566,534.7336],[7569,534.7336],[7606,-349.2664]]}
       Rotation.X {type=MovieSceneScriptingDoubleChannel, default=0, thinned_from=4, keys [[7190,0],[7566,353.5704],[7606,353.5704]]}
       Rotation.Y {type=MovieSceneScriptingDoubleChannel, default=0, thinned_from=4, keys [[7190,0],[7566,22.4098],[7606,22.4098]]}
       Rotation.Z {type=MovieSceneScriptingDoubleChannel, default=35.9927, thinned_from=4, keys [[7190,35.9927],[7566,271.2524],[7606,271.2524]]}
       Scale.X {type=MovieSceneScriptingDoubleChannel, default=0.5144, const [0.5144,2]}
       Scale.Y {type=MovieSceneScriptingDoubleChannel, default=0.5144, const [0.5144,2]}
       Scale.Z {type=MovieSceneScriptingDoubleChannel, default=0.5144, const [0.5144,2]}
      ]
     }]
   }
   {
    type=MovieSceneVisibilityTrack
    sections [{type=MovieSceneVisibilitySection, channels [None {type=MovieSceneScriptingBoolChannel, default=false, keys [[7074,true],[7611,false]]}]}]
   }
  ]
 }
 NiagaraComponent0 {
  id=#4
  parent_id=#5
  possessed_class=NiagaraComponent
  tracks [{
    type=MovieSceneParticleTrack
    display_name=FX System
    sections [{type=MovieSceneParticleSection, channels [None {type=MovieSceneScriptingParticleChannel, const ["<ParticleKey.TRIGGER: 2>", 1]}]}]
   }]
 }
 Ammo Crate 1 - UFO {
  id=#6
  possessed_class=FortStaticMeshActor
  tracks [
   {
    type=MovieScene3DTransformTrack
    sections [{
      type=MovieScene3DTransformSection
      channels [
       Location.X {type=MovieSceneScriptingDoubleChannel, default=-91003.2317, keys [[7116,-91003.2344],[7288,-90757.3047]]}
       Location.Y {type=MovieSceneScriptingDoubleChannel, default=-94797.1039, keys [[7116,-94797.1016],[7288,-94690.6016]]}
       Location.Z {type=MovieSceneScriptingDoubleChannel, default=1152.0444, keys [[7116,1152.0444],[7288,2288.0444]]}
       Rotation.X {type=MovieSceneScriptingDoubleChannel, default=0, keys [[7116,0],[7288,-180]]}
       Rotation.Y {type=MovieSceneScriptingDoubleChannel, default=0, keys [[7116,0],[7288,-82.0774]]}
       Rotation.Z {type=MovieSceneScriptingDoubleChannel, default=23.416, keys [[7116,23.416],[7288,105.6695]]}
       Scale.X {type=MovieSceneScriptingDoubleChannel, default=1, const [1,1]}
       Scale.Y {type=MovieSceneScriptingDoubleChannel, default=1, const [1,1]}
       Scale.Z {type=MovieSceneScriptingDoubleChannel, default=1, const [1,1]}
      ]
     }]
   }
   {
    type=MovieSceneVisibilityTrack
    sections [{type=MovieSceneVisibilitySection, channels [None {type=MovieSceneScriptingBoolChannel, default=false, keys [[7075,true],[7290,false]]}]}]
   }
  ]
 }
 GoggleBud_Glider {
  id=#7
  possessed_class=FortSkeletalMeshActor
  tracks [
   {
    type=MovieScene3DTransformTrack
    sections [
     {
      type=MovieScene3DTransformSection
      end_frame=10800
      start_frame=-6
      channels [
       Weight {type=MovieSceneScriptingFloatChannel, default=1, const [1,24]}
       Location.X {
        type=MovieSceneScriptingDoubleChannel
        default=-5645
        thinned_from=42
        keys [[16,104.8001],[4161,104.8001],[4181,360.0781],[4248,7782.0395],[4284,10327.4462],[4328,9014.0728],[4375,9049.3056],[4580,8985.4918],[5078,9066.9465],[5269,10663.1396],[5415,9802.9777],[5523,8326.9142],[5671,9314.2284],[5767,9858.0166],[5918,8593.7002],[6820,8593.7002],[6849,8809.4665],[6885,7918.9658],[6920,7192.482],[6957,6977.4867],[7580,7237.4213],[7591,7099.2514],[7629,5198.0862],[7640,5128.8691],[7669,7797.8335],[7689,10409.2191],[7707,11423.6467],[7781,10759.1523],[7846,9397.524],[7962,5674.3577],[7995,4785.1048],[8040,5685.3959],[8078,9234.6248],[8097,9671.111],[8128,8631.7205],[8178,7037.9169],[8243,6696.8905],[8304,8278.1113],[8328,7538.1416],[8450,-3191.6943],[8551,104.8001]]
       }
       Location.Y {
        type=MovieSceneScriptingDoubleChannel
        default=5883.0769
        thinned_from=43
        keys [[16,5883.0767],[4161,5883.0767],[4176,5676.5026],[4233,4331.4455],[4261,4837.9796],[4311,7337.6982],[4331,7448.2517],[5080,7366.2516],[5248,8092.3186],[5335,7570.9043],[5409,6855.0058],[5543,5152.8596],[5717,4811.4817],[5846,4006.7414],[5918,3763.2238],[6824,3763.2238],[6851,3575.4179],[6872,3831.075],[6920,5697.5482],[6949,6059.5999],[6984,5947.0865],[7608,5974.1364],[7664,5576.2251],[7707,6463.8673],[7756,8244.3886],[7780,8667.6351],[7815,8619.4655],[7840,7849.428],[7879,4381.3116],[7915,2518.2251],[8007,2908.3361],[8062,5594.6424],[8107,8977.9639],[8205,10320.286],[8264,8752.3568],[8312,3615.5547],[8349,-80.1777],[8393,-200.1166],[8501,6270.1131],[8551,5818.7583],[8907,5818.7583],[9000,5883.0769]]
       }
       Location.Z {
        type=MovieSceneScriptingDoubleChannel
        default=100
        keys [[16,-700.4705],[4161,-700.4705],[4180,-780.5598],[4200,-1316.1339],[4252,-4378.4829],[4275,-5027.1134],[4331,-4845.6903],[4364,-4868.8338],[5080,-4868.8338],[5221,-5069.0997],[5270,-5048.9971],[5468,-4569.7675],[5918,-4484.1592],[6786,-4484.1592],[6824,-4477.2029],[6850,-4492.4174],[6908,-4893.2748],[6957,-4598.7327],[7516,-4608.7129],[7577,-4648.6983],[7602,-4548.8461],[7632,-4076.2071],[7639,-4096.5154],[7690,-5235.819],[7694,-5273.2928],[7763,-4775.5101],[7817,-4766.0476],[7870,-5273.0838],[7877,-5241.8926],[7930,-4362.4494],[8000,-4196.4955],[8026,-4385.4838],[8068,-5409.7907],[8075,-5449.4316],[8103,-5294.8908],[8145,-4730.4793],[8204,-4593.2759],[8258,-4692.7142],[8295,-5214.93],[8304,-5235.5209],[8339,-4507.6633],[8424,-2466.5737],[8529,-687.9226],[8551,-700.4705]]
       }
       Rotation.X {
        type=MovieSceneScriptingDoubleChannel
        default=2
        thinned_from=34
        keys [[16,4.2525],[4209,4.2525],[4261,4.023],[4332,24.8228],[4356,19.5305],[4381,20.2201],[5080,20.2665],[5278,22.6537],[5859,22.0004],[6813,19.1374],[6881,6.4089],[6933,6.9137],[6969,19.5872],[7574,19.578],[7649,-3.5702],[7661,-0.5288],[7690,25.398],[7697,26.3776],[7708,26.8236],[7727,25.726],[7827,16.0154],[7855,13.0025],[7893,26.9042],[7900,29.1303],[7979,11.83],[8055,8.526],[8165,17.7895],[8181,18.6214],[8269,9.1218],[8343,21.113],[8387,21.8134],[8517,4.2525]]
       }
       Rotation.Y {
        type=MovieSceneScriptingDoubleChannel
        default=-1
        thinned_from=39
        keys [[16,1.3952],[4161,1.3952],[4241,-11.4288],[4257,-13.0032],[4335,-3.3842],[4359,1.4575],[5080,1.3637],[5245,1.3637],[5902,1.3161],[6806,3.3581],[6835,6.5749],[6881,-1.9515],[6905,-5.5031],[6933,2.848],[6974,-3.8307],[7010,-0.2768],[7607,-0.2766],[7648,-7.5503],[7681,-14.495],[7715,-17.8758],[7748,-16.8622],[7817,-4.545],[7854,-10.8159],[7893,8.7619],[7902,12.7287],[7922,14.6196],[7960,21.522],[7999,15.0607],[8084,-17.3072],[8121,-20.8358],[8178,-15.3854],[8217,-8.1665],[8293,15.5087],[8328,18.9899],[8403,21.0176],[8462,20.697],[8486,20.0347],[8557,1.3952]]
       }
       Rotation.Z {
        type=MovieSceneScriptingDoubleChannel
        default=-63
        thinned_from=32
        keys [[16,4.8488],[4189,4.8488],[4239,16.3862],[4330,244.919],[4352,253.7471],[5080,253.4431],[5179,352.9697],[5305,252.4317],[5412,195.0052],[5664,360.3288],[5975,68.4388],[6122,67.3525],[6885,67.3525],[6936,-9.4671],[6969,-14.1795],[7242,-9.9206],[7607,-10.5292],[7676,-10.5292],[7786,179.6669],[7818,218.9574],[7848,235.0216],[7874,235.8229],[7994,40.6872],[8038,3.0595],[8092,77.0346],[8180,241.0773],[8254,287.2007],[8304,276.1193],[8367,195.0227],[8515,-15.0422],[8550,4.8488]]
       }
       Scale.X {type=MovieSceneScriptingDoubleChannel, default=0, const [0,24]}
       Scale.Y {type=MovieSceneScriptingDoubleChannel, default=0, const [0,24]}
       Scale.Z {type=MovieSceneScriptingDoubleChannel, default=0, const [0,24]}
      ]
     }
     {
      type=MovieScene3DTransformSection
      channels [
       Location.X {
        type=MovieSceneScriptingDoubleChannel
        default=1773
        keys [[-6,180.6667],[44,259.777],[147,1625.8337],[203,1967.115],[254,2104.2413],[394,840.902],[505,0],[620,1625.8337],[726,2167.2803],[830,870.384],[888,325.6882],[917,180.6667],[967,259.777],[1070,1625.8337],[1126,1967.115],[1177,2104.2413],[1317,840.902],[1428,0],[1543,1625.8337],[1649,2167.2803],[1753,870.384],[1811,325.6882],[1840,180.6667],[1890,259.777],[1993,1625.8337],[2049,1967.115],[2100,2104.2413],[2240,840.902],[2351,0],[2466,1625.8337],[2572,2167.2803],[2676,870.384],[2734,325.6882],[2769,259.777],[2872,1625.8337],[2928,1967.115],[2979,2104.2413],[3119,840.902],[3230,0],[3345,1625.8337],[3451,2167.2803],[3555,870.384],[3613,325.6882],[3642,180.6667],[3692,259.777],[3795,1625.8337],[3851,1967.115],[3902,2104.2413],[4042,840.902],[4161,180.6667],[8551,180.6667],[8587,0],[8702,1625.8337],[8808,2167.2803],[8912,870.384],[8970,325.6882],[9000,180.6667]]
       }
       Location.Y {
        type=MovieSceneScriptingDoubleChannel
        default=0
        keys [[-6,475.169],[57,-512.0766],[134,-1115.145],[203,-586.2088],[245,35.9207],[337,1130.501],[398,959.0658],[509,-512.0766],[607,-1115.145],[726,35.9207],[822,1130.501],[870,929.1743],[917,475.169],[980,-512.0766],[1057,-1115.145],[1126,-586.2088],[1168,35.9207],[1260,1130.501],[1321,959.0658],[1432,-512.0766],[1530,-1115.145],[1649,35.9207],[1745,1130.501],[1793,929.1743],[1840,475.169],[1903,-512.0766],[1980,-1115.145],[2049,-586.2088],[2091,35.9207],[2183,1130.501],[2244,959.0658],[2355,-512.0766],[2453,-1115.145],[2572,35.9207],[2668,1130.501],[2716,929.1743],[2782,-512.0766],[2859,-1115.145],[2928,-586.2088],[2970,35.9207],[3062,1130.501],[3123,959.0658],[3234,-512.0766],[3332,-1115.145],[3451,35.9207],[3547,1130.501],[3595,929.1743],[3642,475.169],[3705,-512.0766],[3782,-1115.145],[3851,-586.2088],[3893,35.9207],[3984,1130.501],[4045,1231.9116],[4161,475.169],[8551,475.169],[8591,-349.847],[8689,-1115.145],[8808,35.9207],[8904,1130.501],[8952,929.1743],[9000,475.169]]
       }
       Location.Z {
        type=MovieSceneScriptingDoubleChannel
        default=0
        keys [[-6,91.6176],[58,4.9135],[121,-16.7535],[193,-33.3555],[245,-43.6846],[356,136.8689],[505,48.5817],[622,-18.3455],[726,-43.6846],[843,136.8689],[917,91.6176],[981,4.9135],[1044,-16.7535],[1116,-33.3555],[1168,-43.6846],[1279,136.8689],[1428,48.5817],[1545,-18.3455],[1649,-43.6846],[1766,136.8689],[1840,91.6176],[1904,4.9135],[1967,-16.7535],[2039,-33.3555],[2091,-43.6846],[2202,136.8689],[2351,48.5817],[2468,-18.3455],[2572,-43.6846],[2689,136.8689],[2783,4.9135],[2846,-16.7535],[2918,-33.3555],[2970,-43.6846],[3081,136.8689],[3230,48.5817],[3347,-18.3455],[3451,-43.6846],[3568,136.8689],[3642,91.6176],[3706,4.9135],[3769,-16.7535],[3841,-33.3555],[3893,-43.6846],[4004,136.8689],[4161,91.6176],[8551,91.6176],[8587,77.7578],[8704,-18.3455],[8808,-43.6846],[8925,136.8689],[9000,91.6176]]
       }
       Rotation.X {
        type=MovieSceneScriptingDoubleChannel
        default=0
        thinned_from=62
        keys [[-6,-8.8802],[89,-3.8693],[149,0],[222,0.9501],[288,15.7412],[396,-2.2254],[496,-5.6962],[660,6.2425],[705,5.5589],[818,-3.0018],[869,-7.4621],[917,-8.8802],[1012,-3.8693],[1072,0],[1145,0.9501],[1211,15.7412],[1319,-2.2254],[1419,-5.6962],[1583,6.2425],[1628,5.5589],[1741,-3.0018],[1792,-7.4621],[1840,-8.8802],[1935,-3.8693],[1995,0],[2068,0.9501],[2134,15.7412],[2242,-2.2254],[2342,-5.6962],[2506,6.2425],[2551,5.5589],[2664,-3.0018],[2715,-7.4621],[2814,-3.8693],[2874,0],[2947,0.9501],[3013,15.7412],[3121,-2.2254],[3221,-5.6962],[3385,6.2425],[3430,5.5589],[3543,-3.0018],[3594,-7.4621],[3642,-8.8802],[3737,-3.8693],[3797,0],[3870,0.9501],[3936,15.7412],[4044,-2.2254],[4161,-8.8802],[8551,-8.8802],[8578,-5.6962],[8742,6.2425],[8787,5.5589],[8900,-3.0018],[8951,-7.4621],[9000,-8.8802]]
       }
       Rotation.Y {
        type=MovieSceneScriptingDoubleChannel
        default=0
        keys [[-6,0.8033],[40,0],[137,-4.721],[202,-6.2305],[203,-6.2283],[300,6.2096],[352,11.9725],[528,-10.6332],[622,-8.6424],[705,-2.8426],[791,4.5107],[917,0.8033],[963,0],[1060,-4.721],[1125,-6.2305],[1126,-6.2283],[1223,6.2096],[1275,11.9725],[1451,-10.6332],[1545,-8.6424],[1628,-2.8426],[1714,4.5107],[1840,0.8033],[1886,0],[1983,-4.721],[2048,-6.2305],[2049,-6.2283],[2146,6.2096],[2198,11.9725],[2374,-10.6332],[2468,-8.6424],[2551,-2.8426],[2637,4.5107],[2765,0],[2862,-4.721],[2927,-6.2305],[2928,-6.2283],[3025,6.2096],[3077,11.9725],[3253,-10.6332],[3347,-8.6424],[3430,-2.8426],[3516,4.5107],[3642,0.8033],[3688,0],[3785,-4.721],[3850,-6.2305],[3851,-6.2283],[3948,6.2096],[4000,11.9725],[4161,0.8033],[8551,0.8033],[8610,-10.6332],[8704,-8.6424],[8787,-2.8426],[8873,4.5107],[9000,0.8033]]
       }
       Rotation.Z {
        type=MovieSceneScriptingDoubleChannel
        default=0
        keys [[-6,13.5904],[52,21.6394],[143,69.9447],[203,99.1289],[207,99.432],[396,-67.367],[539,-3.0012],[726,-36.5987],[849,-6.674],[917,13.5904],[975,21.6394],[1066,69.9447],[1126,99.1289],[1130,99.432],[1319,-67.367],[1462,-3.0012],[1649,-36.5987],[1772,-6.674],[1840,13.5904],[1898,21.6394],[1989,69.9447],[2049,99.1289],[2053,99.432],[2242,-67.367],[2385,-3.0012],[2572,-36.5987],[2695,-6.674],[2777,21.6394],[2868,69.9447],[2928,99.1289],[2932,99.432],[3121,-67.367],[3264,-3.0012],[3451,-36.5987],[3574,-6.674],[3642,13.5904],[3700,21.6394],[3791,69.9447],[3851,99.1289],[3855,99.432],[4044,-67.367],[4161,13.5904],[8551,13.5904],[8621,-3.0012],[8808,-36.5987],[8931,-6.674],[9000,13.5904]]
       }
       Scale.X {type=MovieSceneScriptingDoubleChannel, default=0, const [0,52]}
       Scale.Y {type=MovieSceneScriptingDoubleChannel, default=0, const [0,52]}
       Scale.Z {type=MovieSceneScriptingDoubleChannel, default=0, const [0,52]}
      ]
     }
     {
      type=MovieScene3DTransformSection
      channels [
       Weight {type=MovieSceneScriptingFloatChannel, default=1, const [1,61]}
       Location.X {type=MovieSceneScriptingDoubleChannel, default=-2955, const [-2955,62]}
       Location.Y {type=MovieSceneScriptingDoubleChannel, default=-6560, const [-6560,62]}
       Location.Z {
        type=MovieSceneScriptingDoubleChannel
        default=-698.6482
        thinned_from=207
        keys [[-6,4892.2427],[45,4758.5867],[102,4924.83],[141,4773.2557],[199,4967.0625],[273,4773.2557],[333,4967.1649],[374,4906.1128],[408,4773.2557],[457,4888.8716],[492,4773.2557],[552,4967.1649],[604,4773.2557],[662,4967.0625],[736,4773.2557],[796,4967.1649],[816,4909.0852],[866,4730.1159],[913,4888.1347],[917,4892.2427],[968,4758.5867],[1025,4924.83],[1064,4773.2557],[1122,4967.0625],[1196,4773.2557],[1256,4967.1649],[1297,4906.1128],[1331,4773.2557],[1380,4888.8716],[1415,4773.2557],[1475,4967.1649],[1527,4773.2557],[1585,4967.0625],[1659,4773.2557],[1719,4967.1649],[1739,4909.0852],[1789,4730.1159],[1836,4888.1347],[1840,4892.2427],[1891,4758.5867],[1948,4924.83],[1987,4773.2557],[2045,4967.0625],[2119,4773.2557],[2179,4967.1649],[2220,4906.1128],[2254,4773.2557],[2303,4888.8716],[2338,4773.2557],[2398,4967.1649],[2450,4773.2557],[2508,4967.0625],[2582,4773.2557],[2642,4967.1649],[2662,4909.0852],[2712,4730.1159],[2759,4888.1347],[2763,4892.2427],[2814,4758.5867],[2871,4924.83],[2910,4773.2557],[2968,4967.0625],[3042,4773.2557],[3102,4967.1649],[3143,4906.1128],[3177,4773.2557],[3226,4888.8716],[3261,4773.2557],[3321,4967.1649],[3373,4773.2557],[3431,4967.0625],[3505,4773.2557],[3565,4967.1649],[3585,4909.0852],[3635,4730.1159],[3682,4888.1347],[3686,4892.2427],[3737,4758.5867],[3794,4924.83],[3833,4773.2557],[3891,4967.0625],[3965,4773.2557],[4025,4967.1649],[4066,4906.1128],[4100,4773.2557],[4149,4888.8716],[4184,4773.2557],[4244,4967.1649],[4296,4773.2557],[4354,4967.0625],[4428,4773.2557],[4488,4967.1649],[4508,4909.0852],[4558,4730.1159],[4605,4888.1347],[4609,4892.2427],[4660,4758.5867],[4717,4924.83],[4756,4773.2557],[4814,4967.0625],[4888,4773.2557],[4948,4967.1649],[4989,4906.1128],[5023,4773.2557],[5072,4888.8716],[5107,4773.2557],[5167,4967.1649],[5219,4773.2557],[5277,4967.0625],[5351,4773.2557],[5411,4967.1649],[5431,4909.0852],[5481,4730.1159],[5528,4888.1347],[5532,4892.2427],[5583,4758.5867],[5640,4924.83],[5679,4773.2557],[5737,4967.0625],[5811,4773.2557],[5871,4967.1649],[5912,4906.1128],[5946,4773.2557],[5995,4888.8716],[6030,4773.2557],[6090,4967.1649],[6142,4773.2557],[6200,4967.0625],[6274,4773.2557],[6334,4967.1649],[6354,4909.0852],[6404,4730.1159],[6451,4888.1347],[6455,4892.2427],[6506,4758.5867],[6563,4924.83],[6602,4773.2557],[6660,4967.0625],[6734,4773.2557],[6794,4967.1649],[6835,4906.1128],[6869,4773.2557],[6918,4888.8716],[6953,4773.2557],[7013,4967.1649],[7065,4773.2557],[7123,4967.0625],[7197,4773.2557],[7257,4967.1649],[7277,4909.0852],[7327,4730.1159],[7374,4888.1347],[7378,4892.2427],[7429,4758.5867],[7486,4924.83],[7525,4773.2557],[7583,4967.0625],[7657,4773.2557],[7717,4967.1649],[7758,4906.1128],[7792,4773.2557],[7841,4888.8716],[7876,4773.2557],[7936,4967.1649],[7988,4773.2557],[8046,4967.0625],[8120,4773.2557],[8180,4967.1649],[8200,4909.0852],[8250,4730.1159],[8297,4888.1347],[8301,4892.2427],[8352,4758.5867],[8409,4924.83],[8448,4773.2557],[8506,4967.0625],[8580,4773.2557],[8640,4967.1649],[8681,4906.1128],[8715,4773.2557],[8764,4888.8716],[8799,4773.2557],[8866,4917.168],[8927,4707.6961],[8973,4967.1649],[9000,4892.2427]]
       }
       Rotation.X {
        type=MovieSceneScriptingDoubleChannel
        default=0
        keys [[-6,-11.0829],[57,-15.7424],[117,-10.577],[178,-15.2647],[199,-13.8933],[239,-10.2586],[308,-15.1055],[374,-10.5817],[425,-15.9016],[471,-11.0829],[520,-15.7424],[580,-10.577],[641,-15.2647],[662,-13.8933],[702,-10.2586],[771,-15.1055],[817,-10.7564],[887,-14.8435],[917,-11.0829],[980,-15.7424],[1040,-10.577],[1101,-15.2647],[1122,-13.8933],[1162,-10.2586],[1231,-15.1055],[1297,-10.5817],[1348,-15.9016],[1394,-11.0829],[1443,-15.7424],[1503,-10.577],[1564,-15.2647],[1585,-13.8933],[1625,-10.2586],[1694,-15.1055],[1740,-10.7564],[1810,-14.8435],[1840,-11.0829],[1903,-15.7424],[1963,-10.577],[2024,-15.2647],[2045,-13.8933],[2085,-10.2586],[2154,-15.1055],[2220,-10.5817],[2271,-15.9016],[2317,-11.0829],[2366,-15.7424],[2426,-10.577],[2487,-15.2647],[2508,-13.8933],[2548,-10.2586],[2617,-15.1055],[2663,-10.7564],[2710.645,-13.8593],[2733,-14.8435],[2763,-11.0829],[2826,-15.7424],[2886,-10.577],[2947,-15.2647],[2968,-13.8933],[3008,-10.2586],[3077,-15.1055],[3143,-10.5817],[3194,-15.9016],[3240,-11.0829],[3289,-15.7424],[3349,-10.577],[3410,-15.2647],[3431,-13.8933],[3471,-10.2586],[3540,-15.1055],[3586,-10.7564],[3656,-14.8435],[3686,-11.0829],[3749,-15.7424],[3809,-10.577],[3870,-15.2647],[3891,-13.8933],[3931,-10.2586],[4000,-15.1055],[4066,-10.5817],[4117,-15.9016],[4163,-11.0829],[4212,-15.7424],[4272,-10.577],[4333,-15.2647],[4354,-13.8933],[4394,-10.2586],[4463,-15.1055],[4509,-10.7564],[4579,-14.8435],[4609,-11.0829],[4672,-15.7424],[4732,-10.577],[4793,-15.2647],[4814,-13.8933],[4854,-10.2586],[4923,-15.1055],[4989,-10.5817],[5040,-15.9016],[5086,-11.0829],[5135,-15.7424],[5195,-10.577],[5256,-15.2647],[5277,-13.8933],[5317,-10.2586],[5386,-15.1055],[5432,-10.7564],[5502,-14.8435],[5532,-11.0829],[5595,-15.7424],[5655,-10.577],[5716,-15.2647],[5737,-13.8933],[5777,-10.2586],[5846,-15.1055],[5912,-10.5817],[5963,-15.9016],[6009,-11.0829],[6058,-15.7424],[6118,-10.577],[6179,-15.2647],[6200,-13.8933],[6240,-10.2586],[6309,-15.1055],[6355,-10.7564],[6425,-14.8435],[6455,-11.0829],[6518,-15.7424],[6578,-10.577],[6639,-15.2647],[6660,-13.8933],[6700,-10.2586],[6769,-15.1055],[6835,-10.5817],[6886,-15.9016],[6932,-11.0829],[6981,-15.7424],[7041,-10.577],[7102,-15.2647],[7123,-13.8933],[7163,-10.2586],[7232,-15.1055],[7278,-10.7564],[7348,-14.8435],[7378,-11.0829],[7441,-15.7424],[7501,-10.577],[7562,-15.2647],[7583,-13.8933],[7623,-10.2586],[7692,-15.1055],[7758,-10.5817],[7809,-15.9016],[7855,-11.0829],[7904,-15.7424],[7964,-10.577],[8025,-15.2647],[8046,-13.8933],[8086,-10.2586],[8155,-15.1055],[8201,-10.7564],[8271,-14.8435],[8301,-11.0829],[8364,-15.7424],[8424,-10.577],[8485,-15.2647],[8506,-13.8933],[8546,-10.2586],[8615,-15.1055],[8681,-10.5817],[8732,-15.9016],[8778,-11.0829],[8827,-15.7424],[8973,-11.0829],[8997,-11.0547],[9000,-11.0829]]
       }
       Rotation.Y {
        type=MovieSceneScriptingDoubleChannel
        default=0
        keys [[-6,-1.8042],[7,-1.6543],[111,-3.3404],[178,-1.6543],[199,-1.783],[302,-3.3404],[374,-1.6548],[457,-1.7957],[470,-1.6543],[574,-3.3404],[641,-1.6543],[662,-1.783],[765,-3.3404],[837,-1.6548],[917,-1.8042],[930,-1.6543],[1034,-3.3404],[1101,-1.6543],[1122,-1.783],[1225,-3.3404],[1297,-1.6548],[1380,-1.7957],[1393,-1.6543],[1497,-3.3404],[1564,-1.6543],[1585,-1.783],[1688,-3.3404],[1760,-1.6548],[1840,-1.8042],[1853,-1.6543],[1957,-3.3404],[2024,-1.6543],[2045,-1.783],[2148,-3.3404],[2220,-1.6548],[2303,-1.7957],[2316,-1.6543],[2420,-3.3404],[2487,-1.6543],[2508,-1.783],[2611,-3.3404],[2683,-1.6548],[2710.645,-1.7181],[2763,-1.8042],[2776,-1.6543],[2880,-3.3404],[2947,-1.6543],[2968,-1.783],[3071,-3.3404],[3143,-1.6548],[3226,-1.7957],[3239,-1.6543],[3343,-3.3404],[3410,-1.6543],[3431,-1.783],[3534,-3.3404],[3606,-1.6548],[3686,-1.8042],[3699,-1.6543],[3803,-3.3404],[3870,-1.6543],[3891,-1.783],[3994,-3.3404],[4066,-1.6548],[4149,-1.7957],[4162,-1.6543],[4266,-3.3404],[4333,-1.6543],[4354,-1.783],[4457,-3.3404],[4529,-1.6548],[4609,-1.8042],[4622,-1.6543],[4726,-3.3404],[4793,-1.6543],[4814,-1.783],[4917,-3.3404],[4989,-1.6548],[5072,-1.7957],[5085,-1.6543],[5189,-3.3404],[5256,-1.6543],[5277,-1.783],[5380,-3.3404],[5452,-1.6548],[5532,-1.8042],[5545,-1.6543],[5649,-3.3404],[5716,-1.6543],[5737,-1.783],[5840,-3.3404],[5912,-1.6548],[5995,-1.7957],[6008,-1.6543],[6112,-3.3404],[6179,-1.6543],[6200,-1.783],[6303,-3.3404],[6375,-1.6548],[6455,-1.8042],[6468,-1.6543],[6572,-3.3404],[6639,-1.6543],[6660,-1.783],[6763,-3.3404],[6835,-1.6548],[6918,-1.7957],[6931,-1.6543],[7035,-3.3404],[7102,-1.6543],[7123,-1.783],[7226,-3.3404],[7298,-1.6548],[7378,-1.8042],[7391,-1.6543],[7495,-3.3404],[7562,-1.6543],[7583,-1.783],[7686,-3.3404],[7758,-1.6548],[7841,-1.7957],[7854,-1.6543],[7958,-3.3404],[8025,-1.6543],[8046,-1.783],[8149,-3.3404],[8221,-1.6548],[8301,-1.8042],[8314,-1.6543],[8418,-3.3404],[8485,-1.6543],[8506,-1.783],[8609,-3.3404],[8681,-1.6548],[8764,-1.7957],[8777,-1.6543],[8973,-1.8042],[8997,-1.8618],[9000,-1.8042]]
       }
       Rotation.Z {type=MovieSceneScriptingDoubleChannel, default=0, thinned_from=81, keys [[-6,-989.5686],[9000,-989.5686]]}
      ]
     }
    ]
   }
   {
    type=MovieSceneSkeletalAnimationTrack
    sections [
     {type=MovieSceneSkeletalAnimationSection, end_frame=9292, start_frame=-6}
     {type=MovieSceneSkeletalAnimationSection, end_frame=7591, start_frame=4414}
    ]
   }
   {
    type=MovieSceneVisibilityTrack
    sections [{type=MovieSceneVisibilitySection, channels [None {type=MovieSceneScriptingBoolChannel, default=true}]}]
   }
  ]
 }
 NS_Supply_Drop_Impact_BigCrate2_UFO {id=#8, possessed_class=NiagaraActor, tracks []}
 NS_Supply_Drop_Impact_BigCrate1_UFO {id=#9, possessed_class=NiagaraActor, tracks []}
 NiagaraComponent0 {
  id=#10
  parent_id=#8
  possessed_class=NiagaraComponent
  tracks [{
    type=MovieSceneParticleTrack
    display_name=FX System
    sections [{type=MovieSceneParticleSection, channels [None {type=MovieSceneScriptingParticleChannel, const ["<ParticleKey.TRIGGER: 2>", 1]}]}]
   }]
 }
 BP_LimeRock_CountdownTimer_Fake {
  id=#11
  possessed_class=BP_Generic_Game_Countdown_C
  tracks [
   {
    type=MovieScene3DTransformTrack
    sections [{
      type=MovieScene3DTransformSection
      channels [
       Location.X {type=MovieSceneScriptingDoubleChannel, default=-90560, const [-90560,2]}
       Location.Y {type=MovieSceneScriptingDoubleChannel, default=-94824, const [-94824,2]}
       Location.Z {type=MovieSceneScriptingDoubleChannel, default=1208, const [1208,2]}
       Rotation.X {type=MovieSceneScriptingDoubleChannel, default=0, keys [[7087,0],[7557,-15.7271],[7582,0]]}
       Rotation.Y {type=MovieSceneScriptingDoubleChannel, default=0, keys [[7087,0],[7557,-16.4154],[7582,0]]}
       Rotation.Z {type=MovieSceneScriptingDoubleChannel, default=0, keys [[7087,0],[7557,-31.7616],[7582,0]]}
       Scale.X {type=MovieSceneScriptingDoubleChannel, default=0.162, keys [[7087,0.2],[7557,0.162],[7599,0.2]]}
       Scale.Y {type=MovieSceneScriptingDoubleChannel, default=0.2, const [0.2,3]}
       Scale.Z {type=MovieSceneScriptingDoubleChannel, default=0.3827, keys [[7087,0.2],[7557,0.3827],[7566,0.2]]}
      ]
     }]
   }
   {
    type=MovieSceneVisibilityTrack
    sections [{
      type=MovieSceneVisibilitySection
      channels [None {type=MovieSceneScriptingBoolChannel, default=true, keys [[7087,true],[7151,false],[7157,true],[7557,false],[7563,true]]}]
     }]
   }
  ]
 }
 NS_Hiding_Default_Enter_CrateEnter1_UFO {id=#5, possessed_class=NiagaraActor, tracks []}
 NiagaraComponent0 {
  id=#12
  parent_id=#9
  possessed_class=NiagaraComponent
  tracks [{
    type=MovieSceneParticleTrack
    display_name=FX System
    sections [{type=MovieSceneParticleSection, channels [None {type=MovieSceneScriptingParticleChannel, const ["<ParticleKey.TRIGGER: 2>", 1]}]}]
   }]
 }
 Big Crate 2 - UFO {
  id=#13
  possessed_class=FortStaticMeshActor
  tracks [
   {
    type=MovieScene3DTransformTrack
    sections [{
      type=MovieScene3DTransformSection
      channels [
       Location.X {type=MovieSceneScriptingDoubleChannel, default=3873.8247, thinned_from=4, keys [[7237,3873.8247],[7569,4313.8066],[7601,4277.8066]]}
       Location.Y {type=MovieSceneScriptingDoubleChannel, default=-183.1163, thinned_from=4, keys [[7237,-183.1163],[7569,-187.1086],[7601,-139.1086]]}
       Location.Z {type=MovieSceneScriptingDoubleChannel, default=-353.2664, keys [[7237,-353.2664],[7566,490.7336],[7569,490.7336],[7601,-333.2664]]}
       Rotation.X {type=MovieSceneScriptingDoubleChannel, default=0, thinned_from=4, keys [[7237,0],[7566,82.02],[7601,82.02]]}
       Rotation.Y {type=MovieSceneScriptingDoubleChannel, default=0, thinned_from=4, keys [[7237,0],[7566,-38.4497],[7601,-38.4497]]}
       Rotation.Z {type=MovieSceneScriptingDoubleChannel, default=-0.5199, thinned_from=4, keys [[7237,-0.5199],[7566,-147.8702],[7601,-147.8702]]}
       Scale.X {type=MovieSceneScriptingDoubleChannel, default=0.5144, const [0.5144,4]}
       Scale.Y {type=MovieSceneScriptingDoubleChannel, default=0.5144, const [0.5144,4]}
       Scale.Z {type=MovieSceneScriptingDoubleChannel, default=0.5144, const [0.5144,4]}
      ]
     }]
   }
   {
    type=MovieSceneVisibilityTrack
    sections [{type=MovieSceneVisibilitySection, channels [None {type=MovieSceneScriptingBoolChannel, default=true, keys [[7074,true],[7604,false]]}]}]
   }
  ]
 }
]
display_rate {denominator=1, numerator=60}
tick_resolution {denominator=1, numerator=24000}
tracks [{
  type=MovieSceneSubTrack
  display_name=Subsequences
  sections [
   {type=MovieSceneSubSection, end_frame=9000, start_frame=0}
   {type=MovieSceneSubSection, end_frame=9000, start_frame=0}
   {type=MovieSceneSubSection, end_frame=9000, start_frame=0}
  ]
 }]
