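# Level Sequence that handles audio for the LimeRock_100 pre-event, a scene with a UFO, GoggleBud glider, Fishstix mannequin and two crates.
# Audio tracks bound to the crates, mannequin and glider place sounds per actor, with the glider carrying most of the overlapping sections.
# Event triggers fire on the UFO flying-loop AmbientSound and the dust kick-up Niagara actor, whose audio track is empty.
# All tracks are actor-bound, with no sequence-level tracks.
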LevelSequence /LimeRock/Sequence/LimeRock_100/LimeRock_PreEvent_KK_aud
bindings [
 Big Crate 1 - UFO {
  id=#1
  possessed_class=FortStaticMeshActor
  tracks [{
    type=MovieSceneAudioTrack
    display_name=Audio
    sections [{
      type=MovieSceneAudioSection
      end_frame=7707
      start_frame=7612
      channels [
       Volume {type=MovieSceneScriptingFloatChannel, default=1}
       Pitch {type=MovieSceneScriptingFloatChannel, default=1}
      ]
     }]
   }]
 }
 Big Crate 2 - UFO {
  id=#2
  possessed_class=FortStaticMeshActor
  tracks [{
    type=MovieSceneAudioTrack
    display_name=Audio
    sections [{
      type=MovieSceneAudioSection
      end_frame=7713
      start_frame=7593
      channels [
       Volume {type=MovieSceneScriptingFloatChannel, default=1}
       Pitch {type=MovieSceneScriptingFloatChannel, default=1}
      ]
     }]
   }]
 }
 SC_LimeRock_100_UFO_Flying_LP {
  id=#3
  possessed_class=AmbientSound
  tracks [{
    type=MovieSceneEventTrack
    sections [{
      type=MovieSceneEventTriggerSection
      channels [None {
        type=MovieSceneScriptingEventChannel
        keys [[4433, "<Struct 'MovieSceneEvent' (0x851369a8700) {}>"], [7696, "<Struct 'MovieSceneEvent' (0x851369a8660) {}>"]]
       }]
     }]
   }]
 }
 NS_LimeRock_UFODust_Kickup {
  id=#4
  possessed_class=NiagaraActor
  tracks [
   {type=MovieSceneAudioTrack, display_name=Audio, sections []}
   {
    type=MovieSceneEventTrack
    sections [{
      type=MovieSceneEventTriggerSection
      channels [None {
        type=MovieSceneScriptingEventChannel
        keys [
         [7689, "<Struct 'MovieSceneEvent' (0x851369a7580) {}>"]
         [7870, "<Struct 'MovieSceneEvent' (0x851369a7800) {}>"]
         [8071.925, "<Struct 'MovieSceneEvent' (0x851369a50a0) {}>"]
         [8290.3075, "<Struct 'MovieSceneEvent' (0x851369a65e0) {}>"]
        ]
       }]
     }]
   }
  ]
 }
 GoggleBud_Glider {
  id=#5
  possessed_class=FortSkeletalMeshActor
  tracks [{
    type=MovieSceneAudioTrack
    display_name=Audio
    sections [
     {
      type=MovieSceneAudioSection
      end_frame=4542
      start_frame=4375
      channels [
       Volume {type=MovieSceneScriptingFloatChannel, default=1}
       Pitch {type=MovieSceneScriptingFloatChannel, default=1}
      ]
     }
     {
      type=MovieSceneAudioSection
      end_frame=5191
      start_frame=5002
      channels [
       Volume {type=MovieSceneScriptingFloatChannel, default=1}
       Pitch {type=MovieSceneScriptingFloatChannel, default=1}
      ]
     }
     {
      type=MovieSceneAudioSection
      end_frame=5050
      start_frame=4434
      channels [
       Volume {type=MovieSceneScriptingFloatChannel, default=1}
       Pitch {type=MovieSceneScriptingFloatChannel, default=1}
      ]
     }
     {
      type=MovieSceneAudioSection
      end_frame=4561
      start_frame=4195
      channels [
       Volume {type=MovieSceneScriptingFloatChannel, default=1}
       Pitch {type=MovieSceneScriptingFloatChannel, default=1}
      ]
     }
     {
      type=MovieSceneAudioSection
      end_frame=6167
      start_frame=5892
      channels [
       Volume {type=MovieSceneScriptingFloatChannel, default=1}
       Pitch {type=MovieSceneScriptingFloatChannel, default=1}
      ]
     }
     {
      type=MovieSceneAudioSection
      end_frame=6884
      start_frame=6763
      channels [
       Volume {type=MovieSceneScriptingFloatChannel, default=1}
       Pitch {type=MovieSceneScriptingFloatChannel, default=1}
      ]
     }
     {
      type=MovieSceneAudioSection
      end_frame=6808
      start_frame=5944
      channels [
       Volume {type=MovieSceneScriptingFloatChannel, default=1}
       Pitch {type=MovieSceneScriptingFloatChannel, default=1}
      ]
     }
     {
      type=MovieSceneAudioSection
      end_frame=7021
      start_frame=6862
      channels [
       Volume {type=MovieSceneScriptingFloatChannel, default=1}
       Pitch {type=MovieSceneScriptingFloatChannel, default=1}
      ]
     }
     {
      type=MovieSceneAudioSection
      end_frame=7154
      start_frame=6987
      channels [
       Volume {type=MovieSceneScriptingFloatChannel, default=1}
       Pitch {type=MovieSceneScriptingFloatChannel, default=1}
      ]
     }
     {
      type=MovieSceneAudioSection
      end_frame=7640
      start_frame=7546
      channels [
       Volume {type=MovieSceneScriptingFloatChannel, default=1}
       Pitch {type=MovieSceneScriptingFloatChannel, default=1}
      ]
     }
     {
      type=MovieSceneAudioSection
      end_frame=7588
      start_frame=7050
      channels [
       Volume {type=MovieSceneScriptingFloatChannel, default=1}
       Pitch {type=MovieSceneScriptingFloatChannel, default=1}
      ]
     }
     {
      type=MovieSceneAudioSection
      end_frame=7874
      start_frame=7603
      channels [
       Volume {type=MovieSceneScriptingFloatChannel, default=1}
       Pitch {type=MovieSceneScriptingFloatChannel, default=1}
      ]
     }
     {
      type=MovieSceneAudioSection
      end_frame=8097
      start_frame=7842
      channels [
       Volume {type=MovieSceneScriptingFloatChannel, default=1}
       Pitch {type=MovieSceneScriptingFloatChannel, default=1}
      ]
     }
     {
      type=MovieSceneAudioSection
      end_frame=8288
      start_frame=8024
      channels [
       Volume {type=MovieSceneScriptingFloatChannel, default=1}
       Pitch {type=MovieSceneScriptingFloatChannel, default=1}
      ]
     }
     {
      type=MovieSceneAudioSection
      end_frame=8588
      start_frame=8242
      channels [
       Volume {type=MovieSceneScriptingFloatChannel, default=1}
       Pitch {type=MovieSceneScriptingFloatChannel, default=1}
      ]
     }
     {
      type=MovieSceneAudioSection
      end_frame=7760
      start_frame=7082
      channels [
       Volume {type=MovieSceneScriptingFloatChannel, default=1}
       Pitch {type=MovieSceneScriptingFloatChannel, default=1}
      ]
     }
     {
      type=MovieSceneAudioSection
      end_frame=7812
      start_frame=4442
      channels [
       Volume {type=MovieSceneScriptingFloatChannel, default=1}
       Pitch {type=MovieSceneScriptingFloatChannel, default=1}
      ]
     }
     {
      type=MovieSceneAudioSection
      end_frame=8461
      start_frame=8305
      channels [
       Volume {type=MovieSceneScriptingFloatChannel, default=1}
       Pitch {type=MovieSceneScriptingFloatChannel, default=1}
      ]
     }
     {
      type=MovieSceneAudioSection
      end_frame=7932
      start_frame=7849
      channels [
       Volume {type=MovieSceneScriptingFloatChannel, default=1}
       Pitch {type=MovieSceneScriptingFloatChannel, default=1}
      ]
     }
    ]
   }]
 }
 FortPlayerMannequin - Fishstix {
  id=#6
  possessed_class=FortPlayerMannequin
  tracks [{
    type=MovieSceneAudioTrack
    display_name=Audio
    sections [
     {
      type=MovieSceneAudioSection
      end_frame=7598
      start_frame=7116
      channels [
       Volume {type=MovieSceneScriptingFloatChannel, default=1}
       Pitch {type=MovieSceneScriptingFloatChannel, default=1}
      ]
     }
     {
      type=MovieSceneAudioSection
      end_frame=7779
      start_frame=7622
      channels [
       Volume {type=MovieSceneScriptingFloatChannel, default=1}
       Pitch {type=MovieSceneScriptingFloatChannel, default=1}
      ]
     }
    ]
   }]
 }
]
display_rate {denominator=1, numerator=60}
tick_resolution {denominator=1, numerator=24000}
tracks []
