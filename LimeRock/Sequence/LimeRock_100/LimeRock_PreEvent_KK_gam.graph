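# Gameplay Level Sequence for the LimeRock_100 pre-event, which fires timed events on Blueprint actors and has no animation tracks.
# The SequenceRunner takes a regular series of triggers plus a separate "UFO Flyby" event track.
# The UFO HoverManager receives triggers in the middle to later part, and the KnockbackVolume gets a cluster near the end.
# A GoggleBud glider skeletal mesh is bound but has no tracks.
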LevelSequence /LimeRock/Sequence/LimeRock_100/LimeRock_PreEvent_KK_gam
bindings [
 BP_LimeRock_HoverManager_UFO {
  id=#1
  possessed_class=BP_LimeRock_HoverManager_UFO_C
  tracks [{
    type=MovieSceneEventTrack
    sections [{
      type=MovieSceneEventTriggerSection
      channels [None {
        type=MovieSceneScriptingEventChannel
        keys [
         [4470, "<Struct 'MovieSceneEvent' (0x851369a9740) {}>"]
         [4972, "<Struct 'MovieSceneEvent' (0x851369a96a0) {}>"]
         [5970, "<Struct 'MovieSceneEvent' (0x851369a94c0) {}>"]
         [6756, "<Struct 'MovieSceneEvent' (0x851369a92e0) {}>"]
        ]
       }]
     }]
   }]
 }
 BP_LimeRock_SequenceRunner {
  id=#2
  possessed_class=BP_LimeRock_SequenceRunner_C
  tracks [
   {
    type=MovieSceneEventTrack
    sections [{
      type=MovieSceneEventTriggerSection
      channels [None {
        type=MovieSceneScriptingEventChannel
        keys [
         [557, "<Struct 'MovieSceneEvent' (0x851369a5500) {}>"]
         [1159, "<Struct 'MovieSceneEvent' (0x851369a8ca0) {}>"]
         [1801, "<Struct 'MovieSceneEvent' (0x851369a8de0) {}>"]
         [2414, "<Struct 'MovieSceneEvent' (0x851369a8e80) {}>"]
         [2955, "<Struct 'MovieSceneEvent' (0x851369a65e0) {}>"]
         [3568, "<Struct 'MovieSceneEvent' (0x851369a50a0) {}>"]
         [4170, "<Struct 'MovieSceneEvent' (0x851369a7800) {}>"]
         [4822, "<Struct 'MovieSceneEvent' (0x851369a7580) {}>"]
         [5414, "<Struct 'MovieSceneEvent' (0x851369a8660) {}>"]
         [5966, "<Struct 'MovieSceneEvent' (0x851369a8f20) {}>"]
        ]
       }]
     }]
   }
   {
    type=MovieSceneEventTrack
    display_name=UFO Flyby
    sections [{
      type=MovieSceneEventTriggerSection
      channels [None {
        type=MovieSceneScriptingEventChannel
        keys [
         [4219, "<Struct 'MovieSceneEvent' (0x851369a9920) {}>"]
         [4345, "<Struct 'MovieSceneEvent' (0x851369a99c0) {}>"]
         [7602, "<Struct 'MovieSceneEvent' (0x851369a9880) {}>"]
         [8467, "<Struct 'MovieSceneEvent' (0x851369a97e0) {}>"]
        ]
       }]
     }]
   }
  ]
 }
 BP_LimeRock_KnockbackVolume {
  id=#3
  possessed_class=BP_LimeRock_KnockbackVolume_C
  tracks [{
    type=MovieSceneEventTrack
    sections [{
      type=MovieSceneEventTriggerSection
      channels [None {
        type=MovieSceneScriptingEventChannel
        keys [
         [7683, "<Struct 'MovieSceneEvent' (0x851369a8700) {}>"]
         [7868, "<Struct 'MovieSceneEvent' (0x851369a8fc0) {}>"]
         [8069, "<Struct 'MovieSceneEvent' (0x851369a9c40) {}>"]
         [8286, "<Struct 'MovieSceneEvent' (0x851369a9ce0) {}>"]
         [8310, "<Struct 'MovieSceneEvent' (0x851369a9d80) {}>"]
        ]
       }]
     }]
   }]
 }
 GoggleBud_Glider {id=#4, possessed_class=FortSkeletalMeshActor, tracks []}
]
display_rate {denominator=1, numerator=60}
tick_resolution {denominator=1, numerator=24000}
tracks []
