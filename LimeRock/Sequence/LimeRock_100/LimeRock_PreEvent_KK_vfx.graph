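# Level Sequence driving Niagara VFX for UFO moments in the LimeRock_100 pre-event: tractor beams, dust kickups, wind trails and floating effects.
# Spawn tracks time each actor, attach tracks carry beams, trails and floating FX, and keyed transforms reposition each dust burst.
# TractorBeam2 fires briefly whenever TractorBeam ends, with SmallBeam toggling on both.
# Component parameters fade floating FX and pulse ribbons, whose later opacity keys peak at 40, matching spawn rate.
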LevelSequence /LimeRock/Sequence/LimeRock_100/LimeRock_PreEvent_KK_vfx
bindings [
 NiagaraComponent0 {
  id=#1
  parent_id=#2
  possessed_class=NiagaraComponent
  tracks [{
    type=MovieSceneNiagaraFloatParameterTrack
    display_name=Opacity_Mult
    sections [{type=MovieSceneFloatSection, channels [None {type=MovieSceneScriptingFloatChannel, default=1, keys [[7556.6225,1],[7578.4175,0]]}]}]
   }]
 }
 NiagaraComponent0 {
  id=#3
  parent_id=#4
  possessed_class=NiagaraComponent
  tracks [{
    type=MovieSceneNiagaraBoolParameterTrack
    display_name=SmallBeam
    sections [{type=MovieSceneBoolSection, channels [None {type=MovieSceneScriptingBoolChannel, default=false, keys [[4429,false],[7035,true]]}]}]
   }]
 }
 NS_LimeRock_FloatingFX {
  id=#2
  possessed_class=NiagaraActor
  tracks [
   {
    type=MovieSceneSpawnTrack
    sections [{
      type=MovieSceneSpawnSection
      channels [None {type=MovieSceneScriptingBoolChannel, default=true, keys [[6894.6275,false],[7112.7975,true],[7578.4175,false]]}]
     }]
   }
   {
    type=MovieScene3DTransformTrack
    sections [{
      type=MovieScene3DTransformSection
      channels [
       Location.X {type=MovieSceneScriptingDoubleChannel, default=0}
       Location.Y {type=MovieSceneScriptingDoubleChannel, default=0}
       Location.Z {type=MovieSceneScriptingDoubleChannel, default=0}
       Scale.X {type=MovieSceneScriptingDoubleChannel, default=1}
       Scale.Y {type=MovieSceneScriptingDoubleChannel, default=1}
       Scale.Z {type=MovieSceneScriptingDoubleChannel, default=1}
      ]
     }]
   }
   {type=MovieScene3DAttachTrack, sections [{type=MovieScene3DAttachSection, end_frame=8847, start_frame=5675}]}
  ]
 }
 NiagaraComponent0 {
  id=#5
  parent_id=#6
  possessed_class=NiagaraComponent
  tracks [{
    type=MovieSceneNiagaraBoolParameterTrack
    display_name=SmallBeam
    sections [{type=MovieSceneBoolSection, channels [None {type=MovieSceneScriptingBoolChannel, default=true, keys [[5018,false],[7558,true]]}]}]
   }]
 }
 NS_LimeRock_UFODust_Kickup {
  id=#7
  possessed_class=NiagaraActor
  tracks [
   {
    type=MovieSceneSpawnTrack
    sections [{
      type=MovieSceneSpawnSection
      channels [None {
        type=MovieSceneScriptingBoolChannel
        default=false
        keys [[0,false],[4261,true],[4366,false],[7678,true],[7740,false],[7865,true],[7937,false],[8069,true],[8155,false],[8284,true],[8304,false],[8308,true]]
       }]
     }]
   }
   {
    type=MovieScene3DTransformTrack
    sections [{
      type=MovieScene3DTransformSection
      channels [
       Location.X {
        type=MovieSceneScriptingDoubleChannel
        default=-92024.8878
        keys [[4258,-87576.8906],[7678,-87608.2891],[7865,-88913.2812],[8069,-89045.4688],[8284,-89856.7812],[8308,-89410.4766]]
       }
       Location.Y {
        type=MovieSceneScriptingDoubleChannel
        default=-95906.1094
        keys [[4258,-95414.1094],[7678,-94904.7266],[7865,-95616.5],[8069,-93952.6016],[8284,-94579.9297],[8308,-97181.1953]]
       }
       Location.Z {
        type=MovieSceneScriptingDoubleChannel
        default=1063.2822
        keys [[4258,887.2822],[7678,856.7207],[7865,915.2822],[8069,909.7435],[8284,975.0259],[8308,1059.0458]]
       }
       Rotation.X {type=MovieSceneScriptingDoubleChannel, default=0, keys [[7683,-14.1284],[7865,11.8815],[8069,-4.9066],[8284,2.6337],[8308,4.3071]]}
       Rotation.Y {type=MovieSceneScriptingDoubleChannel, default=0, keys [[7683,0],[7865,0.4065],[8069,-2.5247],[8284,1.4667],[8308,-1.1629]]}
       Rotation.Z {type=MovieSceneScriptingDoubleChannel, default=0, keys [[7683,116.9754],[7865,-9.9915],[8069,163.9553],[8284,-10.6276],[8308,29.08]]}
      ]
     }]
   }
  ]
 }
 NS_LimeRock_UFO_TractorBeam2 {
  id=#6
  possessed_class=NiagaraActor
  tracks [
   {
    type=MovieSceneSpawnTrack
    sections [{
      type=MovieSceneSpawnSection
      channels [None {
        type=MovieSceneScriptingBoolChannel
        default=false
        keys [[0,false],[5018,true],[5113.8125,false],[6776,true],[6871.8525,false],[7558,true],[7663.8125,false]]
       }]
     }]
   }
   {type=MovieScene3DAttachTrack, sections [{type=MovieScene3DAttachSection, end_frame=9000, start_frame=-8}]}
   {
    type=MovieScene3DTransformTrack
    sections [{
      type=MovieScene3DTransformSection
      channels [
       Scale.X {type=MovieSceneScriptingDoubleChannel, default=1}
       Scale.Y {type=MovieSceneScriptingDoubleChannel, default=1}
       Scale.Z {type=MovieSceneScriptingDoubleChannel, default=1}
      ]
     }]
   }
  ]
 }
 NS_LimeRock_UFO_WindTrail {
  id=#8
  possessed_class=NiagaraActor
  tracks [
   {
    type=MovieSceneSpawnTrack
    sections [{
      type=MovieSceneSpawnSection
      channels [None {
        type=MovieSceneScriptingBoolChannel
        default=false
        keys [[4157,false],[4158,true],[4328,true],[4331,false],[7599,false],[7616,true],[7617,true],[7769,false],[7843,false],[7860,true],[7981,true],[8013,false],[8041,false],[8058,true],[8179,true],[8211,false],[8272,false],[8289,true],[8540,true],[8572,false]]
       }]
     }]
   }
   {type=MovieScene3DAttachTrack, sections [{type=MovieScene3DAttachSection, end_frame=9709, start_frame=-4}]}
  ]
 }
 NiagaraComponent0 {
  id=#9
  parent_id=#8
  possessed_class=NiagaraComponent
  tracks [
   {
    type=MovieSceneNiagaraFloatParameterTrack
    display_name=Ribbon_SpawnRate
    sections [{
      type=MovieSceneFloatSection
      channels [None {
        type=MovieSceneScriptingFloatChannel
        default=0
        keys [[4158,0],[4175,40],[4296,40],[4328,0],[7599,0],[7616,40],[7737,40],[7769,0],[7843,0],[7860,40],[7981,40],[8013,0],[8041,0],[8058,40],[8179,40],[8211,0],[8272,0],[8289,40],[8540,40],[8572,0]]
       }]
     }]
   }
   {
    type=MovieSceneNiagaraFloatParameterTrack
    display_name=Ribbon_OpacityMult
    sections [{
      type=MovieSceneFloatSection
      channels [None {
        type=MovieSceneScriptingFloatChannel
        default=0
        keys [[4158,0],[4176,1],[4276,1],[4300,0.3],[4328,0],[7599,0],[7617,1],[7717,1],[7741,0.3],[7769,0],[7843,0],[7860,40],[7981,40],[8013,0],[8041,0],[8058,40],[8179,40],[8211,0],[8272,0],[8289,40],[8540,40],[8572,0]]
       }]
     }]
   }
  ]
 }
 NS_LimeRock_UFO_TractorBeam {
  id=#4
  possessed_class=NiagaraActor
  tracks [
   {
    type=MovieSceneSpawnTrack
    sections [{
      type=MovieSceneSpawnSection
      channels [None {
        type=MovieSceneScriptingBoolChannel
        default=false
        keys [[0,false],[4429,true],[5018,false],[5928,true],[6776,false],[7035,true],[7558,false]]
       }]
     }]
   }
   {type=MovieScene3DAttachTrack, sections [{type=MovieScene3DAttachSection, end_frame=9000, start_frame=-8}]}
   {
    type=MovieScene3DTransformTrack
    sections [{
      type=MovieScene3DTransformSection
      channels [
       Scale.X {type=MovieSceneScriptingDoubleChannel, default=1}
       Scale.Y {type=MovieSceneScriptingDoubleChannel, default=1}
       Scale.Z {type=MovieSceneScriptingDoubleChannel, default=1}
      ]
     }]
   }
  ]
 }
 GoggleBud_Glider {id=#10, possessed_class=FortSkeletalMeshActor, tracks []}
]
display_rate {denominator=1, numerator=60}
tick_resolution {denominator=1, numerator=24000}
tracks []
