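# Level Sequence for LimeRock 100 that animates two tentacle creature skeletal mesh actors, presumably for an event shot.
# Each possessed actor holds a constant transform, while its child SkeletalMeshComponent plays a skeletal animation section over frames 0–862.
# Visibility tracks keep both meshes hidden at first, then reveal them almost together just past frame 300.
# The event level instance and scene root are bound without tracks.
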LevelSequence /LimeRock/Sequence/LimeRock_100/Limerock_100_Anim_Tentacles
bindings [
 LimeRock_LevelInstance_Event_01 {id=#1, possessed_class=BuildingLevelInstance, tracks []}
 SkeletalMeshComponent0 {
  id=#2
  parent_id=#3
  possessed_class=SkeletalMeshComponent
  tracks [
   {
    type=MovieSceneVisibilityTrack
    sections [{
      type=MovieSceneVisibilitySection
      channels [None {type=MovieSceneScriptingBoolChannel, default=false, keys [[263,false],[306,true],[629,true]]}]
     }]
   }
   {type=MovieSceneSkeletalAnimationTrack, sections [{type=MovieSceneSkeletalAnimationSection, end_frame=862, start_frame=0}]}
  ]
 }
 Tentacle_LimeRock_Creature {
  id=#4
  possessed_class=FortSkeletalMeshActor
  tracks [{
    type=MovieScene3DTransformTrack
    sections [{
      type=MovieScene3DTransformSection
      channels [
       Location.X {type=MovieSceneScriptingDoubleChannel, default=-14710, const [-8297.29,1]}
       Location.Y {type=MovieSceneScriptingDoubleChannel, default=8178, const [4266.8413,1]}
       Location.Z {type=MovieSceneScriptingDoubleChannel, default=-1194, const [724.345,1]}
       Rotation.X {type=MovieSceneScriptingDoubleChannel, default=0, const [33.6922,1]}
       Rotation.Y {type=MovieSceneScriptingDoubleChannel, default=0, const [0.2543,1]}
       Rotation.Z {type=MovieSceneScriptingDoubleChannel, default=-41, const [-58.0193,1]}
       Scale.X {type=MovieSceneScriptingDoubleChannel, default=1, const [1,1]}
       Scale.Y {type=MovieSceneScriptingDoubleChannel, default=1, const [1,1]}
       Scale.Z {type=MovieSceneScriptingDoubleChannel, default=1, const [1,1]}
      ]
     }]
   }]
 }
 Tentacle_LimeRock_Creature2 {
  id=#3
  possessed_class=FortSkeletalMeshActor
  tracks [{
    type=MovieScene3DTransformTrack
    sections [{
      type=MovieScene3DTransformSection
      channels [
       Location.X {type=MovieSceneScriptingDoubleChannel, default=-10632.6958, const [-8008.1616,1]}
       Location.Y {type=MovieSceneScriptingDoubleChannel, default=-5847.4937, const [-5847.4937,1]}
       Location.Z {type=MovieSceneScriptingDoubleChannel, default=-1092, const [918.2955,1]}
       Rotation.X {type=MovieSceneScriptingDoubleChannel, default=18.1316, const [42.0145,1]}
       Rotation.Y {type=MovieSceneScriptingDoubleChannel, default=0, const [-3.1318,1]}
       Rotation.Z {type=MovieSceneScriptingDoubleChannel, default=241.7812, const [-108.1797,1]}
       Scale.X {type=MovieSceneScriptingDoubleChannel, default=1, const [1,1]}
       Scale.Y {type=MovieSceneScriptingDoubleChannel, default=1, const [1,1]}
       Scale.Z {type=MovieSceneScriptingDoubleChannel, default=1, const [1,1]}
      ]
     }]
   }]
 }
 SkeletalMeshComponent0 {
  id=#5
  parent_id=#4
  possessed_class=SkeletalMeshComponent
  tracks [
   {
    type=MovieSceneVisibilityTrack
    sections [{
      type=MovieSceneVisibilitySection
      channels [None {type=MovieSceneScriptingBoolChannel, default=false, keys [[246,false],[308,true],[422,true]]}]
     }]
   }
   {type=MovieSceneSkeletalAnimationTrack, sections [{type=MovieSceneSkeletalAnimationSection, end_frame=862, start_frame=0}]}
  ]
 }
 LimeRock_SceneRoot {id=#6, possessed_class=Actor, tracks []}
]
display_rate {denominator=1, numerator=30}
tick_resolution {denominator=1, numerator=24000}
tracks []
